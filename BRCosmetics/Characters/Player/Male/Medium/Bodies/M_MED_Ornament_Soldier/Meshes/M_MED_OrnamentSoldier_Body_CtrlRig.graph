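# Control Rig for the medium male Ornament Soldier body cosmetic, procedurally correcting its fringe, tassel and ornament dyn_ bones as the body poses.
# Parent constraints blend fringe and ornament bases between spine_05 and the clavicles or upper arms, and second fringe bones are aimed at arm targets using two-vector quaternion deltas.
# Remapped knee, arm and head angles weight additive local offsets on selected leg, arm and fringe bones.
# A debug axes draw and several unexecuted TransformConstraint chains remain in the graph.
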ControlRigBlueprint /BRCosmetics/Characters/Player/Male/Medium/Bodies/M_MED_Ornament_Soldier/Meshes/M_MED_OrnamentSoldier_Body_CtrlRig
graphs [
 RigVMFunctionLibrary {links [], nodes []}
 RigVMModel {nodes [
   n1 = &RigUnit_BeginExecution()
   &RigUnit_ParentConstraint(
    AdvancedSettings=(InterpolationType=Average,RotationOrderForFilter=XZY)
    Child=(Type=Bone,Name="dyn_fringeBack_01_Base_l")
    Parents=((Item=(Type=Bone,Name="spine_05"),Weight=1),(Item=(Type=Bone,Name="upperarm_l"),Weight=1))
    Weight=1
    ExecutePin=@n1.ExecutePin
   )
   > &RigUnit_ParentConstraint(
    AdvancedSettings=(InterpolationType=Average,RotationOrderForFilter=XZY)
    Child=(Type=Bone,Name="dyn_upperBackOrnament_01_c_Base")
    Parents=((Item=(Type=Bone,Name="spine_05"),Weight=1),(Item=(Type=Bone,Name="clavicle_r"),Weight=1))
    Weight=1
   )
   > &RigUnit_ParentConstraint(
    AdvancedSettings=(InterpolationType=Average,RotationOrderForFilter=XZY)
    Child=(Type=Bone,Name="dyn_fringeTassleBack_Z_01_Base_l")
    Parents=((Item=(Type=Bone,Name="spine_05"),Weight=1),(Item=(Type=Bone,Name="clavicle_l"),Weight=1))
    Weight=1
   )
   n2 = &RigUnit_GetRelativeBoneTransform(Bone=calf_r, Space=thigh_r)
   > /Script/ControlRig.RigUnit_ModifyBoneTransforms(
    BoneToModify=((Bone="dyn_legHi_H_01_Base_r",Transform=(Rotation=(X=0,Y=0,Z=0,W=1),Translation=(X=0,Y=1,Z=-2),Scale3D=(X=1,Y=1,Z=1))))
    Mode=AdditiveLocal
    WeightMaximum=1
    WeightMinimum=0
    Weight=&RigVMFunction_MathFloatRemap(
     SourceMaximum=117
     SourceMinimum=20
     TargetMinimum=0
     bClamp=true
     Value=&RigVMFunction_MathQuaternionToEuler(RotationOrder=ZYX, Value=@n2.Transform.Rotation).Result.Z
    ).Result
    BoneToModify.0.Transform.Rotation=&RigVMFunction_MathQuaternionFromEuler(Euler=(X=0,Y=50,Z=-30)).Result
   ) {is=struct RigVMUnitNode}
   n3 = &RigUnit_GetRelativeBoneTransform(Bone=calf_r, Space=thigh_r)
   > &RigUnit_ModifyBoneTransforms(
    BoneToModify=((Bone="dyn_legLow_BB_01_Base_l",Transform=(Rotation=(X=0,Y=0,Z=0,W=1),Translation=(X=3,Y=3,Z=4),Scale3D=(X=1,Y=1,Z=1))))
    Mode=AdditiveLocal
    WeightMaximum=1
    WeightMinimum=0
    Weight=&RigVMFunction_MathFloatRemap(
     SourceMaximum=117
     SourceMinimum=20
     TargetMinimum=0
     bClamp=true
     Value=&RigVMFunction_MathQuaternionToEuler(RotationOrder=ZYX, Value=@n3.Transform.Rotation).Result.Z
    ).Result
    BoneToModify.0.Transform.Rotation=&RigVMFunction_MathQuaternionFromEuler(Euler=(X=0,Y=-30,Z=0)).Result
   )
   > &RigUnit_ParentConstraint(
    AdvancedSettings=(InterpolationType=Average,RotationOrderForFilter=XZY)
    Child=(Type=Bone,Name="dyn_fringeBack_01_Base_l")
    Parents=((Item=(Type=Bone,Name="spine_05"),Weight=1),(Item=(Type=Bone,Name="upperarm_l"),Weight=1))
    Weight=1
   )
   > &RigUnit_ParentConstraint(
    AdvancedSettings=(InterpolationType=Average,RotationOrderForFilter=XZY)
    Child=(Type=Bone,Name="dyn_fringeBack_01_Base_r")
    Parents=((Item=(Type=Bone,Name="spine_05"),Weight=1),(Item=(Type=Bone,Name="upperarm_r"),Weight=1))
    Weight=1
   )
   n4 = &RigUnit_GetBoneTransform(Bone=dyn_fringeBack_02_l)
   n5 = /Script/RigVM.RigVMFunction_MathQuaternionGetAxis(Axis=X, Quaternion=@n4.Transform.Rotation) {is=struct RigVMUnitNode, template="Axis::Execute(in Quaternion,in Axis,out Result)"}
   n6 = /Script/RigVM.RigVMFunction_MathVectorSub(
    B=@n4.Transform.Translation
    A=/Script/RigVM.RigVMFunction_MathTransformTransformVector(Location=(X=0,Y=0,Z=0), Transform=&RigUnit_GetBoneTransform(Bone=dyn_fringeArmAimAt_l).Transform) {is=struct RigVMUnitNode}.Result
   ) {is=struct RigVMUnitNode, template="Subtract::Execute(in A,in B,out Result)"}
   > /Script/ControlRig.RigUnit_SetBoneRotation(
    Bone=dyn_fringeBack_02_l
    Space=GlobalSpace
    Weight=1
    bPropagateToChildren=True
    Rotation=/Script/RigVM.RigVMFunction_MathQuaternionMul(
     B=@n4.Transform.Rotation
     A=/Script/RigVM.RigVMFunction_MathQuaternionFromTwoVectors(A=@n5.Result, B=@n6.Result) {is=struct RigVMUnitNode, template="FromTwoVectors::Execute(in A,in B,out Result)"}.Result
    ) {is=struct RigVMUnitNode, template="Multiply::Execute(in A,in B,out Result)"}.Result
   ) {is=struct RigVMUnitNode}
   n7 = &RigUnit_GetBoneTransform(Bone=dyn_fringeBack_02_r)
   n8 = &RigVMFunction_MathQuaternionGetAxis(Axis=X, Quaternion=@n7.Transform.Rotation) {template="Axis::Execute(in Quaternion,in Axis,out Result)"}
   n9 = &RigVMFunction_MathVectorSub(
    B=@n7.Transform.Translation
    A=&RigVMFunction_MathTransformTransformVector(Location=(X=0,Y=0,Z=0), Transform=&RigUnit_GetBoneTransform(Bone=dyn_fringeArmAimAt_r).Transform).Result
   ) {template="Subtract::Execute(in A,in B,out Result)"}
   > &RigUnit_SetBoneRotation(
    Bone=dyn_fringeBack_02_r
    Space=GlobalSpace
    Weight=1
    bPropagateToChildren=True
    Rotation=&RigVMFunction_MathQuaternionMul(
     B=@n7.Transform.Rotation
     A=&RigVMFunction_MathQuaternionFromTwoVectors(A=@n8.Result, B=@n9.Result) {template="FromTwoVectors::Execute(in A,in B,out Result)"}.Result
    ) {template="Multiply::Execute(in A,in B,out Result)"}.Result
   )
   > &RigUnit_ParentConstraint(
    AdvancedSettings=(InterpolationType=Average,RotationOrderForFilter=XZY)
    Child=(Type=Bone,Name="dyn_fringeFront_01_Base_l")
    Parents=((Item=(Type=Bone,Name="spine_05"),Weight=1),(Item=(Type=Bone,Name="upperarm_l"),Weight=1))
    Weight=1
   )
   > &RigUnit_ParentConstraint(
    AdvancedSettings=(InterpolationType=Average,RotationOrderForFilter=XZY)
    Child=(Type=Bone,Name="dyn_fringeFront_01_Base_r")
    Parents=((Item=(Type=Bone,Name="spine_05"),Weight=1),(Item=(Type=Bone,Name="upperarm_r"),Weight=1))
    Weight=1
   )
   n10 = &RigUnit_GetBoneTransform(Bone=dyn_fringeFront_02_l)
   n11 = &RigVMFunction_MathQuaternionGetAxis(Axis=X, Quaternion=@n10.Transform.Rotation) {template="Axis::Execute(in Quaternion,in Axis,out Result)"}
   n12 = &RigVMFunction_MathVectorSub(
    B=@n10.Transform.Translation
    A=&RigVMFunction_MathTransformTransformVector(Location=(X=0,Y=0,Z=0), Transform=&RigUnit_GetBoneTransform(Bone=dyn_fringeArmAimAt_l).Transform).Result
   ) {template="Subtract::Execute(in A,in B,out Result)"}
   > &RigUnit_SetBoneRotation(
    Bone=dyn_fringeFront_02_l
    Space=GlobalSpace
    Weight=1
    bPropagateToChildren=True
    Rotation=&RigVMFunction_MathQuaternionMul(
     B=@n10.Transform.Rotation
     A=&RigVMFunction_MathQuaternionFromTwoVectors(A=@n11.Result, B=@n12.Result) {template="FromTwoVectors::Execute(in A,in B,out Result)"}.Result
    ) {template="Multiply::Execute(in A,in B,out Result)"}.Result
   )
   n13 = &RigUnit_GetBoneTransform(Bone=dyn_fringeFront_02_r)
   n14 = &RigVMFunction_MathQuaternionGetAxis(Axis=X, Quaternion=@n13.Transform.Rotation) {template="Axis::Execute(in Quaternion,in Axis,out Result)"}
   n15 = &RigVMFunction_MathTransformTransformVector(Location=(X=-3,Y=4,Z=0), Transform=&RigUnit_GetBoneTransform(Bone=upperarm_r).Transform)
   n16 = &RigVMFunction_MathVectorSub(B=@n13.Transform.Translation, A=@n15.Result) {template="Subtract::Execute(in A,in B,out Result)"}
   > &RigUnit_SetBoneRotation(
    Bone=dyn_fringeFront_02_r
    Space=GlobalSpace
    Weight=1
    bPropagateToChildren=True
    Rotation=&RigVMFunction_MathQuaternionMul(
     B=@n13.Transform.Rotation
     A=&RigVMFunction_MathQuaternionFromTwoVectors(A=@n14.Result, B=@n16.Result) {template="FromTwoVectors::Execute(in A,in B,out Result)"}.Result
    ) {template="Multiply::Execute(in A,in B,out Result)"}.Result
   )
   n17 = &RigVMFunction_MathFloatRemap(
    SourceMaximum=146
    SourceMinimum=70
    TargetMinimum=0
    bClamp=true
    Value=&RigVMFunction_MathQuaternionToEuler(RotationOrder=ZYX, Value=&RigUnit_GetRelativeBoneTransform(Bone=upperarm_r, Space=clavicle_r).Transform.Rotation).Result.Z
   )
   > &RigUnit_ModifyBoneTransforms(
    BoneToModify=((Bone="dyn_fringeFront_01_r",Transform=(Rotation=(X=0,Y=0,Z=0,W=1),Translation=(X=-5,Y=0,Z=0),Scale3D=(X=1,Y=1,Z=1))))
    Mode=AdditiveLocal
    WeightMaximum=1
    WeightMinimum=0
    Weight=@n17.Result
   )
   > &RigUnit_ModifyBoneTransforms(
    BoneToModify=((Bone="dyn_ornamentArm_A_01_Base_r",Transform=(Rotation=(X=0,Y=0,Z=0,W=1),Translation=(X=0,Y=0,Z=0),Scale3D=(X=1,Y=1,Z=1))))
    Mode=AdditiveLocal
    WeightMaximum=1
    WeightMinimum=0
    Weight=@n17.Result
    BoneToModify.0.Transform.Rotation=&RigVMFunction_MathQuaternionFromEuler(Euler=(X=0,Y=-40,Z=12)).Result
   )
   n18 = &RigVMFunction_MathQuaternionToEuler(RotationOrder=ZYX, Value=&RigUnit_GetRelativeBoneTransform(Bone=upperarm_l, Space=spine_05).Transform.Rotation)
   > &RigUnit_ModifyBoneTransforms(
    BoneToModify=((Bone="dyn_fringeFront_01_l",Transform=(Rotation=(X=0,Y=0,Z=0,W=1),Translation=(X=-5,Y=0,Z=0),Scale3D=(X=1,Y=1,Z=1))))
    Mode=AdditiveLocal
    WeightMaximum=1
    WeightMinimum=0
    Weight=&RigVMFunction_MathFloatRemap(
     SourceMaximum=-65
     SourceMinimum=-170
     TargetMinimum=0
     bClamp=true
     Value=/Script/RigVM.RigVMFunction_MathFloatAbs(Value=@n18.Result.Z) {is=struct RigVMUnitNode, template="Absolute::Execute(in Value,out Result)"}.Result
    ).Result
   )
   n19 = &RigUnit_GetRelativeBoneTransform(Bone=upperarm_r, Space=clavicle_r)
   > &RigUnit_ModifyBoneTransforms(
    BoneToModify=((Bone="dyn_fringeBack_01_r",Transform=(Rotation=(X=0,Y=0,Z=0,W=1),Translation=(X=0,Y=0,Z=0),Scale3D=(X=1,Y=1,Z=1))))
    Mode=AdditiveLocal
    WeightMaximum=1
    WeightMinimum=0
    Weight=&RigVMFunction_MathFloatRemap(
     SourceMaximum=146
     SourceMinimum=70
     TargetMinimum=0
     bClamp=true
     Value=&RigVMFunction_MathQuaternionToEuler(RotationOrder=ZYX, Value=@n19.Transform.Rotation).Result.Z
    ).Result
   )
   n20 = &RigVMFunction_MathQuaternionToEuler(RotationOrder=ZYX, Value=&RigUnit_GetRelativeBoneTransform(Bone=upperarm_l, Space=spine_05).Transform.Rotation)
   > &RigUnit_ModifyBoneTransforms(
    BoneToModify=((Bone="dyn_fringeBack_01_l",Transform=(Rotation=(X=0,Y=0,Z=0,W=1),Translation=(X=0,Y=0,Z=0),Scale3D=(X=1,Y=1,Z=1))))
    Mode=AdditiveLocal
    WeightMaximum=1
    WeightMinimum=0
    Weight=&RigVMFunction_MathFloatRemap(
     SourceMaximum=65
     SourceMinimum=125
     TargetMinimum=0
     bClamp=true
     Value=&RigVMFunction_MathFloatAbs(Value=@n20.Result.Z) {template="Absolute::Execute(in Value,out Result)"}.Result
    ).Result
   )
   n21 = &RigUnit_GetRelativeBoneTransform(Bone=head, Space=spine_05)
   > &RigUnit_ModifyBoneTransforms(
    BoneToModify=((Bone="dyn_fringeBack_01_Base_l",Transform=(Rotation=(X=0,Y=0,Z=0,W=1),Translation=(X=0,Y=0,Z=-1),Scale3D=(X=1,Y=1,Z=1))))
    Mode=AdditiveLocal
    WeightMaximum=1
    WeightMinimum=0
    Weight=&RigVMFunction_MathFloatRemap(
     SourceMaximum=-41
     SourceMinimum=50
     TargetMinimum=0
     bClamp=true
     Value=&RigVMFunction_MathQuaternionToEuler(RotationOrder=ZYX, Value=@n21.Transform.Rotation).Result.Z
    ).Result
   )
   > &RigUnit_ModifyBoneTransforms(
    BoneToModify=((Bone="dyn_fringeBack_01_Base_r",Transform=(Rotation=(X=0,Y=0,Z=0,W=1),Translation=(X=0,Y=0,Z=2.5),Scale3D=(X=1,Y=1,Z=1))))
    Mode=AdditiveLocal
    Weight=1
    WeightMaximum=1
    WeightMinimum=0.6
   )
   > /Script/ControlRig.RigUnit_DebugTransformMutable(
    Color=(R=1,G=0,B=0,A=1)
    DebugDrawSettings=(DepthPriority=SDPG_Foreground,Lifetime=-1)
    Mode=Axes
    Scale=10
    Space=None
    Thickness=1
    Transform=(Rotation=(X=0,Y=0,Z=0,W=1),Translation=(X=0,Y=0,Z=0),Scale3D=(X=1,Y=1,Z=1))
    WorldOffset=(Rotation=(X=0,Y=0,Z=0,W=1),Translation=(X=0,Y=0,Z=0),Scale3D=(X=1,Y=1,Z=1))
    bEnabled=True
    Transform.Translation=@n15.Result
   ) {is=struct RigVMUnitNode}
   > &RigUnit_ParentConstraint(
    AdvancedSettings=(InterpolationType=Average,RotationOrderForFilter=XZY)
    Child=(Type=Bone,Name="dyn_fringeTassleFront_F_01_Base_r")
    Parents=((Item=(Type=Bone,Name="dyn_fringeFront_02_r"),Weight=0.2),(Item=(Type=Bone,Name="dyn_fringeFront_01_r"),Weight=0.8))
    Weight=1
   )
   n22 = &RigUnit_GetBoneTransform(Bone=spine_05)
   /Script/ControlRig.RigUnit_TransformConstraint(
    BaseBone=None
    BaseTransform=(Rotation=(X=0,Y=0,Z=0,W=1),Translation=(X=0,Y=0,Z=0),Scale3D=(X=1,Y=1,Z=1))
    BaseTransformSpace=GlobalSpace
    Bone=dyn_fringeBack_01_Base_l
    ExecutePin=()
    Targets="((Transform=(Rotation=(X=0,Y=0,Z=0,W=1),Translation=(X=0,Y=0,Z=0),Scale3D=(X=1,Y=1,Z=1)),Weight=1,bMaintainOffset=True,Filter=(TranslationFilter=(bX=True,bY=True,bZ=True),RotationFilter=(bX=True,bY=True,bZ=True),ScaleFilter=(bX=True,bY=True,bZ=True))),(Tra ...(truncated)"
    bUseInitialTransforms=True
    Targets.0.Transform.Rotation=@n22.Transform.Rotation
    Targets.1.Transform.Rotation=@n22.Transform.Rotation
    Targets.0.Transform.Translation=@n22.Transform.Translation
    Targets.1.Transform.Translation=&RigUnit_GetBoneTransform(Bone=upperarm_l).Transform.Translation
   ) {is=struct RigVMUnitNode}
   n23 = &RigUnit_GetBoneTransform(Bone=spine_05)
   > &RigUnit_TransformConstraint(
    BaseBone=None
    BaseTransform=(Rotation=(X=0,Y=0,Z=0,W=1),Translation=(X=0,Y=0,Z=0),Scale3D=(X=1,Y=1,Z=1))
    BaseTransformSpace=GlobalSpace
    Bone=dyn_fringeBack_01_Base_r
    Targets="((Transform=(Rotation=(X=0,Y=0,Z=0,W=1),Translation=(X=0,Y=0,Z=0),Scale3D=(X=1,Y=1,Z=1)),Weight=1,bMaintainOffset=True,Filter=(TranslationFilter=(bX=True,bY=True,bZ=True),RotationFilter=(bX=True,bY=True,bZ=True),ScaleFilter=(bX=True,bY=True,bZ=True))),(Tra ...(truncated)"
    bUseInitialTransforms=True
    Targets.1.Transform.Translation=&RigUnit_GetBoneTransform(Bone=upperarm_r).Transform.Translation
    Targets.0.Transform.Rotation=@n23.Transform.Rotation
    Targets.1.Transform.Rotation=@n23.Transform.Rotation
    Targets.0.Transform.Translation=@n23.Transform.Translation
   )
   n24 = &RigUnit_GetBoneTransform(Bone=spine_05)
   &RigUnit_TransformConstraint(
    BaseBone=None
    BaseTransform=(Rotation=(X=0,Y=0,Z=0,W=1),Translation=(X=0,Y=0,Z=0),Scale3D=(X=1,Y=1,Z=1))
    BaseTransformSpace=GlobalSpace
    Bone=dyn_fringeFront_01_Base_l
    ExecutePin=()
    Targets="((Transform=(Rotation=(X=0,Y=0,Z=0,W=1),Translation=(X=0,Y=0,Z=0),Scale3D=(X=1,Y=1,Z=1)),Weight=1,bMaintainOffset=True,Filter=(TranslationFilter=(bX=True,bY=True,bZ=True),RotationFilter=(bX=True,bY=True,bZ=True),ScaleFilter=(bX=True,bY=True,bZ=True))),(Tra ...(truncated)"
    bUseInitialTransforms=True
    Targets.1.Transform.Translation=&RigUnit_GetBoneTransform(Bone=upperarm_l).Transform.Translation
    Targets.0.Transform.Rotation=@n24.Transform.Rotation
    Targets.1.Transform.Rotation=@n24.Transform.Rotation
    Targets.0.Transform.Translation=@n24.Transform.Translation
   )
   n25 = &RigUnit_GetBoneTransform(Bone=spine_05)
   > &RigUnit_TransformConstraint(
    BaseBone=None
    BaseTransform=(Rotation=(X=0,Y=0,Z=0,W=1),Translation=(X=0,Y=0,Z=0),Scale3D=(X=1,Y=1,Z=1))
    BaseTransformSpace=GlobalSpace
    Bone=dyn_fringeFront_01_Base_r
    Targets="((Transform=(Rotation=(X=0,Y=0,Z=0,W=1),Translation=(X=0,Y=0,Z=0),Scale3D=(X=1,Y=1,Z=1)),Weight=1,bMaintainOffset=True,Filter=(TranslationFilter=(bX=True,bY=True,bZ=True),RotationFilter=(bX=True,bY=True,bZ=True),ScaleFilter=(bX=True,bY=True,bZ=True))),(Tra ...(truncated)"
    bUseInitialTransforms=True
    Targets.1.Transform.Translation=&RigUnit_GetBoneTransform(Bone=upperarm_r).Transform.Translation
    Targets.0.Transform.Rotation=@n25.Transform.Rotation
    Targets.1.Transform.Rotation=@n25.Transform.Rotation
    Targets.0.Transform.Translation=@n25.Transform.Translation
   )
   n26 = &RigUnit_GetBoneTransform(Bone=spine_05)
   &RigUnit_TransformConstraint(
    BaseBone=None
    BaseTransform=(Rotation=(X=0,Y=0,Z=0,W=1),Translation=(X=0,Y=0,Z=0),Scale3D=(X=1,Y=1,Z=1))
    BaseTransformSpace=GlobalSpace
    Bone=dyn_fringeBack_01_Base_l
    ExecutePin=()
    Targets="((Transform=(Rotation=(X=0,Y=0,Z=0,W=1),Translation=(X=0,Y=0,Z=0),Scale3D=(X=1,Y=1,Z=1)),Weight=1,bMaintainOffset=True,Filter=(TranslationFilter=(bX=True,bY=True,bZ=True),RotationFilter=(bX=True,bY=True,bZ=True),ScaleFilter=(bX=True,bY=True,bZ=True))),(Tra ...(truncated)"
    bUseInitialTransforms=True
    Targets.1.Transform.Translation=&RigUnit_GetBoneTransform(Bone=upperarm_l).Transform.Translation
    Targets.0.Transform.Rotation=@n26.Transform.Rotation
    Targets.1.Transform.Rotation=@n26.Transform.Rotation
    Targets.0.Transform.Translation=@n26.Transform.Translation
   )
   n27 = &RigUnit_GetBoneTransform(Bone=spine_05)
   > &RigUnit_TransformConstraint(
    BaseBone=None
    BaseTransform=(Rotation=(X=0,Y=0,Z=0,W=1),Translation=(X=0,Y=0,Z=0),Scale3D=(X=1,Y=1,Z=1))
    BaseTransformSpace=GlobalSpace
    Bone=dyn_upperBackOrnament_01_c_Base
    Targets="((Transform=(Rotation=(X=0,Y=0,Z=0,W=1),Translation=(X=0,Y=0,Z=0),Scale3D=(X=1,Y=1,Z=1)),Weight=1,bMaintainOffset=True,Filter=(TranslationFilter=(bX=True,bY=True,bZ=True),RotationFilter=(bX=True,bY=True,bZ=True),ScaleFilter=(bX=True,bY=True,bZ=True))),(Tra ...(truncated)"
    bUseInitialTransforms=True
    Targets.1.Transform.Translation=&RigUnit_GetBoneTransform(Bone=clavicle_r).Transform.Translation
    Targets.0.Transform.Rotation=@n27.Transform.Rotation
    Targets.1.Transform.Rotation=@n27.Transform.Rotation
    Targets.0.Transform.Translation=@n27.Transform.Translation
   )
   n28 = &RigUnit_GetBoneTransform(Bone=clavicle_l)
   > &RigUnit_TransformConstraint(
    BaseBone=None
    BaseTransform=(Rotation=(X=0,Y=0,Z=0,W=1),Translation=(X=0,Y=0,Z=0),Scale3D=(X=1,Y=1,Z=1))
    BaseTransformSpace=GlobalSpace
    Bone=dyn_fringeTassleBack_Z_01_Base_l
    Targets="((Transform=(Rotation=(X=0,Y=0,Z=0,W=1),Translation=(X=0,Y=0,Z=0),Scale3D=(X=1,Y=1,Z=1)),Weight=0.5,bMaintainOffset=True,Filter=(TranslationFilter=(bX=True,bY=True,bZ=True),RotationFilter=(bX=True,bY=True,bZ=True),ScaleFilter=(bX=True,bY=True,bZ=True))),(T ...(truncated)"
    bUseInitialTransforms=True
    Targets.1.Transform.Rotation=@n28.Transform.Rotation
    Targets.1.Transform.Translation=@n28.Transform.Translation
    Targets.0.Transform.Rotation=@n27.Transform.Rotation
    Targets.0.Transform.Translation=@n27.Transform.Translation
   )
   &RigUnit_GetBoneTransform(Bone=upperarm_r)
   &RigUnit_GetBoneTransform(Bone=spine_05)
   &RigVMFunction_MathQuaternionFromEuler(Euler=(X=0,Y=50,Z=-30))
   /Script/RigVM.RigVMFunction_MathVectorNegate(Value=@n5.Result) {is=struct RigVMUnitNode, template="Negate::Execute(in Value,out Result)"}
   &RigVMFunction_MathVectorNegate(Value=@n8.Result) {template="Negate::Execute(in Value,out Result)"}
   &RigVMFunction_MathVectorNegate(Value=@n14.Result) {template="Negate::Execute(in Value,out Result)"}
   &RigVMFunction_MathVectorNegate(Value=@n11.Result) {template="Negate::Execute(in Value,out Result)"}
   n29 = &RigUnit_GetBoneTransform(Bone=dyn_fringeFront_01_r)
   n30 = &RigUnit_GetBoneTransform(Bone=dyn_fringeFront_02_r)
   &RigUnit_TransformConstraint(
    BaseBone=None
    BaseTransform=(Rotation=(X=0,Y=0,Z=0,W=1),Translation=(X=0,Y=0,Z=0),Scale3D=(X=1,Y=1,Z=1))
    BaseTransformSpace=GlobalSpace
    Bone=dyn_fringeTassleFront_F_01_Base_r
    ExecutePin=()
    Targets="((Transform=(Rotation=(X=0,Y=0,Z=0,W=1),Translation=(X=0,Y=0,Z=0),Scale3D=(X=1,Y=1,Z=1)),Weight=0.2,bMaintainOffset=True,Filter=(TranslationFilter=(bX=True,bY=True,bZ=True),RotationFilter=(bX=True,bY=True,bZ=True),ScaleFilter=(bX=True,bY=True,bZ=True))),(T ...(truncated)"
    bUseInitialTransforms=True
    Targets.1.Transform.Rotation=@n29.Transform.Rotation
    Targets.1.Transform.Translation=@n29.Transform.Translation
    Targets.0.Transform.Rotation=@n30.Transform.Rotation
    Targets.0.Transform.Translation=@n30.Transform.Translation
   )
  ]}
]
hierarchy [
 root {loc=<0 0 0>, rot=<0 0 0 1>}
 pelvis {parent=root, loc=<0 -0.7071 93.1191>, rot=<0.0127 0.707 -0.0127 -0.707>}
 spine_01 {parent=pelvis, loc=<6.7494 -0.6388 0>, rot=<0 0 0.034 -0.9994>}
 spine_02 {parent=spine_01, loc=<7.2518 -0.035 0>, rot=<0 0 0.0238 -0.9997>}
 spine_03 {parent=spine_02, loc=<6.0104 0 0>, rot=<0 0 0.0149 -0.9999>}
 spine_04 {parent=spine_03, loc=<8.1024 0 0>, rot=<0 0 -0.0132 -0.9999>}
 spine_05 {parent=spine_04, loc=<10.0146 0 0>, rot=<0 0 -0.0015 -1>}
 clavicle_l {parent=spine_05, loc=<4.3575 -0.1466 -2.3223>, rot=<0.0465 -0.7289 -0.009 -0.6829>}
 upperarm_l {parent=clavicle_l, loc=<15.5111 0 0>, rot=<0.0226 -0.3234 -0.0088 -0.9459>}
 lowerarm_l {parent=upperarm_l, loc=<26.6514 0 0>, rot=<0.0391 -0.0099 -0.2453 -0.9686>}
 hand_l {parent=lowerarm_l, loc=<22.3491 0 0.0001>, rot=<-0.7081 -0.0507 0.0975 -0.6975>}
 index_metacarpal_l {parent=hand_l, loc=<1.6535 0.0183 -2.96>, rot=<0.0158 -0.0124 -0.0173 -0.9996>}
 index_01_l {parent=index_metacarpal_l, loc=<8.3008 0.0357 -0.0524>, rot=<-0.0044 0.0485 0.0924 -0.9945>}
 index_02_l {parent=index_01_l, loc=<3.3587 -0.3004 0.0452>, rot=<-0.002 0.0027 0.3225 -0.9466>}
 index_03_l {parent=index_02_l, loc=<2.5578 -0.123 -0.0252>, rot=<-0.002 0.0067 0.2558 -0.9667>}
 middle_metacarpal_l {parent=hand_l, loc=<2.1856 -0.1725 -0.5984>, rot=<0.006 0.0168 -0.0329 -0.9993>}
 middle_01_l {parent=middle_metacarpal_l, loc=<7.3275 -0.2447 0.0129>, rot=<-0.0208 0.0042 0.0858 -0.9961>}
 middle_02_l {parent=middle_01_l, loc=<4.0123 0.1401 0.1175>, rot=<-0.0218 0.0258 0.3512 -0.9357>}
 middle_03_l {parent=middle_02_l, loc=<2.6077 0.0534 0.1373>, rot=<0.0082 0.0188 0.1708 -0.9851>}
 pinky_metacarpal_l {parent=hand_l, loc=<1.7616 -0.9753 3.3814>, rot=<-0.0397 0.0814 -0.069 -0.9935>}
 pinky_01_l {parent=pinky_metacarpal_l, loc=<7.1709 -0.4831 -0.1094>, rot=<-0.0654 -0.0035 0.2242 -0.9723>}
 pinky_02_l {parent=pinky_01_l, loc=<2.8274 0.0893 -0.1043>, rot=<0.0037 0.0011 0.2854 -0.9584>}
 pinky_03_l {parent=pinky_02_l, loc=<2.2046 0.0307 -0.0078>, rot=<-0.0002 0.0098 0.1445 -0.9895>}
 ring_metacarpal_l {parent=hand_l, loc=<2.2074 -0.3625 1.3638>, rot=<-0.0284 0.0722 -0.0458 -0.9959>}
 ring_01_l {parent=ring_metacarpal_l, loc=<7.0295 -0.6197 -0.1446>, rot=<-0.0489 -0.0252 0.1892 -0.9804>}
 ring_02_l {parent=ring_01_l, loc=<3.6174 0.4646 -0.1027>, rot=<0 0 0.2806 -0.9598>}
 ring_03_l {parent=ring_02_l, loc=<2.4152 0.1141 0.0199>, rot=<0 0 0.1189 -0.9929>}
 thumb_01_l {parent=hand_l, loc=<1.3118 -0.3191 -2.459>, rot=<0.6546 -0.4042 0.0664 -0.6354>}
 thumb_02_l {parent=thumb_01_l, loc=<5.9336 -0.0448 0.0401>, rot=<-0.0037 0.0213 0.298 -0.9543>}
 thumb_03_l {parent=thumb_02_l, loc=<2.2223 0.3682 -0.0907>, rot=<0.0118 -0.013 0.0066 -0.9998>}
 lowerarm_twist_01_l {parent=lowerarm_l, loc=<17.8549 0 0>, rot=<0 0 0 1>}
 lowerarm_twist_02_l {parent=lowerarm_l, loc=<9.6199 0 0>, rot=<0 0 0 1>}
 upperarm_twist_01_l {parent=upperarm_l, loc=<0.0041 0 0>, rot=<0 0 0 -1>}
 upperarm_twist_02_l {parent=upperarm_l, loc=<13.7941 0 0>, rot=<0 0 0 -1>}
 dyn_fringeArm_A_01_Base_l {parent=upperarm_twist_02_l, loc=<-11.2136 -6.9552 3.2553>, rot=<0.4481 0.0742 0.0979 -0.8855>}
 dyn_fringeArm_A_01_l {parent=dyn_fringeArm_A_01_Base_l, loc=<0 0 0>, rot=<0 0 0 1>}
 dyn_fringeArm_C_01_Base_l {parent=upperarm_twist_02_l, loc=<-6.9396 7.4242 5.1715>, rot=<0.8377 0.0492 -0.2704 -0.472>}
 dyn_fringeArm_C_01_l {parent=dyn_fringeArm_C_01_Base_l, loc=<0 0 0>, rot=<0 0 0 -1>}
 dyn_fringeArm_B_01_Base_l {parent=upperarm_twist_02_l, loc=<-6.7934 3.685 5.1459>, rot=<0.6933 0.0396 0.0321 -0.7189>}
 dyn_fringeArm_B_01_l {parent=dyn_fringeArm_B_01_Base_l, loc=<0 0 0>, rot=<0 0 0 1>}
 dyn_fringeArm_B_02_l {parent=dyn_fringeArm_B_01_l, loc=<6.027 0 0>, rot=<-0.0042 -0.0113 0.2637 -0.9645>}
 dyn_fringeArmAimAt_l {parent=upperarm_l, loc=<3.771 0.338 4.744>, rot=<0.6381 -0.589 -0.4953 -0.0239>}
 dyn_fringeArm_D_01_Base_l {parent=upperarm_l, loc=<23.1028 -2.4119 -5.4068>, rot=<-0.6063 -0.0091 -0.0991 -0.789>}
 dyn_fringeArm_D_01_l {parent=dyn_fringeArm_D_01_Base_l, loc=<0 0 0>, rot=<0 0 0 1>}
 dyn_fringeArm_E_01_Base_l {parent=upperarm_l, loc=<22.9599 -0.0489 -5.7861>, rot=<-0.5942 -0.0457 -0.0848 -0.7985>}
 dyn_fringeArm_E_01_l {parent=dyn_fringeArm_E_01_Base_l, loc=<0 0 0>, rot=<0 0 0 -1>}
 dyn_ornamentArm_01_Base_l {parent=clavicle_l, loc=<18.7218 -0.7095 7.2159>, rot=<0.7113 0.0615 -0.1446 -0.6851>}
 dyn_ornamentArm_01_l {parent=dyn_ornamentArm_01_Base_l, loc=<0 0 0>, rot=<0 0 0 1>}
 dyn_ornamentArm_02_l {parent=dyn_ornamentArm_01_l, loc=<3.1981 0 0>, rot=<-0.0214 -0.072 -0.1021 -0.9919>}
 clavicle_r {parent=spine_05, loc=<4.3575 -0.1466 2.3223>, rot=<0.7289 0.0465 -0.6829 0.009>}
 upperarm_r {parent=clavicle_r, loc=<-15.5111 0 0>, rot=<0.0226 -0.3234 -0.0088 -0.9459>}
 lowerarm_r {parent=upperarm_r, loc=<-26.6514 0 0>, rot=<0.0391 -0.0099 -0.2453 -0.9686>}
 hand_r {parent=lowerarm_r, loc=<-22.3492 0 0>, rot=<-0.7081 -0.0507 0.0975 -0.6975>}
 index_metacarpal_r {parent=hand_r, loc=<-1.6534 -0.0181 2.9601>, rot=<0.0158 -0.0124 -0.0173 -0.9996>}
 index_01_r {parent=index_metacarpal_r, loc=<-8.3008 -0.0357 0.0524>, rot=<-0.0044 0.0485 0.0924 -0.9945>}
 index_02_r {parent=index_01_r, loc=<-3.3589 0.3003 -0.0453>, rot=<-0.002 0.0027 0.3225 -0.9466>}
 index_03_r {parent=index_02_r, loc=<-2.5578 0.1231 0.0251>, rot=<-0.002 0.0067 0.2558 -0.9667>}
 middle_metacarpal_r {parent=hand_r, loc=<-2.1856 0.1726 0.5984>, rot=<0.006 0.0168 -0.0329 -0.9993>}
 middle_01_r {parent=middle_metacarpal_r, loc=<-7.3273 0.2447 -0.0129>, rot=<-0.0208 0.0042 0.0858 -0.9961>}
 middle_02_r {parent=middle_01_r, loc=<-4.0141 -0.1411 -0.0699>, rot=<-0.0217 0.0259 0.3512 -0.9357>}
 middle_03_r {parent=middle_02_r, loc=<-2.6074 -0.0534 -0.1312>, rot=<0.0083 0.0188 0.1708 -0.9851>}
 pinky_metacarpal_r {parent=hand_r, loc=<-1.7617 0.9753 -3.3814>, rot=<-0.0397 0.0814 -0.069 -0.9935>}
 pinky_01_r {parent=pinky_metacarpal_r, loc=<-7.1707 0.4832 0.1094>, rot=<-0.0654 -0.0035 0.2242 -0.9723>}
 pinky_02_r {parent=pinky_01_r, loc=<-2.8274 -0.0892 0.1043>, rot=<0.0037 0.0011 0.2854 -0.9584>}
 pinky_03_r {parent=pinky_02_r, loc=<-2.2045 -0.031 0.0078>, rot=<-0.0002 0.0098 0.1445 -0.9895>}
 ring_metacarpal_r {parent=hand_r, loc=<-2.2073 0.3627 -1.3638>, rot=<-0.0284 0.0722 -0.0458 -0.9959>}
 ring_01_r {parent=ring_metacarpal_r, loc=<-7.0297 0.6197 0.1446>, rot=<-0.0489 -0.0252 0.1892 -0.9804>}
 ring_02_r {parent=ring_01_r, loc=<-3.6172 -0.4648 0.1027>, rot=<0 0 0.2806 -0.9598>}
 ring_03_r {parent=ring_02_r, loc=<-2.4152 -0.1141 -0.0199>, rot=<0 0 0.1189 -0.9929>}
 thumb_01_r {parent=hand_r, loc=<-1.3119 0.3192 2.459>, rot=<0.6546 -0.4042 0.0664 -0.6354>}
 thumb_02_r {parent=thumb_01_r, loc=<-5.9335 0.0447 -0.0401>, rot=<-0.0037 0.0213 0.298 -0.9543>}
 thumb_03_r {parent=thumb_02_r, loc=<-2.2223 -0.3682 0.0907>, rot=<0.0118 -0.013 0.0066 -0.9998>}
 lowerarm_twist_01_r {parent=lowerarm_r, loc=<-17.855 0 0>, rot=<0 0 0 1>}
 lowerarm_twist_02_r {parent=lowerarm_r, loc=<-9.6199 0 0>, rot=<0 0 0 1>}
 upperarm_twist_01_r {parent=upperarm_r, loc=<-0.0041 0 0>, rot=<0 0 0 1>}
 dyn_fringeArm_G_01_Base_r {parent=upperarm_twist_01_r, loc=<-4.9179 8.1148 -1.3044>, rot=<0.2694 0.2839 0.9183 0.059>}
 dyn_fringeArm_G_01_r {parent=dyn_fringeArm_G_01_Base_r, loc=<0 0 0>, rot=<0 0 0 1>}
 upperarm_twist_02_r {parent=upperarm_r, loc=<-13.7942 0 0>, rot=<0 0 0 1>}
 dyn_fringeArm_B_01_Base_r {parent=upperarm_twist_02_r, loc=<5.0668 -6.1324 -1.8375>, rot=<0.0216 -0.9146 -0.3836 0.1259>}
 dyn_fringeArm_B_01_r {parent=dyn_fringeArm_B_01_Base_r, loc=<0 0 0>, rot=<0 0 0 -1>}
 dyn_fringeArm_B_02_r {parent=dyn_fringeArm_B_01_r, loc=<6.128 0 0>, rot=<0.0796 0.1197 0.0363 -0.9889>}
 dyn_fringeArm_A_01_Base_r {parent=upperarm_twist_02_r, loc=<3.7598 -1.684 -6.1208>, rot=<-0.106 0.6961 0.7059 0.0765>}
 dyn_fringeArm_A_01_r {parent=dyn_fringeArm_A_01_Base_r, loc=<0 0 0>, rot=<0 0 0 1>}
 dyn_fringeArm_A_02_r {parent=dyn_fringeArm_A_01_r, loc=<4.9174 0 0>, rot=<0.0084 0.0349 0.2294 -0.9727>}
 dyn_fringeArm_A_03_r {parent=dyn_fringeArm_A_01_r, loc=<1.9638 0.9786 -2.9023>, rot=<-0.0111 -0.0355 0.1642 -0.9857>}
 dyn_fringeArm_H_01_Base_r {parent=upperarm_twist_02_r, loc=<2.1883 2.8777 -4.7281>, rot=<0.0504 0.6938 0.7156 -0.0637>}
 dyn_fringeArm_H_01_r {parent=dyn_fringeArm_H_01_Base_r, loc=<0 0 0>, rot=<0 0 0 1>}
 dyn_fringeArm_H_02_r {parent=dyn_fringeArm_H_01_r, loc=<3.1873 0 0>, rot=<0.0173 0.0143 0.0102 -0.9997>}
 dyn_fringeArmAimAt_r {parent=upperarm_r, loc=<-3.7706 0.3377 -4.7438>, rot=<0.6381 -0.589 -0.4953 -0.0239>}
 dyn_fringeArm_F_01_Base_r {parent=upperarm_r, loc=<-20.3403 4.8355 0.7453>, rot=<0.1411 -0.0043 0.9718 0.1891>}
 dyn_fringeArm_F_01_r {parent=dyn_fringeArm_F_01_Base_r, loc=<0 0 0>, rot=<0 0 0 1>}
 dyn_fringeArm_F_02_r {parent=dyn_fringeArm_F_01_r, loc=<3.1738 0 0>, rot=<-0.0558 0.1319 -0.1113 -0.9834>}
 dyn_fringeArm_F_03_r {parent=dyn_fringeArm_F_01_r, loc=<2.978 -0.3875 2.3225>, rot=<-0.0457 0.1449 -0.1117 -0.9821>}
 dyn_fringeArm_F_04_r {parent=dyn_fringeArm_F_01_r, loc=<2.9373 0.7232 -1.8452>, rot=<-0.0608 0.0637 -0.0822 -0.9927>}
 dyn_fringeArm_E_01_Base_r {parent=upperarm_r, loc=<-18.6494 2.5227 3.8244>, rot=<0.1394 -0.6001 0.7689 0.1712>}
 dyn_fringeArm_E_01_r {parent=dyn_fringeArm_E_01_Base_r, loc=<0 0 0>, rot=<0 0 0 1>}
 dyn_fringeArm_E_02_r {parent=dyn_fringeArm_E_01_r, loc=<4.3619 -0.3066 -0.7982>, rot=<0.0637 -0.0109 0.0354 -0.9973>}
 dyn_fringeArm_E_03_r {parent=dyn_fringeArm_E_01_r, loc=<4.2479 0.266 0.6926>, rot=<0.1046 0.0164 0.1007 -0.9893>}
 dyn_fringeArm_D_01_Base_r {parent=upperarm_r, loc=<-18.7846 -1.8162 3.9863>, rot=<0.3074 -0.716 0.5833 0.2295>}
 dyn_fringeArm_D_01_r {parent=dyn_fringeArm_D_01_Base_r, loc=<0 0 0>, rot=<0 0 0 1>}
 dyn_fringeArm_D_02_r {parent=dyn_fringeArm_D_01_r, loc=<4.1248 0 0>, rot=<0.1036 0.0557 0.1477 -0.982>}
 dyn_fringeArm_C_01_Base_r {parent=upperarm_r, loc=<-18.9876 -5.5204 1.5444>, rot=<-0.0196 -0.9618 -0.0529 0.2679>}
 dyn_fringeArm_C_01_r {parent=dyn_fringeArm_C_01_Base_r, loc=<0 0 0>, rot=<0 0 0 1>}
 dyn_fringeArm_C_02_r {parent=dyn_fringeArm_C_01_r, loc=<3.4594 0 0>, rot=<0.0326 -0.1015 0.2211 -0.9694>}
 dyn_fringeArm_C_03_r {parent=dyn_fringeArm_C_02_r, loc=<2.3514 0 0>, rot=<0.0004 -0.0201 -0.0953 -0.9952>}
 dyn_fringeArm_C_04_r {parent=dyn_fringeArm_C_02_r, loc=<1.6651 -0.3646 1.6691>, rot=<0.0425 0.1793 -0.1376 -0.9732>}
 dyn_ornamentArm_A_01_Base_r {parent=clavicle_r, loc=<-22.279 3.7653 -2.4423>, rot=<0.2756 0.6427 0.713 -0.051>}
 dyn_ornamentArm_A_01_r {parent=dyn_ornamentArm_A_01_Base_r, loc=<0 0 0>, rot=<0 0 0 -1>}
 dyn_ornamentArm_A_02_r {parent=dyn_ornamentArm_A_01_r, loc=<3.3787 0 0>, rot=<-0.005 -0.0092 0.0484 -0.9988>}
 dyn_ornamentArm_B_01_Base_r {parent=clavicle_r, loc=<-14.6487 7.4728 -5.5682>, rot=<0.477 0.4937 0.7028 0.1865>}
 dyn_ornamentArm_B_01_r {parent=dyn_ornamentArm_B_01_Base_r, loc=<0 0 0>, rot=<0 0 0 -1>}
 dyn_ornamentArm_B_02_r {parent=dyn_ornamentArm_B_01_r, loc=<3.4924 0 0>, rot=<0.0431 0.0462 -0.2051 -0.9767>}
 neck_01 {parent=spine_05, loc=<11.3138 0.1529 0>, rot=<0 0 -0.2003 -0.9797>}
 neck_02 {parent=neck_01, loc=<4.5627 0 0>, rot=<0 0 0.0504 -0.9987>}
 head {parent=neck_02, loc=<5.3914 0 0>, rot=<0 0 0.115 -0.9934>}
 dyn_fringeBack_01_Base_r {parent=spine_05, loc=<-0.2726 -14.4288 9.1061>, rot=<-0.4393 0.3518 -0.5667 -0.6018>}
 dyn_fringeBack_01_r {parent=dyn_fringeBack_01_Base_r, loc=<0 0 0>, rot=<0 0 0 1>}
 dyn_fringeBack_02_r {parent=dyn_fringeBack_01_r, loc=<7.0255 0 0>, rot=<0.0624 -0.2481 0.1025 -0.9613>}
 dyn_fringeTassleBackA_01_Base_r {parent=dyn_fringeBack_02_r, loc=<-0.3684 6.6782 2.5254>, rot=<-0.4955 -0.6501 0.383 0.4303>}
 dyn_fringeTassleBackA_01_r {parent=dyn_fringeTassleBackA_01_Base_r, loc=<0 0 0>, rot=<0 0 0 -1>}
 dyn_fringeTassleBackA_02_r {parent=dyn_fringeTassleBackA_01_r, loc=<4.6624 0 0>, rot=<0.0463 0.0526 -0.0236 -0.9973>}
 dyn_fringeTassleBackB_01_Base_r {parent=dyn_fringeBack_02_r, loc=<2.6819 6.4798 3.9437>, rot=<-0.5261 -0.6515 0.3506 0.4194>}
 dyn_fringeTassleBackB_01_r {parent=dyn_fringeTassleBackB_01_Base_r, loc=<0 0 0>, rot=<0 0 0 1>}
 dyn_fringeTassleBack_01_Base_r {parent=dyn_fringeBack_01_r, loc=<3.9457 6.5634 0.6243>, rot=<0.4936 0.5458 -0.4276 -0.525>}
 dyn_fringeTassleBack_01_r {parent=dyn_fringeTassleBack_01_Base_r, loc=<0 0 0>, rot=<0 0 0 -1>}
 dyn_fringeTassleBack_02_r {parent=dyn_fringeTassleBack_01_r, loc=<5.1825 0.7636 1.8457>, rot=<0.0544 0.134 0.0182 -0.9893>}
 dyn_fringeTassleBack_03_r {parent=dyn_fringeTassleBack_01_r, loc=<5.296 0.6163 -1.6018>, rot=<-0.0227 -0.011 -0.0726 -0.997>}
 dyn_fringeBack_01_Base_l {parent=spine_05, loc=<0.0871 -14.692 -6.3703>, rot=<0.4499 -0.4156 -0.5812 -0.5358>}
 dyn_fringeBack_01_l {parent=dyn_fringeBack_01_Base_l, loc=<0 0 0>, rot=<0 0 0 1>}
 dyn_fringeBack_02_l {parent=dyn_fringeBack_01_l, loc=<11.505 0 0>, rot=<-0.0786 0.3784 0.1884 -0.9029>}
 dyn_fringeTassleBack_D_01_Base_l {parent=dyn_fringeBack_02_l, loc=<2.2831 7.524 -4.456>, rot=<-0.5817 -0.7135 -0.2899 -0.2617>}
 dyn_fringeTassleBack_D_01_l {parent=dyn_fringeTassleBack_D_01_Base_l, loc=<0 0 0>, rot=<0 0 0 1>}
 dyn_fringeTassleBack_C_01_Base_l {parent=dyn_fringeBack_02_l, loc=<-0.1162 7.7809 -2.6346>, rot=<-0.5313 -0.7175 -0.37 -0.2568>}
 dyn_fringeTassleBack_C_01_l {parent=dyn_fringeTassleBack_C_01_Base_l, loc=<0 0 0>, rot=<0 0 0 -1>}
 dyn_fringeTassleBack_C_02_l {parent=dyn_fringeTassleBack_C_01_l, loc=<4.9232 0 0>, rot=<-0.021 0.0045 0.0352 -0.9991>}
 dyn_fringeTassleBack_B_01_Base_l {parent=dyn_fringeBack_01_l, loc=<10.359 7.035 0.1531>, rot=<0.4921 0.4387 0.5823 0.4757>}
 dyn_fringeTassleBack_B_01_l {parent=dyn_fringeTassleBack_B_01_Base_l, loc=<0 0 0>, rot=<0 0 0 -1>}
 dyn_fringeTassleBack_B_02_l {parent=dyn_fringeTassleBack_B_01_l, loc=<5.5756 0 0>, rot=<-0.0405 0.0233 0.1257 -0.991>}
 dyn_fringeTassleBack_A_01_Base_l {parent=dyn_fringeBack_01_l, loc=<2.1717 4.2506 -0.5366>, rot=<0.5595 0.5271 0.4652 0.4391>}
 dyn_fringeTassleBack_A_01_l {parent=dyn_fringeTassleBack_A_01_Base_l, loc=<0 0 0>, rot=<0 0 0 -1>}
 dyn_fringeTassleBack_A_02_l {parent=dyn_fringeTassleBack_A_01_l, loc=<4.7861 0 0>, rot=<0.0322 -0.0942 0.0669 -0.9928>}
 dyn_fringeTassleBack_01_c_Base {parent=spine_05, loc=<-2.8773 -14.8474 3.0885>, rot=<0.0119 -0.9993 -0.0025 -0.0358>}
 dyn_fringeTassleBack_01_c {parent=dyn_fringeTassleBack_01_c_Base, loc=<0 0 0>, rot=<0 0 0 1>}
 dyn_fringeTassleBack_02_c {parent=dyn_fringeTassleBack_01_c, loc=<5.3089 0 0>, rot=<0 0 0 1>}
 dyn_fringeFront_01_Base_r {parent=spine_05, loc=<2.9357 12.6804 7.7207>, rot=<-0.1947 0.7095 0.1263 -0.6654>}
 dyn_fringeFront_01_r {parent=dyn_fringeFront_01_Base_r, loc=<0 0 0>, rot=<0 0 0 -1>}
 dyn_fringeFront_02_r {parent=dyn_fringeFront_01_r, loc=<11.0252 0 0>, rot=<0.0908 -0.1047 0.2537 -0.9573>}
 dyn_fringeTassleFront_F_01_Base_r {parent=dyn_fringeFront_02_r, loc=<0.4612 2.2146 6.1295>, rot=<-0.2785 0.5802 -0.3844 -0.6618>}
 dyn_fringeTassleFront_F_01_r {parent=dyn_fringeTassleFront_F_01_Base_r, loc=<0 0 0>, rot=<0 0 0 -1>}
 dyn_fringeTassleFront_C_01_Base_r {parent=dyn_fringeFront_01_r, loc=<4.4663 3.2041 7.9032>, rot=<-0.0792 0.6024 -0.2535 -0.7527>}
 dyn_fringeTassleFront_C_01_r {parent=dyn_fringeTassleFront_C_01_Base_r, loc=<0 0 0>, rot=<0 0 0 -1>}
 dyn_fringeTassleFront_C_02_r {parent=dyn_fringeTassleFront_C_01_r, loc=<4.9125 0 0>, rot=<0.0099 0.083 0.0753 -0.9937>}
 dyn_fringeTassleFront_D_01_Base_r {parent=dyn_fringeFront_01_r, loc=<8.5509 2.8769 7.4124>, rot=<-0.1122 0.5847 -0.2112 -0.7752>}
 dyn_fringeTassleFront_D_01_r {parent=dyn_fringeTassleFront_D_01_Base_r, loc=<0 0 0>, rot=<0 0 0 -1>}
 dyn_fringeTassleFront_E_01_Base_r {parent=dyn_fringeFront_01_r, loc=<11.1984 2.5649 7.1337>, rot=<-0.1032 0.5449 -0.2206 -0.8024>}
 dyn_fringeTassleFront_E_01_r {parent=dyn_fringeTassleFront_E_01_Base_r, loc=<0 0 0>, rot=<0 0 0 1>}
 dyn_fringeTassleFront_E_02_r {parent=dyn_fringeTassleFront_E_01_r, loc=<5.8145 0 0>, rot=<0.0037 0.0596 0.0196 -0.998>}
 dyn_fringeFront_01_Base_l {parent=spine_05, loc=<2.2319 13.3249 -6.1982>, rot=<0.0931 -0.6402 0.0489 -0.761>}
 dyn_fringeFront_01_l {parent=dyn_fringeFront_01_Base_l, loc=<0 0 0>, rot=<0 0 0 -1>}
 dyn_fringeFront_02_l {parent=dyn_fringeFront_01_l, loc=<8.4106 0 0>, rot=<0.0525 -0.0831 0.3458 -0.9331>}
 dyn_fringeTassleFrontBase_l {parent=dyn_fringeFront_02_l, loc=<1.6096 2.181 -1.6045>, rot=<0.6855 0.3355 -0.645 0.0383>}
 dyn_fringeTassleFront_D_01_Base_l {parent=dyn_fringeTassleFrontBase_l, loc=<8.282 0 0>, rot=<0.9799 -0.0261 0.067 0.186>}
 dyn_fringeTassleFront_D_01_l {parent=dyn_fringeTassleFront_D_01_Base_l, loc=<0 0 0>, rot=<0 0 0 -1>}
 dyn_fringeTassleFront_E_01_Base_l {parent=dyn_fringeTassleFrontBase_l, loc=<6.4835 0.1562 -3.099>, rot=<0.9737 0.0493 -0.0518 0.2165>}
 dyn_fringeTassleFront_E_01_l {parent=dyn_fringeTassleFront_E_01_Base_l, loc=<0 0 0>, rot=<0 0 0 1>}
 dyn_fringeTassleFront_E_02_l {parent=dyn_fringeTassleFront_E_01_l, loc=<6.5821 0 0>, rot=<-0.0171 -0.1114 0.0285 -0.9932>}
 dyn_fringeTassleFront_F_01_Base_l {parent=dyn_fringeTassleFrontBase_l, loc=<5.0831 0.6566 -6.1987>, rot=<0.9722 0.0219 -0.0676 0.2231>}
 dyn_fringeTassleFront_F_01_l {parent=dyn_fringeTassleFront_F_01_Base_l, loc=<0 0 0>, rot=<0 0 0 -1>}
 dyn_fringeTassleFront_C_01_Base_l {parent=dyn_fringeFront_01_l, loc=<0.1602 3.9256 -8.1762>, rot=<0.0136 -0.6535 -0.1413 -0.7435>}
 dyn_fringeTassleFront_C_01_l {parent=dyn_fringeTassleFront_C_01_Base_l, loc=<0 0 0>, rot=<0 0 0 1>}
 dyn_fringeTassleFront_C_02_l {parent=dyn_fringeTassleFront_C_01_l, loc=<5.8798 0 0>, rot=<-0.0088 -0.0968 0.076 -0.9924>}
 dyn_upperBackOrnament_01_c_Base {parent=spine_05, loc=<11.385 -12.1119 4.2354>, rot=<0.4986 -0.8664 -0.026 -0.0123>}
 dyn_upperBackOrnament_01_c {parent=dyn_upperBackOrnament_01_c_Base, loc=<0 0 0>, rot=<0 0 0 1>}
 dyn_upperBackOrnament_02_c {parent=dyn_upperBackOrnament_01_c, loc=<3.312 0 0>, rot=<-0.0406 -0.0489 -0.0858 -0.9943>}
 dyn_fringeTassleBack_Z_01_Base_l {parent=spine_05, loc=<7.0179 -11.5545 -12.4652>, rot=<0.2603 -0.885 -0.1447 -0.3579>}
 dyn_fringeTassleBack_Z_01_l {parent=dyn_fringeTassleBack_Z_01_Base_l, loc=<0 0 0>, rot=<0 0 0 1>}
 dyn_fringeTassleFront_B_01_Base_r {parent=spine_04, loc=<3.1551 15.7623 5.9782>, rot=<-0.0725 -0.9925 0.0154 0.0975>}
 dyn_fringeTassleFront_B_01_r {parent=dyn_fringeTassleFront_B_01_Base_r, loc=<0 0 0>, rot=<0 0 0 -1>}
 dyn_fringeTassleFront_B_02_r {parent=dyn_fringeTassleFront_B_01_r, loc=<7.4825 0 0>, rot=<-0.0135 -0.0565 -0.0057 -0.9983>}
 dyn_fringeTassleFront_A_01_Base_r {parent=spine_04, loc=<4.333 15.7687 2.5358>, rot=<-0.1196 -0.9928 0 0.0001>}
 dyn_fringeTassleFront_A_01_r {parent=dyn_fringeTassleFront_A_01_Base_r, loc=<0 0 0>, rot=<0 0 0 -1>}
 dyn_fringeTassleFront_A_02_r {parent=dyn_fringeTassleFront_A_01_r, loc=<6.867 0 0>, rot=<-0.006 -0.0234 0.0728 -0.9971>}
 dyn_fringeTassleFront_A_01_Base_l {parent=spine_04, loc=<0.9641 16.8923 -0.9364>, rot=<-0.0646 -0.9979 -0.0012 -0.0082>}
 dyn_fringeTassleFront_A_01_l {parent=dyn_fringeTassleFront_A_01_Base_l, loc=<0 0 0>, rot=<0 0 0 -1>}
 dyn_fringeTassleFront_A_02_l {parent=dyn_fringeTassleFront_A_01_l, loc=<8.0078 0 0>, rot=<0.0056 0.0356 0.055 -0.9978>}
 dyn_fringeTassleFront_B_01_Base_l {parent=spine_04, loc=<1.4033 16.5188 -3.9971>, rot=<-0.0407 -0.9984 -0.0049 -0.0395>}
 dyn_fringeTassleFront_B_01_l {parent=dyn_fringeTassleFront_B_01_Base_l, loc=<0 0 0>, rot=<0 0 0 1>}
 dyn_torsoFringe_B_01_Base_r {parent=spine_03, loc=<-0.1123 1.6818 15.5958>, rot=<0.0787 -0.7157 0.6927 0.0422>}
 dyn_torsoFringe_B_01_r {parent=dyn_torsoFringe_B_01_Base_r, loc=<0 0 0>, rot=<0 0 0 -1>}
 dyn_torsoFringe_B_01_Base_l {parent=spine_03, loc=<3.238 7.6017 -13.8628>, rot=<-0.0027 0.6265 0.7764 0.068>}
 dyn_torsoFringe_B_01_l {parent=dyn_torsoFringe_B_01_Base_l, loc=<0 0 0>, rot=<0 0 0 1>}
 dyn_torsoFringe_C_01_Base_r {parent=spine_03, loc=<3.238 7.6017 13.77>, rot=<-0.1214 -0.6175 0.7652 0.1359>}
 dyn_torsoFringe_C_01_r {parent=dyn_torsoFringe_C_01_Base_r, loc=<0 0 0>, rot=<0 0 0 -1>}
 dyn_torsoFringe_D_01_Base_r {parent=spine_03, loc=<4.1373 -12.8328 9.7162>, rot=<-0.0511 -0.9798 0.1328 0.1408>}
 dyn_torsoFringe_D_01_r {parent=dyn_torsoFringe_D_01_Base_r, loc=<0 0 0>, rot=<0 0 0 1>}
 dyn_torsoFringe_E_01_Base_r {parent=spine_03, loc=<3.5088 -11.3757 13.0142>, rot=<-0.0909 -0.9573 0.2372 0.1378>}
 dyn_torsoFringe_E_01_r {parent=dyn_torsoFringe_E_01_Base_r, loc=<0 0 0>, rot=<0 0 0 1>}
 dyn_torsoFringe_F_01_Base_r {parent=spine_03, loc=<5.7613 -9.4771 15.411>, rot=<-0.0701 -0.9613 0.1991 0.1768>}
 dyn_torsoFringe_F_01_r {parent=dyn_torsoFringe_F_01_Base_r, loc=<0 0 0>, rot=<0 0 0 1>}
 thigh_l {parent=pelvis, loc=<-2.5693 -0.0109 -7.7501>, rot=<0.0012 0.0572 0.0155 -0.9982>}
 calf_l {parent=thigh_l, loc=<-39.0484 0 0>, rot=<0.0006 -0.0161 -0.0391 -0.9991>}
 calf_twist_01_l {parent=calf_l, loc=<-35.8131 -0.0967 0.0397>, rot=<0.0237 0.0095 -0.0035 -0.9997>}
 dyn_legLow_Q_01_l {parent=calf_twist_01_l, loc=<4.8437 -10.9605 -5.2647>, rot=<0.9492 0.2134 0.022 -0.2304>}
 dyn_legLow_R_01_l {parent=calf_twist_01_l, loc=<4.3327 -3.0387 -9.4078>, rot=<0.7998 0.1841 0.0889 -0.5644>}
 dyn_legLow_S_01_l {parent=calf_twist_01_l, loc=<4.8553 2.5495 -8.2555>, rot=<0.5335 0.0788 0.2334 -0.8091>}
 dyn_legLow_T_01_l {parent=calf_twist_01_l, loc=<5.6453 5.9104 -3.7481>, rot=<0.0648 0.0018 0.2772 -0.9586>}
 dyn_legLow_U_01_l {parent=calf_twist_01_l, loc=<5.2551 5.3202 4.3358>, rot=<-0.2937 -0.1997 0.1595 -0.9211>}
 dyn_legLow_V_01_l {parent=calf_twist_01_l, loc=<7.7889 -11.7218 3.5066>, rot=<0.9348 0.2747 -0.1251 0.187>}
 dyn_legLow_X_01_l {parent=calf_twist_01_l, loc=<8.8568 -5.1949 8.1713>, rot=<-0.6395 -0.0377 0.2009 -0.7411>}
 calf_twist_02_l {parent=calf_l, loc=<-18.8064 -0.0967 0.0397>, rot=<0.0237 0.0095 -0.0035 -0.9997>}
 dyn_legLow_G_01_l {parent=calf_twist_02_l, loc=<6.0713 -10.9315 6.4005>, rot=<0.827 0.3161 -0.0874 0.4566>}
 dyn_legLow_H_01_l {parent=calf_twist_02_l, loc=<8.3145 -12.8298 1.0984>, rot=<0.9382 0.3454 -0.0004 0.0231>}
 dyn_legLow_J_01_l {parent=calf_twist_02_l, loc=<4.7677 -3.8489 -7.2137>, rot=<0.673 0.2021 0.2629 -0.6612>}
 dyn_legLow_K_01_l {parent=calf_twist_02_l, loc=<2.8022 3.0949 -7.461>, rot=<0.3456 0.0721 0.3744 -0.8575>}
 dyn_legLow_L_01_l {parent=calf_twist_02_l, loc=<3.3391 6.8086 -1.2549>, rot=<0.0844 0.0408 0.2159 -0.9719>}
 dyn_legLow_M_01_l {parent=calf_twist_02_l, loc=<4.6803 1.3677 6.3616>, rot=<-0.5542 -0.1343 0.23 -0.7886>}
 dyn_legLow_N_01_l {parent=calf_twist_02_l, loc=<-4.8802 -12.2215 3.0228>, rot=<0.9385 0.3286 -0.0098 -0.1052>}
 dyn_legLow_P_01_l {parent=calf_twist_02_l, loc=<-5.3746 -10.3409 -4.7862>, rot=<0.9082 0.2784 0.074 -0.3036>}
 dyn_legLow_Y_01_l {parent=calf_twist_02_l, loc=<-4.2772 -3.4728 -6.8126>, rot=<0.6992 0.1812 0.1857 -0.6662>}
 dyn_legLow_W_01_l {parent=calf_twist_02_l, loc=<-5.4149 -2.7424 8.3245>, rot=<-0.6824 -0.1414 0.2028 -0.6879>}
 dyn_legLow_Z_01_l {parent=calf_twist_02_l, loc=<-5.967 5.6201 -1.911>, rot=<0.246 0.0264 0.2435 -0.9378>}
 foot_l {parent=calf_l, loc=<-42.561 0.1077 0.4292>, rot=<0.0418 -0.0455 0.0343 -0.9975>}
 ball_l {parent=foot_l, loc=<-6.1599 13.8264 0.6806>, rot=<0.0001 -0.0001 -0.7063 -0.7079>}
 dyn_legLow_A_01_Base_l {parent=calf_l, loc=<-1.5366 -12.4297 1.864>, rot=<0.9076 0.4068 0.0242 0.1012>}
 dyn_legLow_A_01_l {parent=dyn_legLow_A_01_Base_l, loc=<0 0 0>, rot=<0 0 0 -1>}
 dyn_legLow_BB_01_Base_l {parent=calf_l, loc=<-4.4431 -7.9103 -3.4073>, rot=<0.4019 -0.8151 -0.4031 -0.108>}
 dyn_legLow_BB_01_l {parent=dyn_legLow_BB_01_Base_l, loc=<0 0 0>, rot=<0 0 0 1>}
 dyn_legLow_BB_02_l {parent=dyn_legLow_BB_01_l, loc=<3.2266 0 0>, rot=<0 -0.1045 0 -0.9945>}
 dyn_legLow_B_01_Base_l {parent=calf_l, loc=<-0.2639 -10.9994 -4.6591>, rot=<0.8599 0.2757 0.1375 -0.4069>}
 dyn_legLow_B_01_l {parent=dyn_legLow_B_01_Base_l, loc=<0 0 0>, rot=<0 0 0 1>}
 dyn_legLow_C_01_Base_l {parent=calf_l, loc=<-0.9686 -3.6511 -7.2361>, rot=<0.627 0.2194 0.1279 -0.7365>}
 dyn_legLow_C_01_l {parent=dyn_legLow_C_01_Base_l, loc=<0 0 0>, rot=<0 0 0 1>}
 dyn_legLow_D_01_Base_l {parent=calf_l, loc=<-2.3461 5.97 -2.8752>, rot=<0.1998 0.011 0.1575 -0.967>}
 dyn_legLow_D_01_l {parent=dyn_legLow_D_01_Base_l, loc=<0 0 0>, rot=<0 0 0 -1>}
 dyn_legLow_E_01_Base_l {parent=calf_l, loc=<-1.6558 -5.8054 7.1762>, rot=<-0.7277 -0.178 0.1617 -0.6423>}
 dyn_legLow_E_01_l {parent=dyn_legLow_E_01_Base_l, loc=<0 0 0>, rot=<0 0 0 -1>}
 dyn_legLow_F_01_l {parent=calf_l, loc=<-1.812 2.4911 5.7186>, rot=<-0.4811 -0.1295 0.1088 -0.8602>}
 thigh_twist_01_l {parent=thigh_l, loc=<-21.6303 0 0>, rot=<0.0013 0 0 -1>}
 dyn_legHi_A_01_l {parent=thigh_twist_01_l, loc=<14.7678 -2.5556 -9.594>, rot=<0.7296 0.2501 0.1957 -0.6057>}
 dyn_legHi_B_01_l {parent=thigh_twist_01_l, loc=<13.0003 4.9535 -8.8634>, rot=<0.4922 0.2915 0.2386 -0.7847>}
 dyn_legHi_C_01_l {parent=thigh_twist_01_l, loc=<8.3924 10.0622 -3.6611>, rot=<0.3701 -0.0438 0.2579 -0.8914>}
 dyn_legHi_D_01_l {parent=thigh_twist_01_l, loc=<7.5159 -8.1743 -3.9039>, rot=<0.9339 0.1378 0.1932 -0.2675>}
 dyn_legHi_E_01_l {parent=thigh_twist_01_l, loc=<6.5961 -2.3926 -7.988>, rot=<0.7664 0.2459 0.1637 -0.5705>}
 dyn_legHi_F_01_l {parent=thigh_twist_01_l, loc=<5.7127 4.7512 -9.6555>, rot=<0.4865 0.145 0.3017 -0.807>}
 dyn_legHi_G_01_l {parent=thigh_twist_01_l, loc=<-2.1971 2.0618 -9.4917>, rot=<0.6574 0.2061 0.0616 -0.7222>}
 dyn_legHi_J_01_Base_l {parent=thigh_twist_01_l, loc=<-5.3162 -6.2736 -6.0843>, rot=<0.9178 0.3147 0.0284 -0.2406>}
 dyn_legHi_J_01_l {parent=dyn_legHi_J_01_Base_l, loc=<0 0 0>, rot=<0 0 0 1>}
 dyn_legHi_K_01_Base_l {parent=thigh_twist_01_l, loc=<-8.0909 -9.1472 -1.5002>, rot=<0.9326 0.3304 -0.0069 -0.1453>}
 dyn_legHi_K_01_l {parent=dyn_legHi_K_01_Base_l, loc=<0 0 0>, rot=<0 0 0 1>}
 dyn_legHi_L_01_l {parent=thigh_twist_01_l, loc=<-6.7963 0.6698 -9.1685>, rot=<0.7407 0.1566 0.0191 -0.653>}
 dyn_legHi_N_01_l {parent=thigh_twist_01_l, loc=<-8.8256 9.9821 -2.0776>, rot=<-0.0925 -0.0751 0.1927 -0.974>}
 dyn_legHi_P_01_l {parent=thigh_twist_01_l, loc=<-14.1555 9.6795 -3.5592>, rot=<-0.2729 -0.9558 -0.0307 -0.1048>}
 thigh_r {parent=pelvis, loc=<-2.5691 -0.0109 7.7501>, rot=<0.0572 -0.0012 0.9982 0.0155>}
 calf_r {parent=thigh_r, loc=<39.0487 0 0>, rot=<0.0006 -0.0161 -0.0391 -0.9991>}
 calf_twist_01_r {parent=calf_r, loc=<35.8132 0.0967 -0.0397>, rot=<0.0237 0.0095 -0.0035 -0.9997>}
 dyn_legLow_Q_01_r {parent=calf_twist_01_r, loc=<-5.084 11.4247 5.5121>, rot=<0.9492 0.2134 0.022 -0.2304>}
 dyn_legLow_R_01_r {parent=calf_twist_01_r, loc=<-4.7595 3.3578 10.3481>, rot=<0.7998 0.1841 0.0889 -0.5644>}
 dyn_legLow_S_01_r {parent=calf_twist_01_r, loc=<-5.5098 -3.0055 9.4269>, rot=<0.5335 0.0788 0.2334 -0.8091>}
 dyn_legLow_T_01_r {parent=calf_twist_01_r, loc=<-5.6454 -5.9104 3.7481>, rot=<0.0648 0.0018 0.2772 -0.9586>}
 dyn_legLow_U_01_r {parent=calf_twist_01_r, loc=<-5.2553 -5.3202 -4.3358>, rot=<-0.591 -0.1331 0.2181 -0.7651>}
 dyn_legLow_V_01_r {parent=calf_twist_01_r, loc=<-7.6017 11.4612 -3.4127>, rot=<0.9348 0.2747 -0.1251 0.187>}
 dyn_legLow_X_01_r {parent=calf_twist_01_r, loc=<-8.857 5.195 -8.1713>, rot=<-0.4259 -0.0884 0.1843 -0.8814>}
 calf_twist_02_r {parent=calf_r, loc=<18.8064 0.0967 -0.0397>, rot=<0.0237 0.0095 -0.0035 -0.9997>}
 dyn_legLow_G_01_r {parent=calf_twist_02_r, loc=<-5.4826 10.5572 -5.7168>, rot=<0.827 0.3161 -0.0874 0.4566>}
 dyn_legLow_H_01_r {parent=calf_twist_02_r, loc=<-6.2186 13.3274 -1.0883>, rot=<0.9382 0.3454 -0.0004 0.0231>}
 dyn_legLow_J_01_r {parent=calf_twist_02_r, loc=<-2.8331 3.6537 9.3478>, rot=<0.673 0.2021 0.2629 -0.6612>}
 dyn_legLow_K_01_r {parent=calf_twist_02_r, loc=<-2.8023 -3.0949 7.461>, rot=<0.3456 0.0721 0.3744 -0.8575>}
 dyn_legLow_L_01_r {parent=calf_twist_02_r, loc=<-3.3391 -6.8086 1.2549>, rot=<0.0844 0.0408 0.2159 -0.9719>}
 dyn_legLow_M_01_r {parent=calf_twist_02_r, loc=<-0.4575 -4.7749 -3.3264>, rot=<-0.2346 -0.131 0.2812 -0.9213>}
 dyn_legLow_M_02_r {parent=dyn_legLow_M_01_r, loc=<3.2481 -0.8118 0.2271>, rot=<0 0 0 1>}
 dyn_legLow_N_01_r {parent=calf_twist_02_r, loc=<4.8801 12.2215 -3.0229>, rot=<0.9385 0.3286 -0.0098 -0.1052>}
 dyn_legLow_P_01_r {parent=calf_twist_02_r, loc=<4.8517 10.9682 5.2707>, rot=<0.9082 0.2784 0.074 -0.3036>}
 dyn_legLow_Y_01_r {parent=calf_twist_02_r, loc=<3.7351 3.5233 7.7481>, rot=<0.6992 0.1812 0.1857 -0.6662>}
 dyn_legLow_Z_01_r {parent=calf_twist_02_r, loc=<5.5796 -6.247 2.2808>, rot=<0.246 0.0264 0.2435 -0.9378>}
 dyn_legLow_W_01_r {parent=calf_twist_02_r, loc=<5.4148 2.7424 -8.3245>, rot=<-0.6824 -0.1414 0.2028 -0.6879>}
 foot_r {parent=calf_r, loc=<42.5611 -0.1077 -0.4292>, rot=<0.0418 -0.0455 0.0343 -0.9975>}
 ball_r {parent=foot_r, loc=<6.1598 -13.8264 -0.6806>, rot=<0.0001 -0.0001 -0.7063 -0.7079>}
 dyn_legLow_A_01_Base_r {parent=calf_r, loc=<1.5366 12.4297 -1.864>, rot=<0.9076 0.4068 0.0242 0.1012>}
 dyn_legLow_A_01_r {parent=dyn_legLow_A_01_Base_r, loc=<0 0 0>, rot=<0 0 0 1>}
 dyn_legLow_B_01_Base_r {parent=calf_r, loc=<0.2638 10.9994 4.659>, rot=<0.8599 0.2757 0.1375 -0.4069>}
 dyn_legLow_B_01_r {parent=dyn_legLow_B_01_Base_r, loc=<0 0 0>, rot=<0 0 0 -1>}
 dyn_legLow_C_01_Base_r {parent=calf_r, loc=<0.78 8.4718 7.3755>, rot=<0.8681 0.2794 0.1294 -0.3894>}
 dyn_legLow_C_01_r {parent=dyn_legLow_C_01_Base_r, loc=<0 0 0>, rot=<0 0 0 -1>}
 dyn_legLow_D_01_Base_r {parent=calf_r, loc=<1.9006 3.7675 8.3879>, rot=<0.7515 0.2291 0.2172 -0.5793>}
 dyn_legLow_D_01_r {parent=dyn_legLow_D_01_Base_r, loc=<0 0 0>, rot=<-0.0698 0 0 -0.9976>}
 dyn_legLow_E_01_Base_r {parent=calf_r, loc=<1.6558 5.8054 -7.1762>, rot=<-0.7277 -0.178 0.1617 -0.6423>}
 dyn_legLow_E_01_r {parent=dyn_legLow_E_01_Base_r, loc=<0 0 0>, rot=<0 0 0 1>}
 dyn_legLow_F_01_r {parent=calf_r, loc=<1.8119 -2.4911 -5.7186>, rot=<-0.4811 -0.1295 0.1088 -0.8602>}
 thigh_twist_01_r {parent=thigh_r, loc=<21.6303 0 0>, rot=<0.0013 0 0 -1>}
 dyn_legHi_A_01_r {parent=thigh_twist_01_r, loc=<-14.7677 2.5556 9.5941>, rot=<0.7296 0.2501 0.1957 -0.6057>}
 dyn_legHi_B_01_r {parent=thigh_twist_01_r, loc=<-13.0001 -4.9535 8.8634>, rot=<0.621 0.2457 0.2856 -0.6874>}
 dyn_legHi_C_01_r {parent=thigh_twist_01_r, loc=<-12.9738 -9.7017 5.4908>, rot=<0.0682 -0.0379 0.3094 -0.9477>}
 dyn_legHi_C_02_r {parent=dyn_legHi_C_01_r, loc=<2.9233 0 0>, rot=<0 0 0 1>}
 dyn_legHi_D_01_r {parent=thigh_twist_01_r, loc=<-7.5157 8.1743 3.9039>, rot=<0.9339 0.1378 0.1932 -0.2675>}
 dyn_legHi_E_01_r {parent=thigh_twist_01_r, loc=<-6.596 2.3926 7.988>, rot=<0.7664 0.2459 0.1637 -0.5705>}
 dyn_legHi_F_01_r {parent=thigh_twist_01_r, loc=<-5.7124 -4.7512 9.6555>, rot=<0.4865 0.145 0.3017 -0.807>}
 dyn_legHi_G_01_r {parent=thigh_twist_01_r, loc=<2.1973 -2.0618 9.4916>, rot=<0.6574 0.2061 0.0616 -0.7222>}
 dyn_legHi_H_01_Base_r {parent=thigh_twist_01_r, loc=<0.5824 3.6755 7.7111>, rot=<0.8822 0.4129 -0.0074 -0.2262>}
 dyn_legHi_H_01_r {parent=dyn_legHi_H_01_Base_r, loc=<0 0 0>, rot=<0 0 0 1>}
 dyn_legHi_H_02_r {parent=dyn_legHi_H_01_r, loc=<3.0855 0 0>, rot=<-0.0642 -0.125 -0.0415 -0.9892>}
 dyn_legHi_J_01_Base_r {parent=thigh_twist_01_r, loc=<5.3164 6.2736 6.0843>, rot=<0.9178 0.3147 0.0284 -0.2406>}
 dyn_legHi_J_01_r {parent=dyn_legHi_J_01_Base_r, loc=<0 0 0>, rot=<0 0 0 1>}
 dyn_legHi_K_01_Base_r {parent=thigh_twist_01_r, loc=<8.091 9.1472 1.5002>, rot=<0.9326 0.3304 -0.0069 -0.1453>}
 dyn_legHi_K_01_r {parent=dyn_legHi_K_01_Base_r, loc=<0 0 0>, rot=<0 0 0 1>}
 dyn_legHi_L_01_r {parent=thigh_twist_01_r, loc=<6.7965 -0.6698 9.1684>, rot=<0.7407 0.1566 0.0191 -0.653>}
 dyn_legHi_M_01_r {parent=thigh_twist_01_r, loc=<6.9134 -8.6727 -3.4382>, rot=<-0.268 -0.0312 -0.0315 -0.9624>}
 dyn_legHi_N_01_r {parent=thigh_twist_01_r, loc=<8.8257 -9.9821 2.0777>, rot=<-0.0925 -0.0751 0.1927 -0.974>}
 dyn_rearOrnament_01_Base {parent=pelvis, loc=<2.0704 -11.995 -6.0531>, rot=<0.2579 -0.9479 -0.1114 -0.1501>}
 dyn_rearOrnament_01 {parent=dyn_rearOrnament_01_Base, loc=<0 0 0>, rot=<0 0 0 -1>}
 dyn_rearOrnament_02 {parent=dyn_rearOrnament_01, loc=<3.3946 0 0>, rot=<-0.0128 -0.0536 -0.0861 -0.9948>}
 dyn_torsoFringe_A_01_Base_l {parent=pelvis, loc=<6.0843 -10.6916 -11.5699>, rot=<0.0997 -0.9278 -0.3579 0.0334>}
 dyn_torsoFringe_A_01_l {parent=dyn_torsoFringe_A_01_Base_l, loc=<0 0 0>, rot=<0 0 0 1>}
 dyn_torsoFringe_A_01_Base_r {parent=pelvis, loc=<5.5876 -10.8257 11.2529>, rot=<0.1067 -0.9211 0.3744 -0.0103>}
 dyn_torsoFringe_A_01_r {parent=dyn_torsoFringe_A_01_Base_r, loc=<0 0 0>, rot=<0 0 0 -1>}
]
