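# Physics asset for the back fringe and tassel secondary motion on the male medium Ornament Soldier body.
# Box bodies on the dyn_ bones hang from spine_05, some as two-link chains, plus two right torso fringes from spine_03.
# Narrow swing and twist limits restrain the pieces, with the center chain tightest in twist.
# Capsules and offset spheres approximate the upper spine, linked down to a small pelvis box by angularly free constraints.
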PhysicsAsset /BRCosmetics/Characters/Player/Male/Medium/Bodies/M_MED_Ornament_Soldier/Meshes/M_MED_Ornament_Soldier_PhysicsBack
bodies {
 dyn_fringeTassleBackA_01_r [{shapeType=Box, extentX=1.0901, extentY=2.4438, extentZ=5.2731, center=<2.2279 0 0>, rotation=<rot 0 90 90>}]
 dyn_fringeTassleBackB_01_r [{shapeType=Box, extentX=1.217, extentY=2.9049, extentZ=9.2005, center=<4.1783 0.0004 0>, rotation=<rot 0 90 90>}]
 dyn_fringeTassleBack_01_c [{shapeType=Box, extentX=2.6303, extentY=13.0583, extentZ=4.9971, center=<2.608 0.1757 0>, rotation=<rot 0 89.9999 89.9999>}]
 dyn_fringeTassleBack_01_r [{shapeType=Box, extentX=2.9217, extentY=7.5995, extentZ=5.8696, center=<2.1885 0 0>, rotation=<rot 0 90 90>}]
 dyn_fringeTassleBack_02_c [{shapeType=Box, extentX=2.9458, extentY=14.1147, extentZ=8.4752, center=<2.8331 -0.5283 -0.3161>, rotation=<rot 0 89.9999 89.9999>}]
 dyn_fringeTassleBack_02_r [{shapeType=Box, extentX=2.3021, extentY=4.3418, extentZ=7.7944, center=<3.1225 0.0002 0>, rotation=<rot 0 90 90>}]
 dyn_fringeTassleBack_A_01_l [{shapeType=Box, extentX=3.51, extentY=7.6239, extentZ=5.5689, center=<2.5275 0.0001 0>, rotation=<rot 0 90 90>}]
 dyn_fringeTassleBack_B_01_l [{shapeType=Box, extentX=3.3102, extentY=8.7688, extentZ=8.0804, center=<2.8721 -0.0001 0>, rotation=<rot 0 90 90>}]
 dyn_fringeTassleBack_B_02_l [{shapeType=Box, extentX=2.9442, extentY=8.9722, extentZ=7.3153, center=<3.1266 0.0002 -0.0689>, rotation=<rot 0 90 90>}]
 dyn_fringeTassleBack_C_01_l [{shapeType=Box, extentX=1.7464, extentY=2.7452, extentZ=6.922, center=<2.6643 -0.0001 0>, rotation=<rot 0 90 90>}]
 dyn_fringeTassleBack_D_01_l [{shapeType=Box, extentX=1.255, extentY=3.176, extentZ=10.4926, center=<4.4428 0 0>, rotation=<rot 0 90 90>}]
 dyn_fringeTassleBack_Z_01_l [{shapeType=Box, extentX=1.2725, extentY=7.565, extentZ=8.7393, center=<4.3814 -0.0003 0>, rotation=<rot 0 90 90>}]
 dyn_torsoFringe_D_01_r [{shapeType=Box, extentX=3.8931, extentY=7.4315, extentZ=3.0201, center=<3.0037 0 0>, rotation=<rot 0 90 0>}]
 dyn_torsoFringe_F_01_r [{shapeType=Box, extentX=1.8931, extentY=7.4315, extentZ=3.0201, center=<3.0037 0 0>, rotation=<rot 0 90 0>}]
 pelvis [{shapeType=Box, extentX=1.3845, extentY=1.5396, extentZ=1.9676}]
 spine_03 [{shapeType=Capsule, length=5.5873, radius=11.5686}]
 spine_04 [
  {shapeType=Sphere, radius=7.2386, center=<0.0002 -6.3501 -11.726>}
  {shapeType=Capsule, length=12.9971, radius=13.4173, center=<0 -0.5234 0>}
 ]
 spine_05 [
  {shapeType=Sphere, radius=5.25, shapeName=spine_05_sphere, center=<0.706 -5.6626 -21.388>}
  {shapeType=Capsule, length=22.6402, radius=12.6784}
 ]
}
constraints [
 {bone1Name=dyn_fringeTassleBackA_01_r, bone2Name=spine_05, swing1LimitDegrees=20, swing2LimitDegrees=12, twistLimitDegrees=8}
 {bone1Name=dyn_fringeTassleBackB_01_r, bone2Name=spine_05, swing1LimitDegrees=20, swing2LimitDegrees=12, twistLimitDegrees=8}
 {bone1Name=dyn_fringeTassleBack_01_r, bone2Name=spine_05, swing1LimitDegrees=20, swing2LimitDegrees=12, twistLimitDegrees=8}
 {bone1Name=dyn_fringeTassleBack_02_r, bone2Name=dyn_fringeTassleBack_01_r, swing1LimitDegrees=20, swing2LimitDegrees=12, twistLimitDegrees=8}
 {bone1Name=dyn_fringeTassleBack_D_01_l, bone2Name=spine_05, swing1LimitDegrees=20, swing2LimitDegrees=12, twistLimitDegrees=8}
 {bone1Name=dyn_fringeTassleBack_C_01_l, bone2Name=spine_05, swing1LimitDegrees=20, swing2LimitDegrees=12, twistLimitDegrees=8}
 {bone1Name=dyn_fringeTassleBack_B_01_l, bone2Name=spine_05, swing1LimitDegrees=20, swing2LimitDegrees=12, twistLimitDegrees=8}
 {
  bone1Name=dyn_fringeTassleBack_B_02_l
  bone2Name=dyn_fringeTassleBack_B_01_l
  swing1LimitDegrees=20
  swing2LimitDegrees=12
  twistLimitDegrees=8
 }
 {bone1Name=dyn_fringeTassleBack_A_01_l, bone2Name=spine_05, swing1LimitDegrees=20, swing2LimitDegrees=12, twistLimitDegrees=8}
 {bone1Name=dyn_fringeTassleBack_01_c, bone2Name=spine_05, swing1LimitDegrees=20, swing2LimitDegrees=12, twistLimitDegrees=2}
 {bone1Name=dyn_fringeTassleBack_02_c, bone2Name=dyn_fringeTassleBack_01_c, swing1LimitDegrees=15, swing2LimitDegrees=12, twistLimitDegrees=1}
 {
  bone1Name=spine_03
  bone2Name=pelvis
  swing1LimitDegrees=45
  swing1Motion=Free
  swing2LimitDegrees=45
  swing2Motion=Free
  twistLimitDegrees=45
  twistMotion=Free
 }
 {
  bone1Name=spine_04
  bone2Name=spine_03
  swing1LimitDegrees=45
  swing1Motion=Free
  swing2LimitDegrees=45
  swing2Motion=Free
  twistLimitDegrees=45
  twistMotion=Free
 }
 {
  bone1Name=spine_05
  bone2Name=spine_04
  swing1LimitDegrees=45
  swing1Motion=Free
  swing2LimitDegrees=45
  swing2Motion=Free
  twistLimitDegrees=45
  twistMotion=Free
 }
 {bone1Name=dyn_torsoFringe_F_01_r, bone2Name=spine_03, swing1LimitDegrees=25, swing2LimitDegrees=12, twistLimitDegrees=10}
 {bone1Name=dyn_torsoFringe_D_01_r, bone2Name=spine_03, swing1LimitDegrees=25, swing2LimitDegrees=12, twistLimitDegrees=10}
 {bone1Name=dyn_fringeTassleBack_Z_01_l, bone2Name=spine_05, swing1LimitDegrees=20, swing2LimitDegrees=12, twistLimitDegrees=8}
]
