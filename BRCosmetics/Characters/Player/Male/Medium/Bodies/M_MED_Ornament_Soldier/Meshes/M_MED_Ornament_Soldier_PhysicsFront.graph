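# Physics asset for the front of the medium male Soldier ornament, covering its dangling arm fringe and front tassels.
# Spine and upper-arm bodies form a freely rotating backbone that anchors capsule fringe pieces and box tassels, some through intermediate base bodies.
# Fringe and tassel joints use tight asymmetric swing and twist limits.
# Right-arm fringe mostly forms short chains, one branching, while left-arm fringe is single-link, partly hung from a twist bone.
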PhysicsAsset /BRCosmetics/Characters/Player/Male/Medium/Bodies/M_MED_Ornament_Soldier/Meshes/M_MED_Ornament_Soldier_PhysicsFront
bodies {
 dyn_fringeArm_A_01_l [{shapeType=Capsule, length=6, radius=1.5, shapeName=dyn_fringeArm_A_01_l_capsule, center=<2.8049 0.0003 0>, rotation=<rot 90 0 0>}]
 dyn_fringeArm_A_01_r [{shapeType=Capsule, length=7, radius=1.2, shapeName=dyn_fringeArm_A_01_r_capsule, center=<2.8049 0.0003 0>, rotation=<rot 90 0 0>}]
 dyn_fringeArm_A_03_r [{shapeType=Capsule, length=7, radius=1.2, shapeName=dyn_fringeArm_A_03_r_capsule, center=<2.8049 0.0003 0>, rotation=<rot 90 0 0>}]
 dyn_fringeArm_B_01_l [{shapeType=Capsule, length=6, radius=1.5, shapeName=dyn_fringeArm_B_01_l_capsule, center=<2.8049 0.0003 0>, rotation=<rot 90 0 0>}]
 dyn_fringeArm_B_01_r [{shapeType=Capsule, length=7, radius=1.2, shapeName=dyn_fringeArm_B_01_r_capsule, center=<2.8049 0.0003 0>, rotation=<rot 90 0 0>}]
 dyn_fringeArm_C_01_l [{shapeType=Capsule, length=6, radius=1.5, shapeName=dyn_fringeArm_C_01_l_capsule, center=<2.8049 0.0003 0>, rotation=<rot 90 0 0>}]
 dyn_fringeArm_C_01_r [{shapeType=Capsule, length=7, radius=1.2, shapeName=dyn_fringeArm_C_01_r_capsule, center=<2.8049 0.0003 0>, rotation=<rot 90 0 0>}]
 dyn_fringeArm_C_02_r [{shapeType=Capsule, length=7, radius=1.2, shapeName=dyn_fringeArm_C_02_r_capsule, center=<2.8049 0.0003 0>, rotation=<rot 90 0 0>}]
 dyn_fringeArm_D_01_l [{shapeType=Capsule, length=6, radius=1.5, shapeName=dyn_fringeArm_D_01_l_capsule, center=<2.8049 0.0003 0>, rotation=<rot 90 0 0>}]
 dyn_fringeArm_D_01_r [{shapeType=Capsule, length=7, radius=1.2, shapeName=dyn_fringeArm_D_01_r_capsule, center=<2.8049 0.0003 0>, rotation=<rot 90 0 0>}]
 dyn_fringeArm_D_02_r [{shapeType=Capsule, length=7, radius=1.2, shapeName=dyn_fringeArm_D_02_r_capsule, center=<2.8049 0.0003 0>, rotation=<rot 90 0 0>}]
 dyn_fringeArm_E_01_l [{shapeType=Capsule, length=6, radius=1.5, shapeName=dyn_fringeArm_E_01_l_capsule, center=<2.8049 0.0003 0>, rotation=<rot 90 0 0>}]
 dyn_fringeArm_E_01_r [{shapeType=Capsule, length=7, radius=1.2, shapeName=dyn_fringeArm_E_01_r_capsule, center=<2.8049 0.0003 0>, rotation=<rot 90 0 0>}]
 dyn_fringeArm_E_03_r [{shapeType=Capsule, length=7, radius=1.2, shapeName=dyn_fringeArm_E_03_r_capsule, center=<2.8049 0.0003 0>, rotation=<rot 90 0 0>}]
 dyn_fringeArm_F_01_r [{shapeType=Capsule, length=7, radius=1.2, shapeName=dyn_fringeArm_F_01_r_capsule, center=<2.8049 0.0003 0>, rotation=<rot 90 0 0>}]
 dyn_fringeArm_F_03_r [{shapeType=Capsule, length=7, radius=1.2, shapeName=dyn_fringeArm_F_03_r_capsule, center=<2.8049 0.0003 0>, rotation=<rot 90 0 0>}]
 dyn_fringeArm_F_04_r [{shapeType=Capsule, length=7, radius=1.2, shapeName=dyn_fringeArm_F_04_r_capsule, center=<2.8049 0.0003 0>, rotation=<rot 90 0 0>}]
 dyn_fringeArm_G_01_r [{shapeType=Capsule, length=13, radius=3, shapeName=dyn_fringeArm_G_01_r_capsule, center=<6.2675 0.0004 -2.1803>, rotation=<rot 90 0 0>}]
 dyn_fringeArm_H_01_r [{shapeType=Capsule, length=7, radius=1.2, shapeName=dyn_fringeArm_H_01_r_capsule, center=<2.8049 0.0003 0>, rotation=<rot 90 0 0>}]
 dyn_fringeArm_H_02_r [{shapeType=Capsule, length=7, radius=1.2, shapeName=dyn_fringeArm_H_02_r_capsule, center=<2.8049 0.0003 0>, rotation=<rot 90 0 0>}]
 dyn_fringeTassleFrontBase_l [{shapeType=Box, extentX=3.6655, extentY=6.0355, extentZ=6.77, center=<2.994 -1.513 -1.081>, rotation=<rot 0 103.0247 -88.3343>}]
 dyn_fringeTassleFront_B_01_r [{shapeType=Box, extentX=2.4334, extentY=7.1834, extentZ=6.8526, center=<3.2554 0 0.0001>, rotation=<rot 0 90 0>}]
 dyn_fringeTassleFront_C_01_l [{shapeType=Box, extentX=2.4334, extentY=7.1834, extentZ=6.8526, center=<3.2554 0 0.0001>, rotation=<rot 0 90 0>}]
 dyn_fringeTassleFront_C_01_r [{shapeType=Box, extentX=2.4334, extentY=7.1834, extentZ=6.8526, center=<3.2554 0 0.0001>, rotation=<rot 0 90 0>}]
 dyn_fringeTassleFront_C_02_l [{shapeType=Box, extentX=2.4334, extentY=7.1834, extentZ=6.8526, center=<3.2554 0 0.0001>, rotation=<rot 0 90 0>}]
 dyn_fringeTassleFront_C_02_r [{shapeType=Box, extentX=2.4334, extentY=7.1834, extentZ=6.8526, center=<3.2554 0 0.0001>, rotation=<rot 0 90 0>}]
 dyn_fringeTassleFront_D_01_l [{shapeType=Box, extentX=2.4334, extentY=7.1834, extentZ=6.8526, center=<3.2554 0 0.0001>, rotation=<rot 0 90 0>}]
 dyn_fringeTassleFront_D_01_r [{shapeType=Box, extentX=2.4334, extentY=7.1834, extentZ=6.8526, center=<3.2554 0 0.0001>, rotation=<rot 0 90 0>}]
 dyn_fringeTassleFront_E_01_l [{shapeType=Box, extentX=2.4334, extentY=7.1834, extentZ=6.8526, center=<3.2554 0 0.0001>, rotation=<rot 0 90 0>}]
 dyn_fringeTassleFront_E_01_r [{shapeType=Box, extentX=2.4334, extentY=6.9857, extentZ=3.1742, center=<3.2554 0 0.0001>, rotation=<rot 0 90 0>}]
 dyn_fringeTassleFront_E_02_r [{shapeType=Box, extentX=2.4334, extentY=4.9758, extentZ=3.2347, center=<2.2006 0 0.0001>, rotation=<rot 0 90 0>}]
 dyn_fringeTassleFront_F_01_Base_l [{shapeType=Box, extentX=1.01, extentY=1.01, extentZ=1.01, rotation=<rot 0 -90 0>}]
 dyn_fringeTassleFront_F_01_l [{shapeType=Box, extentX=2.4334, extentY=7.1834, extentZ=6.8526, center=<3.2554 0 0.0001>, rotation=<rot 0 90 0>}]
 dyn_fringeTassleFront_F_01_r [{shapeType=Box, extentX=1.6666, extentY=7.1834, extentZ=2.7339, center=<3.2554 0 0.0001>, rotation=<rot -18.2315 89.9999 0>}]
 pelvis [{shapeType=Box, extentX=1.3845, extentY=1.5396, extentZ=1.9676}]
 spine_03 [{shapeType=Capsule, length=5.5873, radius=11.5686}]
 spine_04 [{shapeType=Sphere, radius=5.4886, center=<0.0002 -6.3501 -11.726>}, {shapeType=Capsule, length=13.7471, radius=12.6673, center=<0 0.2459 0>}]
 spine_05 [{shapeType=Capsule, length=14.6402, radius=12.6784}]
 upperarm_l [{shapeType=Box, extentX=1.1, extentY=1.1, extentZ=1.1, rotation=<rot 38.3141 -90 -3.249>}]
 upperarm_r [{shapeType=Capsule, length=13.7155, radius=5.8497, center=<-12.7513 -0.0001 0>, rotation=<rot 90 0 0>}]
 upperarm_twist_02_l [{shapeType=Box, extentX=0.9246, extentY=1.1, extentZ=1.2, center=<0.3131 -2.7494 2.2188>, rotation=<rot 0 176.1938 -72.8313>}]
}
constraints [
 {
  bone1Name=spine_03
  bone2Name=pelvis
  swing1LimitDegrees=45
  swing1Motion=Free
  swing2LimitDegrees=45
  swing2Motion=Free
  twistLimitDegrees=45
  twistMotion=Free
 }
 {
  bone1Name=spine_04
  bone2Name=spine_03
  swing1LimitDegrees=45
  swing1Motion=Free
  swing2LimitDegrees=45
  swing2Motion=Free
  twistLimitDegrees=45
  twistMotion=Free
 }
 {
  bone1Name=spine_05
  bone2Name=spine_04
  swing1LimitDegrees=45
  swing1Motion=Free
  swing2LimitDegrees=45
  swing2Motion=Free
  twistLimitDegrees=45
  twistMotion=Free
 }
 {bone1Name=dyn_fringeTassleFront_F_01_r, bone2Name=spine_05, swing1LimitDegrees=20, swing2LimitDegrees=12, twistLimitDegrees=8}
 {bone1Name=dyn_fringeTassleFront_C_01_r, bone2Name=spine_05, swing1LimitDegrees=20, swing2LimitDegrees=12, twistLimitDegrees=8}
 {
  bone1Name=dyn_fringeTassleFront_C_02_r
  bone2Name=dyn_fringeTassleFront_C_01_r
  swing1LimitDegrees=20
  swing2LimitDegrees=12
  twistLimitDegrees=8
 }
 {bone1Name=dyn_fringeTassleFront_D_01_r, bone2Name=spine_05, swing1LimitDegrees=20, swing2LimitDegrees=12, twistLimitDegrees=8}
 {bone1Name=dyn_fringeTassleFront_E_01_r, bone2Name=spine_05, swing1LimitDegrees=20, swing2LimitDegrees=12, twistLimitDegrees=8}
 {
  bone1Name=dyn_fringeTassleFront_E_02_r
  bone2Name=dyn_fringeTassleFront_E_01_r
  swing1LimitDegrees=20
  swing2LimitDegrees=12
  twistLimitDegrees=8
 }
 {
  bone1Name=dyn_fringeTassleFrontBase_l
  bone2Name=spine_05
  swing1LimitDegrees=45
  swing1Motion=Free
  swing2LimitDegrees=45
  swing2Motion=Free
  twistLimitDegrees=45
  twistMotion=Free
 }
 {
  bone1Name=dyn_fringeTassleFront_D_01_l
  bone2Name=dyn_fringeTassleFrontBase_l
  swing1LimitDegrees=20
  swing2LimitDegrees=12
  twistLimitDegrees=8
 }
 {
  bone1Name=dyn_fringeTassleFront_E_01_l
  bone2Name=dyn_fringeTassleFrontBase_l
  swing1LimitDegrees=20
  swing2LimitDegrees=12
  twistLimitDegrees=8
 }
 {
  bone1Name=dyn_fringeTassleFront_F_01_Base_l
  bone2Name=dyn_fringeTassleFrontBase_l
  swing1LimitDegrees=45
  swing1Motion=Free
  swing2LimitDegrees=45
  swing2Motion=Free
  twistLimitDegrees=45
  twistMotion=Free
 }
 {
  bone1Name=dyn_fringeTassleFront_F_01_l
  bone2Name=dyn_fringeTassleFront_F_01_Base_l
  swing1LimitDegrees=20
  swing2LimitDegrees=12
  twistLimitDegrees=8
 }
 {bone1Name=dyn_fringeTassleFront_C_01_l, bone2Name=spine_05, swing1LimitDegrees=20, swing2LimitDegrees=12, twistLimitDegrees=8}
 {
  bone1Name=dyn_fringeTassleFront_C_02_l
  bone2Name=dyn_fringeTassleFront_C_01_l
  swing1LimitDegrees=20
  swing2LimitDegrees=12
  twistLimitDegrees=8
 }
 {bone1Name=dyn_fringeTassleFront_B_01_r, bone2Name=spine_04, swing1LimitDegrees=20, swing2LimitDegrees=12, twistLimitDegrees=8}
 {bone1Name=dyn_fringeArm_G_01_r, bone2Name=upperarm_r, swing1LimitDegrees=17.5, swing2LimitDegrees=6, twistLimitDegrees=5}
 {bone1Name=dyn_fringeArm_B_01_r, bone2Name=upperarm_r, swing1LimitDegrees=25, swing2LimitDegrees=12, twistLimitDegrees=10}
 {bone1Name=dyn_fringeArm_A_01_r, bone2Name=upperarm_r, swing1LimitDegrees=25, swing2LimitDegrees=12, twistLimitDegrees=10}
 {bone1Name=dyn_fringeArm_A_03_r, bone2Name=dyn_fringeArm_A_01_r, swing1LimitDegrees=25, swing2LimitDegrees=12, twistLimitDegrees=10}
 {bone1Name=dyn_fringeArm_H_01_r, bone2Name=upperarm_r, swing1LimitDegrees=25, swing2LimitDegrees=12, twistLimitDegrees=10}
 {bone1Name=dyn_fringeArm_H_02_r, bone2Name=dyn_fringeArm_H_01_r, swing1LimitDegrees=25, swing2LimitDegrees=12, twistLimitDegrees=10}
 {bone1Name=dyn_fringeArm_F_01_r, bone2Name=upperarm_r, swing1LimitDegrees=25, swing2LimitDegrees=12, twistLimitDegrees=10}
 {bone1Name=dyn_fringeArm_F_03_r, bone2Name=dyn_fringeArm_F_01_r, swing1LimitDegrees=25, swing2LimitDegrees=12, twistLimitDegrees=10}
 {bone1Name=dyn_fringeArm_F_04_r, bone2Name=dyn_fringeArm_F_01_r, swing1LimitDegrees=25, swing2LimitDegrees=12, twistLimitDegrees=10}
 {bone1Name=dyn_fringeArm_E_01_r, bone2Name=upperarm_r, swing1LimitDegrees=25, swing2LimitDegrees=12, twistLimitDegrees=10}
 {bone1Name=dyn_fringeArm_E_03_r, bone2Name=dyn_fringeArm_E_01_r, swing1LimitDegrees=25, swing2LimitDegrees=12, twistLimitDegrees=10}
 {bone1Name=dyn_fringeArm_D_01_r, bone2Name=upperarm_r, swing1LimitDegrees=25, swing2LimitDegrees=12, twistLimitDegrees=10}
 {bone1Name=dyn_fringeArm_D_02_r, bone2Name=dyn_fringeArm_D_01_r, swing1LimitDegrees=25, swing2LimitDegrees=12, twistLimitDegrees=10}
 {bone1Name=dyn_fringeArm_C_01_r, bone2Name=upperarm_r, swing1LimitDegrees=25, swing2LimitDegrees=12, twistLimitDegrees=10}
 {bone1Name=dyn_fringeArm_C_02_r, bone2Name=dyn_fringeArm_C_01_r, swing1LimitDegrees=25, swing2LimitDegrees=12, twistLimitDegrees=10}
 {
  bone1Name=upperarm_r
  bone2Name=spine_05
  swing1LimitDegrees=45
  swing1Motion=Free
  swing2LimitDegrees=45
  swing2Motion=Free
  twistLimitDegrees=45
  twistMotion=Free
 }
 {bone1Name=dyn_fringeArm_A_01_l, bone2Name=upperarm_twist_02_l, swing1LimitDegrees=25, swing2LimitDegrees=12, twistLimitDegrees=10}
 {bone1Name=dyn_fringeArm_C_01_l, bone2Name=upperarm_twist_02_l, swing1LimitDegrees=25, swing2LimitDegrees=12, twistLimitDegrees=10}
 {bone1Name=dyn_fringeArm_B_01_l, bone2Name=upperarm_twist_02_l, swing1LimitDegrees=25, swing2LimitDegrees=12, twistLimitDegrees=10}
 {bone1Name=dyn_fringeArm_D_01_l, bone2Name=upperarm_l, swing1LimitDegrees=25, swing2LimitDegrees=12, twistLimitDegrees=10}
 {bone1Name=dyn_fringeArm_E_01_l, bone2Name=upperarm_l, swing1LimitDegrees=25, swing2LimitDegrees=12, twistLimitDegrees=10}
 {
  bone1Name=upperarm_twist_02_l
  bone2Name=upperarm_l
  swing1LimitDegrees=45
  swing1Motion=Free
  swing2LimitDegrees=45
  swing2Motion=Free
  twistLimitDegrees=45
  twistMotion=Free
 }
 {
  bone1Name=upperarm_l
  bone2Name=spine_05
  swing1LimitDegrees=45
  swing1Motion=Free
  swing2LimitDegrees=45
  swing2Motion=Free
  twistLimitDegrees=45
  twistMotion=Free
 }
]
