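# Physics asset for the left leg of the male medium Ornament Soldier cosmetic body, covering its dynamic leg bones.
# Small boxes on the pelvis and calf bones and one large thigh capsule form the anchor chain, linked by looser constraints.
# Thin capsules on dyn_legHi bones hang from the thigh twist and those on dyn_legLow from the calf bones, all sharing tighter, asymmetric swing limits.
# One low bone, P, also carries a box.
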PhysicsAsset /BRCosmetics/Characters/Player/Male/Medium/Bodies/M_MED_Ornament_Soldier/Meshes/M_MED_Ornament_Soldier_PhysicsLLeg
bodies {
 calf_l [{shapeType=Box, extentX=1.3845, extentY=1.5396, extentZ=1.9676}]
 calf_twist_01_l [{shapeType=Box, extentX=1.3845, extentY=1.5396, extentZ=1.9676}]
 calf_twist_02_l [{shapeType=Box, extentX=1.3845, extentY=1.5396, extentZ=1.9676}]
 dyn_legHi_A_01_l [{shapeType=Capsule, length=11, radius=2, shapeName=dyn_legHi_A_01_l_capsule, center=<-4.5839 0.0001 0>, rotation=<rot 90 0 0>}]
 dyn_legHi_D_01_l [{shapeType=Capsule, length=11, radius=2, shapeName=dyn_legHi_D_01_l_capsule, center=<-5.1801 0 -0.0001>, rotation=<rot 90 0 0>}]
 dyn_legHi_E_01_l [{shapeType=Capsule, length=11, radius=2, shapeName=dyn_legHi_E_01_l_capsule, center=<-5.1801 0 -0.0001>, rotation=<rot 90 0 0>}]
 dyn_legHi_G_01_l [{shapeType=Capsule, length=11, radius=2, shapeName=dyn_legHi_G_01_l_capsule, center=<-5.1801 0 -0.0001>, rotation=<rot 90 0 0>}]
 dyn_legHi_J_01_l [{shapeType=Capsule, length=11, radius=2, shapeName=dyn_legHi_J_01_l_capsule, center=<-5.1801 0 -0.0001>, rotation=<rot 90 0 0>}]
 dyn_legHi_K_01_l [{shapeType=Capsule, length=11, radius=2, shapeName=dyn_legHi_K_01_l_capsule, center=<-3.876 0 -0.0001>, rotation=<rot 90 0 0>}]
 dyn_legHi_L_01_l [{shapeType=Capsule, length=11, radius=2, shapeName=dyn_legHi_L_01_l_capsule, center=<-5.1801 0 -0.0001>, rotation=<rot 90 0 0>}]
 dyn_legLow_A_01_l [{shapeType=Capsule, length=11, radius=2, shapeName=dyn_legLow_A_01_l_capsule, center=<-5.1801 0 -0.0001>, rotation=<rot 90 0 0>}]
 dyn_legLow_B_01_l [{shapeType=Capsule, length=11, radius=2, shapeName=dyn_legLow_B_01_l_capsule, center=<-5.1801 0 -0.0001>, rotation=<rot 90 0 0>}]
 dyn_legLow_C_01_l [{shapeType=Capsule, length=11, radius=2, shapeName=dyn_legLow_C_01_l_capsule, center=<-5.1801 0 -0.0001>, rotation=<rot 90 0 0>}]
 dyn_legLow_E_01_l [{shapeType=Capsule, length=11, radius=2, shapeName=dyn_legLow_E_01_l_capsule, center=<-5.1801 0 -0.0001>, rotation=<rot 90 0 0>}]
 dyn_legLow_F_01_l [{shapeType=Capsule, length=11, radius=2, shapeName=dyn_legLow_F_01_l_capsule, center=<-5.1801 0 -0.0001>, rotation=<rot 90 0 0>}]
 dyn_legLow_G_01_l [{shapeType=Capsule, length=11, radius=2, shapeName=dyn_legLow_G_01_l_capsule, center=<-5.1801 0 -0.0001>, rotation=<rot 90 0 0>}]
 dyn_legLow_H_01_l [{shapeType=Capsule, length=11, radius=2, shapeName=dyn_legLow_H_01_l_capsule, center=<-5.1801 0 -0.0001>, rotation=<rot 90 0 0>}]
 dyn_legLow_J_01_l [{shapeType=Capsule, length=11, radius=2, shapeName=dyn_legLow_J_01_l_capsule, center=<-5.1801 0 -0.0001>, rotation=<rot 90 0 0>}]
 dyn_legLow_N_01_l [{shapeType=Capsule, length=11, radius=2, shapeName=dyn_legLow_N_01_l_capsule, center=<-5.1801 0 -0.0001>, rotation=<rot 90 0 0>}]
 dyn_legLow_P_01_l [
  {shapeType=Capsule, length=9, radius=2, shapeName=dyn_legLow_P_01_l_capsule, rotation=<rot 90 0 0>}
  {shapeType=Box, extentX=1.8931, extentY=11.4732, extentZ=8.487, center=<-5.1801 0 -0.0001>, rotation=<rot 0 90 0>}
 ]
 dyn_legLow_Q_01_l [{shapeType=Capsule, length=9, radius=2, shapeName=dyn_legLow_Q_01_l_capsule, center=<-4.0631 0.0001 0>, rotation=<rot 90 0 0>}]
 dyn_legLow_R_01_l [{shapeType=Capsule, length=12, radius=2, shapeName=dyn_legLow_R_01_l_capsule, center=<-3.832 0 -0.2518>}]
 dyn_legLow_V_01_l [{shapeType=Capsule, length=12, radius=2, shapeName=dyn_legLow_V_01_l_capsule, center=<-5.2329 0 0>}]
 dyn_legLow_W_01_l [{shapeType=Capsule, length=11, radius=2, shapeName=dyn_legLow_W_01_l_capsule, center=<-5.1801 0 -0.0001>, rotation=<rot 90 0 0>}]
 dyn_legLow_X_01_l [{shapeType=Capsule, length=10, radius=2, shapeName=dyn_legLow_X_01_l_capsule, center=<-5.1801 0 -0.0001>, rotation=<rot 90 0 0>}]
 dyn_legLow_Y_01_l [{shapeType=Capsule, length=11, radius=2, shapeName=dyn_legLow_Y_01_l_capsule, center=<-5.1801 0 -0.0001>, rotation=<rot 90 0 0>}]
 pelvis [{shapeType=Box, extentX=1.3845, extentY=1.5396, extentZ=1.9676}]
 thigh_twist_01_l [{shapeType=Capsule, length=19.9339, radius=8.8733, center=<-4.4015 -0.0466 -0.0001>, rotation=<rot 0 90.6064 89.9999>}]
}
constraints [
 {bone1Name=calf_l, bone2Name=pelvis, swing1LimitDegrees=45, swing2LimitDegrees=45, twistLimitDegrees=45}
 {bone1Name=calf_twist_01_l, bone2Name=calf_l, swing1LimitDegrees=45, swing2LimitDegrees=45, twistLimitDegrees=45}
 {bone1Name=calf_twist_02_l, bone2Name=calf_l, swing1LimitDegrees=45, swing2LimitDegrees=45, twistLimitDegrees=45}
 {bone1Name=thigh_twist_01_l, bone2Name=pelvis, swing1LimitDegrees=45, swing2LimitDegrees=45, twistLimitDegrees=45}
 {bone1Name=dyn_legLow_Q_01_l, bone2Name=calf_twist_01_l, swing1LimitDegrees=25, swing2LimitDegrees=12, twistLimitDegrees=10}
 {bone1Name=dyn_legLow_R_01_l, bone2Name=calf_twist_01_l, swing1LimitDegrees=25, swing2LimitDegrees=12, twistLimitDegrees=10}
 {bone1Name=dyn_legLow_V_01_l, bone2Name=calf_twist_01_l, swing1LimitDegrees=25, swing2LimitDegrees=12, twistLimitDegrees=10}
 {bone1Name=dyn_legLow_X_01_l, bone2Name=calf_twist_01_l, swing1LimitDegrees=25, swing2LimitDegrees=12, twistLimitDegrees=10}
 {bone1Name=dyn_legLow_N_01_l, bone2Name=calf_twist_02_l, swing1LimitDegrees=25, swing2LimitDegrees=12, twistLimitDegrees=10}
 {bone1Name=dyn_legLow_P_01_l, bone2Name=calf_twist_02_l, swing1LimitDegrees=25, swing2LimitDegrees=12, twistLimitDegrees=10}
 {bone1Name=dyn_legLow_Y_01_l, bone2Name=calf_twist_02_l, swing1LimitDegrees=25, swing2LimitDegrees=12, twistLimitDegrees=10}
 {bone1Name=dyn_legLow_W_01_l, bone2Name=calf_twist_02_l, swing1LimitDegrees=25, swing2LimitDegrees=12, twistLimitDegrees=10}
 {bone1Name=dyn_legHi_D_01_l, bone2Name=thigh_twist_01_l, swing1LimitDegrees=25, swing2LimitDegrees=12, twistLimitDegrees=10}
 {bone1Name=dyn_legHi_E_01_l, bone2Name=thigh_twist_01_l, swing1LimitDegrees=25, swing2LimitDegrees=12, twistLimitDegrees=10}
 {bone1Name=dyn_legHi_G_01_l, bone2Name=thigh_twist_01_l, swing1LimitDegrees=25, swing2LimitDegrees=12, twistLimitDegrees=10}
 {bone1Name=dyn_legHi_L_01_l, bone2Name=thigh_twist_01_l, swing1LimitDegrees=25, swing2LimitDegrees=12, twistLimitDegrees=10}
 {bone1Name=dyn_legLow_G_01_l, bone2Name=calf_twist_02_l, swing1LimitDegrees=25, swing2LimitDegrees=12, twistLimitDegrees=10}
 {bone1Name=dyn_legLow_H_01_l, bone2Name=calf_twist_02_l, swing1LimitDegrees=25, swing2LimitDegrees=12, twistLimitDegrees=10}
 {bone1Name=dyn_legLow_J_01_l, bone2Name=calf_twist_02_l, swing1LimitDegrees=25, swing2LimitDegrees=12, twistLimitDegrees=10}
 {bone1Name=dyn_legLow_A_01_l, bone2Name=calf_l, swing1LimitDegrees=25, swing2LimitDegrees=12, twistLimitDegrees=10}
 {bone1Name=dyn_legLow_B_01_l, bone2Name=calf_l, swing1LimitDegrees=25, swing2LimitDegrees=12, twistLimitDegrees=10}
 {bone1Name=dyn_legLow_C_01_l, bone2Name=calf_l, swing1LimitDegrees=25, swing2LimitDegrees=12, twistLimitDegrees=10}
 {bone1Name=dyn_legLow_E_01_l, bone2Name=calf_l, swing1LimitDegrees=25, swing2LimitDegrees=12, twistLimitDegrees=10}
 {bone1Name=dyn_legLow_F_01_l, bone2Name=calf_l, swing1LimitDegrees=25, swing2LimitDegrees=12, twistLimitDegrees=10}
 {bone1Name=dyn_legHi_J_01_l, bone2Name=thigh_twist_01_l, swing1LimitDegrees=25, swing2LimitDegrees=12, twistLimitDegrees=10}
 {bone1Name=dyn_legHi_K_01_l, bone2Name=thigh_twist_01_l, swing1LimitDegrees=25, swing2LimitDegrees=12, twistLimitDegrees=10}
 {bone1Name=dyn_legHi_A_01_l, bone2Name=thigh_twist_01_l, swing1LimitDegrees=25, swing2LimitDegrees=12, twistLimitDegrees=10}
]
