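# Physics asset for the right leg of the male medium Ornament Soldier body, covering its dynamic upper- and lower-leg bones.
# Small boxes on pelvis, calf and calf twist form the base hierarchy, with a large capsule for the thigh twist.
# Each dyn_legHi/legLow bone has a small capsule constrained directly to thigh twist, calf or calf twist, without chaining.
# These ornament joints use tight, asymmetric swing and twist limits, narrower than the core leg joints.
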PhysicsAsset /BRCosmetics/Characters/Player/Male/Medium/Bodies/M_MED_Ornament_Soldier/Meshes/M_MED_Ornament_Soldier_PhysicsRLeg
bodies {
 calf_r [{shapeType=Box, extentX=1.3845, extentY=1.5396, extentZ=1.9676}]
 calf_twist_01_r [{shapeType=Box, extentX=1.3845, extentY=1.5396, extentZ=1.9676}]
 dyn_legHi_A_01_r [{shapeType=Capsule, length=9, radius=2, shapeName=dyn_legHi_A_01_r_capsule, center=<3.5799 0 0>, rotation=<rot 90 0 0>}]
 dyn_legHi_B_01_r [{shapeType=Capsule, length=9, radius=2, shapeName=dyn_legHi_B_01_r_capsule, center=<3.5799 0 0>, rotation=<rot 90 0 0>}]
 dyn_legHi_D_01_r [{shapeType=Capsule, length=9, radius=2, shapeName=dyn_legHi_D_01_r_capsule, center=<3.5799 0 0>, rotation=<rot 90 0 0>}]
 dyn_legHi_E_01_r [{shapeType=Capsule, length=9, radius=2, shapeName=dyn_legHi_E_01_r_capsule, center=<3.5799 0 0>, rotation=<rot 90 0 0>}]
 dyn_legHi_F_01_r [{shapeType=Capsule, length=9, radius=2, shapeName=dyn_legHi_F_01_r_capsule, center=<3.5799 0 0>, rotation=<rot 90 0 0>}]
 dyn_legHi_G_01_r [{shapeType=Capsule, length=9, radius=2, shapeName=dyn_legHi_G_01_r_capsule, center=<3.5799 0 0>, rotation=<rot 90 0 0>}]
 dyn_legHi_J_01_r [{shapeType=Capsule, length=9, radius=2, shapeName=dyn_legHi_J_01_r_capsule, center=<3.5799 0 0>, rotation=<rot 90 0 0>}]
 dyn_legHi_L_01_r [{shapeType=Capsule, length=9, radius=2, shapeName=dyn_legHi_L_01_r_capsule, center=<3.5799 0 0>, rotation=<rot 90 0 0>}]
 dyn_legLow_A_01_r [{shapeType=Capsule, length=9, radius=2, shapeName=dyn_legLow_A_01_r_capsule, center=<3.8004 -0.0001 0>, rotation=<rot 90 0 0>}]
 dyn_legLow_B_01_r [{shapeType=Capsule, length=9, radius=2, shapeName=dyn_legLow_B_01_r_capsule, center=<3.5799 0 0>, rotation=<rot 90 0 0>}]
 dyn_legLow_C_01_r [{shapeType=Capsule, length=9, radius=2, shapeName=dyn_legLow_C_01_r_capsule, center=<3.5799 0 0>, rotation=<rot 90 0 0>}]
 dyn_legLow_D_01_r [{shapeType=Capsule, length=9, radius=2, shapeName=dyn_legLow_D_01_r_capsule, center=<3.5799 0 0>, rotation=<rot 90 0 0>}]
 dyn_legLow_E_01_r [{shapeType=Capsule, length=9, radius=2, shapeName=dyn_legLow_E_01_r_capsule, center=<3.5799 0 0>, rotation=<rot 90 0 0>}]
 dyn_legLow_G_01_r [{shapeType=Capsule, length=9, radius=2, shapeName=dyn_legLow_G_01_r_capsule, center=<4.0377 0 0>, rotation=<rot 90 0 0>}]
 dyn_legLow_H_01_r [{shapeType=Capsule, length=9, radius=2, shapeName=dyn_legLow_H_01_r_capsule, center=<3.5799 0 0>, rotation=<rot 90 0 0>}]
 dyn_legLow_J_01_r [{shapeType=Capsule, length=9, radius=2, shapeName=dyn_legLow_J_01_r_capsule, center=<3.2554 0 0.0001>, rotation=<rot 90 0 0>}]
 dyn_legLow_N_01_r [{shapeType=Capsule, length=9, radius=2, shapeName=dyn_legLow_N_01_r_capsule, center=<2.9467 0 0>, rotation=<rot 90 0 0>}]
 dyn_legLow_P_01_r [{shapeType=Capsule, length=9, radius=2, shapeName=dyn_legLow_P_01_r_capsule, center=<3.4229 0 0.0001>, rotation=<rot 90 0 0>}]
 dyn_legLow_Q_01_r [{shapeType=Capsule, length=9, radius=2, shapeName=dyn_legLow_Q_01_r_capsule, center=<4.6211 0 0>, rotation=<rot 90 0 0>}]
 dyn_legLow_R_01_r [{shapeType=Capsule, length=9, radius=2, shapeName=dyn_legLow_R_01_r_capsule, center=<5 0 0>, rotation=<rot 90 0 0>}]
 dyn_legLow_S_01_r [{shapeType=Capsule, length=9, radius=2, shapeName=dyn_legLow_S_01_r_capsule, center=<4.0436 0.0001 0>, rotation=<rot 90 0 0>}]
 dyn_legLow_V_01_r [{shapeType=Capsule, length=9, radius=2, shapeName=dyn_legLow_V_01_r_capsule, center=<5 0 0>, rotation=<rot 90 0 0>}]
 dyn_legLow_X_01_r [{shapeType=Capsule, length=9, radius=2, shapeName=dyn_legLow_X_01_r_capsule, center=<5 0 0>, rotation=<rot 90 0 0>}]
 dyn_legLow_Y_01_r [{shapeType=Capsule, length=9, radius=2, shapeName=dyn_legLow_Y_01_r_capsule, center=<3.8004 -0.0001 0>, rotation=<rot 90 0 0>}]
 pelvis [{shapeType=Box, extentX=1.3845, extentY=1.5396, extentZ=1.9676}]
 thigh_twist_01_r [{shapeType=Capsule, length=23.8877, radius=9.0589, center=<4.0828 0.0002 0>, rotation=<rot 0 90 90>}]
}
constraints [
 {bone1Name=calf_r, bone2Name=pelvis, swing1LimitDegrees=45, swing2LimitDegrees=45, twistLimitDegrees=45}
 {bone1Name=calf_twist_01_r, bone2Name=calf_r, swing1LimitDegrees=45, swing2LimitDegrees=45, twistLimitDegrees=45}
 {bone1Name=thigh_twist_01_r, bone2Name=pelvis, swing1LimitDegrees=45, swing2LimitDegrees=45, twistLimitDegrees=45}
 {bone1Name=dyn_legLow_Q_01_r, bone2Name=calf_twist_01_r, swing1LimitDegrees=25, swing2LimitDegrees=12, twistLimitDegrees=10}
 {bone1Name=dyn_legLow_R_01_r, bone2Name=calf_twist_01_r, swing1LimitDegrees=20, swing2LimitDegrees=12, twistLimitDegrees=8}
 {bone1Name=dyn_legLow_S_01_r, bone2Name=calf_twist_01_r, swing1LimitDegrees=20, swing2LimitDegrees=12, twistLimitDegrees=8}
 {bone1Name=dyn_legLow_V_01_r, bone2Name=calf_twist_01_r, swing1LimitDegrees=20, swing2LimitDegrees=12, twistLimitDegrees=8}
 {bone1Name=dyn_legLow_X_01_r, bone2Name=calf_twist_01_r, swing1LimitDegrees=20, swing2LimitDegrees=12, twistLimitDegrees=8}
 {bone1Name=dyn_legLow_G_01_r, bone2Name=calf_r, swing1LimitDegrees=20, swing2LimitDegrees=12, twistLimitDegrees=8}
 {bone1Name=dyn_legLow_H_01_r, bone2Name=calf_r, swing1LimitDegrees=20, swing2LimitDegrees=12, twistLimitDegrees=8}
 {bone1Name=dyn_legLow_J_01_r, bone2Name=calf_r, swing1LimitDegrees=20, swing2LimitDegrees=12, twistLimitDegrees=8}
 {bone1Name=dyn_legLow_N_01_r, bone2Name=calf_r, swing1LimitDegrees=20, swing2LimitDegrees=12, twistLimitDegrees=8}
 {bone1Name=dyn_legLow_P_01_r, bone2Name=calf_r, swing1LimitDegrees=20, swing2LimitDegrees=12, twistLimitDegrees=8}
 {bone1Name=dyn_legLow_Y_01_r, bone2Name=calf_r, swing1LimitDegrees=20, swing2LimitDegrees=12, twistLimitDegrees=8}
 {bone1Name=dyn_legLow_A_01_r, bone2Name=calf_r, swing1LimitDegrees=20, swing2LimitDegrees=12, twistLimitDegrees=8}
 {bone1Name=dyn_legLow_B_01_r, bone2Name=calf_r, swing1LimitDegrees=20, swing2LimitDegrees=12, twistLimitDegrees=8}
 {bone1Name=dyn_legLow_C_01_r, bone2Name=calf_r, swing1LimitDegrees=20, swing2LimitDegrees=12, twistLimitDegrees=8}
 {bone1Name=dyn_legLow_D_01_r, bone2Name=calf_r, swing1LimitDegrees=20, swing2LimitDegrees=12, twistLimitDegrees=8}
 {bone1Name=dyn_legLow_E_01_r, bone2Name=calf_r, swing1LimitDegrees=20, swing2LimitDegrees=12, twistLimitDegrees=8}
 {bone1Name=dyn_legHi_A_01_r, bone2Name=thigh_twist_01_r, swing1LimitDegrees=25, swing2LimitDegrees=12, twistLimitDegrees=10}
 {bone1Name=dyn_legHi_B_01_r, bone2Name=thigh_twist_01_r, swing1LimitDegrees=25, swing2LimitDegrees=12, twistLimitDegrees=10}
 {bone1Name=dyn_legHi_D_01_r, bone2Name=thigh_twist_01_r, swing1LimitDegrees=25, swing2LimitDegrees=12, twistLimitDegrees=10}
 {bone1Name=dyn_legHi_E_01_r, bone2Name=thigh_twist_01_r, swing1LimitDegrees=25, swing2LimitDegrees=12, twistLimitDegrees=10}
 {bone1Name=dyn_legHi_F_01_r, bone2Name=thigh_twist_01_r, swing1LimitDegrees=25, swing2LimitDegrees=12, twistLimitDegrees=10}
 {bone1Name=dyn_legHi_G_01_r, bone2Name=thigh_twist_01_r, swing1LimitDegrees=25, swing2LimitDegrees=12, twistLimitDegrees=10}
 {bone1Name=dyn_legHi_J_01_r, bone2Name=thigh_twist_01_r, swing1LimitDegrees=25, swing2LimitDegrees=12, twistLimitDegrees=10}
 {bone1Name=dyn_legHi_L_01_r, bone2Name=thigh_twist_01_r, swing1LimitDegrees=25, swing2LimitDegrees=12, twistLimitDegrees=10}
]
