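# Physics asset for the dangling ornaments on the medium male Soldier player body, covering dyn_ bones on the legs, arms, upper back and rear.
# Pelvis, spine, thigh, calf and upper-arm bodies provide anchors or body volumes, with arm ornaments constrained to spine_05 rather than the arms.
# Ornaments are short chains, mostly a thin link and rounded tip, with tight angular limits.
# Some chains start at small Base bodies with fully free angular constraints.
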PhysicsAsset /Game/Characters/Player/Male/Medium/Bodies/M_MED_Ornament_Soldier/Meshes/M_MED_Ornament_Soldier_PhysicsOrnament
bodies {
 calf_l [{shapeType=Box, extentX=1.3845, extentY=1.5396, extentZ=1.9676}]
 calf_r [{shapeType=Box, extentX=1.3845, extentY=1.5396, extentZ=1.9676}]
 dyn_legHi_C_01_r [{shapeType=Box, extentX=1.2542, extentY=3.8901, extentZ=1.5386, center=<1.8719 0 0>, rotation=<rot 0 90 0>}]
 dyn_legHi_C_02_r [{shapeType=Sphere, radius=4.457, center=<4.3481 0.0141 -0.0418>}]
 dyn_legHi_H_01_Base_r [{shapeType=Box, extentX=1.01, extentY=1.01, extentZ=1.01, rotation=<rot 0 -90 0>}]
 dyn_legHi_H_01_r [{shapeType=Box, extentX=1.2542, extentY=4.9695, extentZ=1.5386, center=<1.7252 0 -0.0001>, rotation=<rot 0 90 0>}]
 dyn_legHi_H_02_r [{shapeType=Sphere, radius=4.2651, center=<5.4819 0.0481 0.0636>}]
 dyn_legHi_P_01_l [{shapeType=Sphere, radius=4.4527, center=<4.7192 0.0248 -0.076>}]
 dyn_legLow_BB_01_Base_l [{shapeType=Box, extentX=1.01, extentY=1.01, extentZ=1.01, rotation=<rot 0 -90 0>}]
 dyn_legLow_BB_01_l [{shapeType=Box, extentX=1.2466, extentY=1.2705, extentZ=4.8849, center=<1.5965 0.0955 -0.0456>, rotation=<rot -37.8159 -89.9997 93.2183>}]
 dyn_legLow_BB_02_l [{shapeType=Sphere, radius=4.4541, center=<4.6047 0.5574 0.0703>}]
 dyn_legLow_M_01_r [{shapeType=Capsule, length=1.5651, radius=0.6673, center=<1.1817 0 0>, rotation=<rot 90 0 0>}]
 dyn_legLow_M_02_r [{shapeType=Capsule, length=3.0981, radius=3.5539, center=<4.0692 -0.0067 0.0827>, rotation=<rot 0 -87.6729 -89.4282>}]
 dyn_ornamentArm_01_l [{shapeType=Box, extentX=1.2542, extentY=4.9695, extentZ=1.5386, center=<1.7252 0 -0.0001>, rotation=<rot 0 90 0>}]
 dyn_ornamentArm_02_l [{shapeType=Sphere, radius=4.5269, center=<5.4819 0.0481 0.0636>}]
 dyn_ornamentArm_A_01_Base_r [{shapeType=Capsule, length=0.505, radius=0.505, rotation=<rot 0 -90 0>}]
 dyn_ornamentArm_A_01_r [{shapeType=Box, extentX=1.2542, extentY=4.9695, extentZ=1.5386, center=<1.7252 0 -0.0001>, rotation=<rot 0 90 0>}]
 dyn_ornamentArm_A_02_r [{shapeType=Sphere, radius=4.2651, center=<5.4819 0.0481 0.0636>}]
 dyn_ornamentArm_B_01_Base_r [{shapeType=Capsule, length=0.505, radius=0.505, rotation=<rot 0 -90 0>}]
 dyn_ornamentArm_B_01_r [{shapeType=Box, extentX=1.2542, extentY=4.9695, extentZ=1.5386, center=<1.7252 0 -0.0001>, rotation=<rot 0 90 0>}]
 dyn_ornamentArm_B_02_r [{shapeType=Sphere, radius=4.2651, center=<3.9139 -1.6146 0.0636>}]
 dyn_rearOrnament_01 [{shapeType=Box, extentX=1.2542, extentY=4.9695, extentZ=1.5386, center=<1.7252 0 -0.0001>, rotation=<rot 0 90 0>}]
 dyn_rearOrnament_02 [{shapeType=Sphere, radius=4.2651, center=<5.4819 0.0481 0.0636>}]
 dyn_upperBackOrnament_01_c [{shapeType=Box, extentX=1.2542, extentY=4.9695, extentZ=1.5386, center=<1.7252 0 -0.0001>, rotation=<rot 0 90 0>}]
 dyn_upperBackOrnament_01_c_Base [{shapeType=Box, extentX=1.01, extentY=1.01, extentZ=1.01, rotation=<rot 0 -90 0>}]
 dyn_upperBackOrnament_02_c [{shapeType=Sphere, radius=4.5269, center=<5.4819 0.0481 0.0636>}]
 pelvis [{shapeType=Box, extentX=1.3845, extentY=1.5396, extentZ=1.9676}]
 spine_05 [{shapeType=Capsule, length=14.6402, radius=12.6784}]
 thigh_twist_01_l [{shapeType=Capsule, length=19.9339, radius=8.8733, center=<-4.4015 -0.0466 -0.0001>, rotation=<rot 0 90.6064 89.9999>}]
 thigh_twist_01_r [{shapeType=Capsule, length=23.8877, radius=9.0589, center=<4.0828 0.0002 0>, rotation=<rot 0 90 90>}]
 upperarm_l [{shapeType=Capsule, length=13.7155, radius=5.8497, center=<12.7513 0.0001 0>, rotation=<rot 90 0 180>}]
 upperarm_r [{shapeType=Capsule, length=13.7155, radius=5.8497, center=<-12.7513 -0.0001 0>, rotation=<rot 90 0 0>}]
}
constraints [
 {bone1Name=calf_l, bone2Name=pelvis, swing1LimitDegrees=45, swing2LimitDegrees=45, twistLimitDegrees=45}
 {bone1Name=thigh_twist_01_l, bone2Name=pelvis, swing1LimitDegrees=45, swing2LimitDegrees=45, twistLimitDegrees=45}
 {bone1Name=calf_r, bone2Name=pelvis, swing1LimitDegrees=45, swing2LimitDegrees=45, twistLimitDegrees=45}
 {bone1Name=thigh_twist_01_r, bone2Name=pelvis, swing1LimitDegrees=45, swing2LimitDegrees=45, twistLimitDegrees=45}
 {bone1Name=dyn_legLow_M_01_r, bone2Name=calf_r, swing1LimitDegrees=12, swing2LimitDegrees=12, twistLimitDegrees=12}
 {bone1Name=dyn_legLow_M_02_r, bone2Name=dyn_legLow_M_01_r, swing1LimitDegrees=16, swing2LimitDegrees=12, twistLimitDegrees=8}
 {bone1Name=dyn_legHi_P_01_l, bone2Name=thigh_twist_01_l, swing1LimitDegrees=14, swing2LimitDegrees=14, twistLimitDegrees=14}
 {
  bone1Name=dyn_legLow_BB_01_Base_l
  bone2Name=calf_l
  swing1LimitDegrees=45
  swing1Motion=Free
  swing2LimitDegrees=45
  swing2Motion=Free
  twistLimitDegrees=45
  twistMotion=Free
 }
 {bone1Name=dyn_legLow_BB_01_l, bone2Name=dyn_legLow_BB_01_Base_l, swing1LimitDegrees=16, swing2LimitDegrees=16, twistLimitDegrees=16}
 {bone1Name=dyn_legLow_BB_02_l, bone2Name=dyn_legLow_BB_01_l, swing1LimitDegrees=16, swing2LimitDegrees=16, twistLimitDegrees=16}
 {bone1Name=dyn_legHi_C_01_r, bone2Name=thigh_twist_01_r, swing1LimitDegrees=25, swing2LimitDegrees=12, twistLimitDegrees=10}
 {bone1Name=dyn_legHi_C_02_r, bone2Name=dyn_legHi_C_01_r, swing1LimitDegrees=16, swing2LimitDegrees=16, twistLimitDegrees=16}
 {
  bone1Name=dyn_legHi_H_01_Base_r
  bone2Name=thigh_twist_01_r
  swing1LimitDegrees=45
  swing1Motion=Free
  swing2LimitDegrees=45
  swing2Motion=Free
  twistLimitDegrees=45
  twistMotion=Free
 }
 {bone1Name=dyn_legHi_H_01_r, bone2Name=dyn_legHi_H_01_Base_r, swing1LimitDegrees=16, swing2LimitDegrees=16, twistLimitDegrees=16}
 {bone1Name=dyn_legHi_H_02_r, bone2Name=dyn_legHi_H_01_r, swing1LimitDegrees=16, swing2LimitDegrees=16, twistLimitDegrees=16}
 {bone1Name=dyn_rearOrnament_01, bone2Name=pelvis, swing1LimitDegrees=16, swing2LimitDegrees=16, twistLimitDegrees=16}
 {bone1Name=dyn_rearOrnament_02, bone2Name=dyn_rearOrnament_01, swing1LimitDegrees=16, swing2LimitDegrees=16, twistLimitDegrees=16}
 {
  bone1Name=spine_05
  bone2Name=pelvis
  swing1LimitDegrees=45
  swing1Motion=Free
  swing2LimitDegrees=45
  swing2Motion=Free
  twistLimitDegrees=45
  twistMotion=Free
 }
 {
  bone1Name=dyn_ornamentArm_A_01_Base_r
  bone2Name=spine_05
  swing1LimitDegrees=45
  swing1Motion=Free
  swing2LimitDegrees=45
  swing2Motion=Free
  twistLimitDegrees=45
  twistMotion=Free
 }
 {bone1Name=dyn_ornamentArm_A_01_r, bone2Name=dyn_ornamentArm_A_01_Base_r, swing1LimitDegrees=16, swing2LimitDegrees=16, twistLimitDegrees=16}
 {bone1Name=dyn_ornamentArm_A_02_r, bone2Name=dyn_ornamentArm_A_01_r, swing1LimitDegrees=24, swing2LimitDegrees=16, twistLimitDegrees=16}
 {
  bone1Name=dyn_ornamentArm_B_01_Base_r
  bone2Name=spine_05
  swing1LimitDegrees=45
  swing1Motion=Free
  swing2LimitDegrees=45
  swing2Motion=Free
  twistLimitDegrees=45
  twistMotion=Free
 }
 {bone1Name=dyn_ornamentArm_B_01_r, bone2Name=dyn_ornamentArm_B_01_Base_r, swing1LimitDegrees=16, swing2LimitDegrees=16, twistLimitDegrees=16}
 {bone1Name=dyn_ornamentArm_B_02_r, bone2Name=dyn_ornamentArm_B_01_r, swing1LimitDegrees=16, swing2LimitDegrees=16, twistLimitDegrees=16}
 {
  bone1Name=upperarm_r
  bone2Name=spine_05
  swing1LimitDegrees=45
  swing1Motion=Free
  swing2LimitDegrees=45
  swing2Motion=Free
  twistLimitDegrees=45
  twistMotion=Free
 }
 {
  bone1Name=dyn_upperBackOrnament_01_c_Base
  bone2Name=spine_05
  swing1LimitDegrees=45
  swing1Motion=Free
  swing2LimitDegrees=45
  swing2Motion=Free
  twistLimitDegrees=45
  twistMotion=Free
 }
 {
  bone1Name=dyn_upperBackOrnament_01_c
  bone2Name=dyn_upperBackOrnament_01_c_Base
  swing1LimitDegrees=16
  swing2LimitDegrees=16
  twistLimitDegrees=16
 }
 {
  bone1Name=dyn_upperBackOrnament_02_c
  bone2Name=dyn_upperBackOrnament_01_c
  swing1LimitDegrees=24
  swing2LimitDegrees=16
  twistLimitDegrees=16
 }
 {bone1Name=dyn_ornamentArm_01_l, bone2Name=spine_05, swing1LimitDegrees=16, swing2LimitDegrees=16, twistLimitDegrees=16}
 {bone1Name=dyn_ornamentArm_02_l, bone2Name=dyn_ornamentArm_01_l, swing1LimitDegrees=16, swing2LimitDegrees=16, twistLimitDegrees=16}
 {
  bone1Name=upperarm_l
  bone2Name=spine_05
  swing1LimitDegrees=45
  swing1Motion=Free
  swing2LimitDegrees=45
  swing2Motion=Free
  twistLimitDegrees=45
  twistMotion=Free
 }
]
