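# Animation Blueprint for the Ornament Soldier male body cosmetic, adding secondary motion to its ornaments, fringe and tassels.
# Copied parent pose runs through Control Rig, chained AnimDynamics, leg/front RigidBody sims and a bone offset into a cached pose.
# wearingBackbling, set on customization load, selects between two back RigidBody branches.
# Event updates copy blink curves to part and emote-prop materials and hide a material element when backbling is worn.
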AnimBlueprint /BRCosmetics/Characters/Player/Male/Medium/Bodies/M_MED_Ornament_Soldier/Meshes/M_MED_Ornament_Soldier_AnimBP
target_skeleton=&/M_MED_Ornament_Soldier_Skeleton
graphs [
 AnimGraph {nodes [
   RigidBody() {
    bClampLinearTranslationLimitToRefPose=true
    bOverrideWorldGravity=true
    overridePhysicsAsset=/Game/Characters/Player/Male/Medium/Bodies/M_MED_Ornament_Soldier/Meshes/M_MED_Ornament_Soldier_PhysicsOrnament
    simulationSpace=BaseBoneSpace
    alphaBoolBlend {blendOption=Linear}
    baseBoneRef {boneName=root}
    componentAppliedLinearAccClamp=<6000 3000 3000>
    componentLinearAccScale=<0.4 0.4 0.53>
    componentLinearVelScale=<0.2 0.2 0.3>
    overrideWorldGravity {z=-980}
   }
   n1 = RigidBody(ComponentPose=Local To Component(LocalPose=Use cached pose 'ctrlRigCache'())) {
    bClampLinearTranslationLimitToRefPose=true
    bOverrideWorldGravity=true
    overridePhysicsAsset=&/M_MED_Ornament_Soldier_PhysicsBack
    simulationSpace=BaseBoneSpace
    alphaBoolBlend {blendOption=Linear}
    baseBoneRef {boneName=root}
    componentAppliedLinearAccClamp=<6000 3000 3000>
    componentLinearAccScale=<0.4 0.4 0.53>
    componentLinearVelScale=<0.2 0.2 0.3>
    overrideWorldGravity {z=-980}
   }
   n2 = RigidBody(ComponentPose=Local To Component(LocalPose=Use cached pose 'ctrlRigCache'())) {
    bClampLinearTranslationLimitToRefPose=true
    bOverrideWorldGravity=true
    overridePhysicsAsset=&M_MED_Ornament_Soldier_PhysicsBack
    simulationSpace=BaseBoneSpace
    alphaBoolBlend {blendOption=Linear}
    baseBoneRef {boneName=root}
    componentAppliedLinearAccClamp=<6000 3000 3000>
    componentLinearAccScale=<0.4 0.4 0.53>
    componentLinearVelScale=<0.2 0.2 0.3>
    overrideWorldGravity {z=-980}
   }
   Output Pose(Result=Blend Poses by bool(
     BlendPose_0=Component To Local(ComponentPose=@n2)
     BlendPose_1=Component To Local(ComponentPose=@n1)
     bActiveValue=Get wearingBackbling() -> (wearingBackbling:bool)
    ) {blendType=Linear}) {layerGroup=None}
   n3 = Control Rig(Source=Copy Pose From Mesh() {bUseAttachedParent=true}) {
    alphaBoolBlend {blendOption=Linear}
    controlRigAssetReference {blueprintRigClass=&/M_MED_OrnamentSoldier_Body_CtrlRig.M_MED_OrnamentSoldier_Body_CtrlRig_C}
   }
   n4 = Anim Dynamics - Bone: dyn_ornamentArm_02_l(ComponentPose=Local To Component(LocalPose=@n3), Alpha=1, ExternalForce="0,0,0") {
    angularSpringConstant=10
    bGravityOverrideInSimSpace=true
    bUseGravityOverride=true
    alphaBoolBlend {blendOption=Linear}
    boundBone {boneName=dyn_ornamentArm_02_l}
    componentLinearAccScale=<0.3 0.3 0.45>
    componentLinearVelScale=<0.3 0.3 0.45>
    gravityOverride {z=-980}
    physicsBodyDefinitions [{
      sphereCollisionRadius=0
      boundBone {boneName=dyn_ornamentArm_02_l}
      boxExtents=<8 20 20>
      constraintSetup {angularLimitsMax=<2 5 30>, angularLimitsMin=<-5 -10 -110>, angularTarget=<0 0 0>}
      localJointOffset=<8 0 0>
     }]
   }
   n5 = Anim Dynamics - Bone: dyn_ornamentArm_A_02_r(ComponentPose=@n4, Alpha=1, ExternalForce="0,0,0") {
    angularSpringConstant=10
    bGravityOverrideInSimSpace=true
    bUseGravityOverride=true
    alphaBoolBlend {blendOption=Linear}
    boundBone {boneName=dyn_ornamentArm_A_02_r}
    componentLinearAccScale=<0.3 0.3 0.45>
    componentLinearVelScale=<0.3 0.3 0.45>
    gravityOverride {z=-980}
    physicsBodyDefinitions [{
      sphereCollisionRadius=0
      boundBone {boneName=dyn_ornamentArm_A_02_r}
      boxExtents=<8 20 20>
      constraintSetup {angularLimitsMax=<5 5 50>, angularLimitsMin=<-5 -10 -90>, angularTarget=<0 0 0>}
      localJointOffset=<8 0 0>
     }]
   }
   n6 = Anim Dynamics - Bone: dyn_ornamentArm_B_02_r(ComponentPose=@n5, Alpha=1, ExternalForce="0,0,0") {
    angularSpringConstant=10
    bGravityOverrideInSimSpace=true
    bUseGravityOverride=true
    alphaBoolBlend {blendOption=Linear}
    boundBone {boneName=dyn_ornamentArm_B_02_r}
    componentLinearAccScale=<0.3 0.3 0.45>
    componentLinearVelScale=<0.3 0.3 0.45>
    gravityOverride {z=-980}
    physicsBodyDefinitions [{
      sphereCollisionRadius=0
      boundBone {boneName=dyn_ornamentArm_B_02_r}
      boxExtents=<8 20 20>
      constraintSetup {angularLimitsMax=<5 5 20>, angularLimitsMin=<-5 -2 -120>, angularTarget=<0 0 0>}
      localJointOffset=<8 0 0>
     }]
   }
   n7 = Anim Dynamics - Bone: dyn_upperBackOrnament_01_c(ComponentPose=@n6, Alpha=1, ExternalForce="0,0,0") {
    angularSpringConstant=10
    bGravityOverrideInSimSpace=true
    bUseGravityOverride=true
    alphaBoolBlend {blendOption=Linear}
    boundBone {boneName=dyn_upperBackOrnament_01_c}
    componentLinearAccScale=<0.3 0.3 0.45>
    componentLinearVelScale=<0.3 0.3 0.45>
    gravityOverride {z=-980}
    physicsBodyDefinitions [{
      sphereCollisionRadius=0
      boundBone {boneName=dyn_upperBackOrnament_01_c}
      boxExtents=<8 20 20>
      constraintSetup {angularLimitsMax=<5 5 30>, angularLimitsMin=<-5 -5 -45>, angularTarget=<0 0 0>}
      localJointOffset=<8 0 0>
     }]
   }
   n8 = Anim Dynamics - Bone: dyn_fringeTassleFront_C_02_l(ComponentPose=@n7, Alpha=1, ExternalForce="0,0,0") {
    angularSpringConstant=10
    bGravityOverrideInSimSpace=true
    bUseGravityOverride=true
    alphaBoolBlend {blendOption=Linear}
    boundBone {boneName=dyn_fringeTassleFront_C_02_l}
    componentLinearAccScale=<0.3 0.3 0.45>
    componentLinearVelScale=<0.3 0.3 0.45>
    gravityOverride {z=-980}
    physicsBodyDefinitions [{
      sphereCollisionRadius=0
      boundBone {boneName=dyn_fringeTassleFront_C_02_l}
      boxExtents=<10 20 20>
      constraintSetup {angularLimitsMax=<10 12 90>, angularLimitsMin=<-10 -12 -10>, angularTarget=<0 0 0>}
      localJointOffset=<5 0 0>
     }]
   }
   n9 = Anim Dynamics - Bone: dyn_rearOrnament_02(ComponentPose=@n8, Alpha=1, ExternalForce="0,0,0") {
    angularSpringConstant=50
    bAngularSpring=true
    bGravityOverrideInSimSpace=true
    bUseGravityOverride=true
    alphaBoolBlend {blendOption=Linear}
    boundBone {boneName=dyn_rearOrnament_02}
    componentLinearAccScale=<0.3 0.3 0.45>
    componentLinearVelScale=<0.3 0.3 0.45>
    gravityOverride {z=-980}
    physicsBodyDefinitions [{
      sphereCollisionRadius=0
      boundBone {boneName=dyn_rearOrnament_02}
      boxExtents=<8 20 20>
      constraintSetup {angularLimitsMax=<5 5 20>, angularLimitsMin=<-5 -10 -95>, angularTarget=<0 0 0>}
      localJointOffset=<8 0 0>
     }]
   }
   n10 = Anim Dynamics - Bone: dyn_legLow_BB_02_l(ComponentPose=@n9, Alpha=1, ExternalForce="0,0,0") {
    angularSpringConstant=50
    bAngularSpring=true
    bGravityOverrideInSimSpace=true
    bUseGravityOverride=true
    alphaBoolBlend {blendOption=Linear}
    boundBone {boneName=dyn_legLow_BB_02_l}
    componentLinearAccScale=<0.3 0.3 0.45>
    componentLinearVelScale=<0.3 0.3 0.45>
    gravityOverride {z=-980}
    physicsBodyDefinitions [{
      sphereCollisionRadius=0
      boundBone {boneName=dyn_legLow_BB_02_l}
      boxExtents=<10 20 20>
      constraintSetup {angularLimitsMax=<5 5 30>, angularLimitsMin=<-5 -10 -65>, angularTarget=<0 0 0>}
      localJointOffset=<8 0 0>
     }]
   }
   n11 = Anim Dynamics - Bone: dyn_legHi_P_01_l(ComponentPose=@n10, Alpha=1, ExternalForce="0,0,0") {
    angularSpringConstant=50
    bAngularSpring=true
    bGravityOverrideInSimSpace=true
    bUseGravityOverride=true
    alphaBoolBlend {blendOption=Linear}
    boundBone {boneName=dyn_legHi_P_01_l}
    componentLinearAccScale=<0.3 0.3 0.45>
    componentLinearVelScale=<0.3 0.3 0.45>
    gravityOverride {z=-980}
    physicsBodyDefinitions [{
      sphereCollisionRadius=0
      boundBone {boneName=dyn_legHi_P_01_l}
      boxExtents=<10 20 20>
      constraintSetup {angularLimitsMax=<5 2 85>, angularLimitsMin=<-5 -10 -35>, angularTarget=<0 0 0>}
      localJointOffset=<8 0 0>
     }]
   }
   n12 = Anim Dynamics - Bone: dyn_legHi_C_02_r(ComponentPose=@n11, Alpha=1, ExternalForce="0,0,0") {
    angularSpringConstant=50
    bAngularSpring=true
    bGravityOverrideInSimSpace=true
    bUseGravityOverride=true
    alphaBoolBlend {blendOption=Linear}
    boundBone {boneName=dyn_legHi_C_02_r}
    componentLinearAccScale=<0.3 0.3 0.45>
    componentLinearVelScale=<0.3 0.3 0.45>
    gravityOverride {z=-980}
    physicsBodyDefinitions [{
      sphereCollisionRadius=0
      boundBone {boneName=dyn_legHi_C_02_r}
      boxExtents=<10 20 20>
      constraintSetup {angularLimitsMax=<5 5 40>, angularLimitsMin=<-5 -10 -45>, angularTarget=<0 0 0>}
      localJointOffset=<7 0 0>
     }]
   }
   n13 = Anim Dynamics - Bone: dyn_legHi_H_02_r(ComponentPose=@n12, Alpha=1, ExternalForce="0,0,0") {
    angularSpringConstant=50
    bAngularSpring=true
    bGravityOverrideInSimSpace=true
    bUseGravityOverride=true
    alphaBoolBlend {blendOption=Linear}
    boundBone {boneName=dyn_legHi_H_02_r}
    componentLinearAccScale=<0.3 0.3 0.45>
    componentLinearVelScale=<0.3 0.3 0.45>
    gravityOverride {z=-980}
    physicsBodyDefinitions [{
      sphereCollisionRadius=0
      boundBone {boneName=dyn_legHi_H_02_r}
      boxExtents=<10 20 20>
      constraintSetup {angularLimitsMax=<5 2 25>, angularLimitsMin=<-5 -10 -90>, angularTarget=<0 0 0>}
      localJointOffset=<8 0 0>
     }]
   }
   n14 = Anim Dynamics - Bone: dyn_legLow_M_02_r(ComponentPose=@n13, Alpha=1, ExternalForce="0,0,0") {
    angularSpringConstant=50
    bAngularSpring=true
    bGravityOverrideInSimSpace=true
    bUseGravityOverride=true
    alphaBoolBlend {blendOption=Linear}
    boundBone {boneName=dyn_legLow_M_02_r}
    componentLinearAccScale=<0.3 0.3 0.45>
    componentLinearVelScale=<0.3 0.3 0.45>
    gravityOverride {z=-980}
    physicsBodyDefinitions [{
      sphereCollisionRadius=0
      boundBone {boneName=dyn_legLow_M_02_r}
      boxExtents=<10 20 20>
      constraintSetup {angularLimitsMax=<15 30 90>, angularLimitsMin=<-15 -30 -20>, angularTarget=<0 0 0>}
      localJointOffset=<8 0 0>
     }]
   }
   n15 = Anim Dynamics - Bone: dyn_legHi_M_01_r(ComponentPose=@n14, Alpha=1, ExternalForce="0,0,0") {
    angularSpringConstant=50
    bAngularSpring=true
    bGravityOverrideInSimSpace=true
    bUseGravityOverride=true
    alphaBoolBlend {blendOption=Linear}
    boundBone {boneName=dyn_legHi_M_01_r}
    componentLinearAccScale=<0.3 0.3 0.45>
    componentLinearVelScale=<0.3 0.3 0.45>
    gravityOverride {z=-980}
    physicsBodyDefinitions [{
      sphereCollisionRadius=0
      boundBone {boneName=dyn_legHi_M_01_r}
      boxExtents=<10 20 20>
      constraintSetup {angularLimitsMax=<2 2 3>, angularLimitsMin=<-4 -10 -90>, angularTarget=<0 0 0>}
      localJointOffset=<8 0 0>
     }]
   }
   n16 = RigidBody(ComponentPose=@n15) {
    bClampLinearTranslationLimitToRefPose=true
    bOverrideWorldGravity=true
    overridePhysicsAsset=&/M_MED_Ornament_Soldier_PhysicsRLeg
    simulationSpace=BaseBoneSpace
    alphaBoolBlend {blendOption=Linear}
    baseBoneRef {boneName=root}
    componentAppliedLinearAccClamp=<6000 3000 3000>
    componentLinearAccScale=<0.4 0.4 0.53>
    componentLinearVelScale=<0.2 0.2 0.3>
    overrideWorldGravity {z=-980}
   }
   n17 = RigidBody(ComponentPose=@n16) {
    bClampLinearTranslationLimitToRefPose=true
    bOverrideWorldGravity=true
    overridePhysicsAsset=&/M_MED_Ornament_Soldier_PhysicsLLeg
    simulationSpace=BaseBoneSpace
    alphaBoolBlend {blendOption=Linear}
    baseBoneRef {boneName=root}
    componentAppliedLinearAccClamp=<6000 3000 3000>
    componentLinearAccScale=<0.4 0.4 0.53>
    componentLinearVelScale=<0.2 0.2 0.3>
    overrideWorldGravity {z=-980}
   }
   n18 = RigidBody(ComponentPose=@n17) {
    bClampLinearTranslationLimitToRefPose=true
    bOverrideWorldGravity=true
    overridePhysicsAsset=&/M_MED_Ornament_Soldier_PhysicsFront
    simulationSpace=BaseBoneSpace
    alphaBoolBlend {blendOption=Linear}
    baseBoneRef {boneName=root}
    componentAppliedLinearAccClamp=<6000 3000 3000>
    componentLinearAccScale=<0.4 0.4 0.53>
    componentLinearVelScale=<0.2 0.2 0.3>
    overrideWorldGravity {z=-980}
   }
   n19 = "Transform (Modify) Bone - Bone: dyn_fringeArm_G_01_r"!(ComponentPose=@n18, Alpha=1, Translation="0,-2,-2") -> (Pose) {
    translationMode=BMM_Additive
    translationSpace=BCS_BoneSpace
    alphaBoolBlend {blendOption=Linear}
    alphaScaleBiasClamp {bInterpResult=true, interpSpeedDecreasing=5, interpSpeedIncreasing=5}
    boneToModify {boneName=dyn_fringeArm_G_01_r}
    translation {y=-2, z=-2}
   }
   Save cached pose 'ctrlRigCache'(Pose=Component To Local(ComponentPose=@n19)) {cacheName=ctrlRigCache}
  ]}
 EventGraph {nodes [
   Event BlueprintUpdateAnimation()
   n1 = TryGetFortPawn()
   n2 = GetOwnerSkeletalMeshComponent()
   > Branch(Condition=AND Boolean(A=IsValid(Object=@n2), B=IsValid(Object=@n1)))
   > Is Valid(InputObject=GetAnimInstance(self=@n2))
   > n3 = Sequence!(execute) -> (then_0, then_1, then_2:Exec)
   > Is Valid(InputObject=GetPartSkeletalMeshComponent())
   n4 = GetPartSkeletalMeshComponent()
   n5 = GetAnimInstance(self=GetOwnerSkeletalMeshComponent())
   > SetScalarParameterValueOnMaterials!(
    execute
    self:Mesh Component Object Reference=@n4
    ParameterName:Name=UseScalarParameterForBlink
    ParameterValue:float=GetCurveValue(self=@n5, CurveName=UseScalarParameterForBlink)
   ) -> (then)
   n6 = GetAnimInstance(self=GetOwnerSkeletalMeshComponent())
   > SetScalarParameterValueOnMaterials(self=@n4, ParameterName=BlinkControl, ParameterValue=GetCurveValue(self=@n6, CurveName=BlinkControl))
   n7 = TryGetFortPawn()
   n8 = Break Emote Mesh Comp Data!(EmoteMeshCompData:"Emote Mesh Comp Data Structure (by ref)"=Find(
     TargetMap:Map of Integers to Emote Mesh Comp Data Structures=Get EmoteMeshCompDatas!(self:Fort Pawn Object Reference=@n7) -> (EmoteMeshCompDatas:Map of Integers to Emote Mesh Comp Data Structures)
     Key:"Integer (by ref)"
    ) -> (Value:Emote Mesh Comp Data Structure)) -> (MeshComp:Mesh Component Object Reference, PropExclusionTag:Gameplay Tag Structure, bTickEvenWhenNotRendered:bool=~) {is=title K2Node_BreakStruct}
   Is Valid(exec=@n3.then_1, InputObject=@n8)
   > SetScalarParameterValueOnMaterials(self=@n8, ParameterName=UseScalarParameterForBlink, ParameterValue=GetCurveValue(self=@n5, CurveName=UseScalarParameterForBlinkProp))
   > SetScalarParameterValueOnMaterials(self=@n8, ParameterName=BlinkControl, ParameterValue=GetCurveValue(self=@n6, CurveName=BlinkControlProp))
   Branch(execute=@n3.then_2, Condition=Get wearingBackbling())
   > n9 = Get Dynamic Material Instance()
   > SetScalarParameterValue(self=@n9, ParameterName=Hide Element 05, Value=1)
   Event OnCharacterCustomizationLoaded()
   > Set wearingBackbling(wearingBackbling:bool=CheckForBackbling()) -> (Output_Get:bool)
  ]}
]
variables [wearingBackbling {type=boolean}]
