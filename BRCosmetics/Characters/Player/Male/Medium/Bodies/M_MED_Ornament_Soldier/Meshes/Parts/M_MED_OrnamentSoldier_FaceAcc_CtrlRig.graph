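# Control Rig for the face accessory of the male Ornament Soldier character, posing three dynamic head bone chains (headLowA, B and C) as the head and right arm move.
# Head and right upper-arm rotations relative to spine_05 are each reduced to one Euler axis, remapped and smoothed with time-integrated lerp accumulators.
# These weight additive local offsets on the chains' base bones: head alone drives B and C, while head multiplied by arm drives A.
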ControlRigBlueprint /BRCosmetics/Characters/Player/Male/Medium/Bodies/M_MED_Ornament_Soldier/Meshes/Parts/M_MED_OrnamentSoldier_FaceAcc_CtrlRig
graphs [
 RigVMFunctionLibrary {links [], nodes []}
 RigVMModel {nodes [
   n1 = &RigUnit_BeginExecution()
   n2 = &RigUnit_GetRelativeBoneTransform(Bone=head, Space=spine_05)
   n3 = /Script/RigVM.RigVMFunction_AccumulateFloatLerp(
    Blend=10
    InitialValue=0
    bIntegrateDeltaTime=true
    TargetValue=&RigVMFunction_MathFloatRemap(
     SourceMaximum=-45
     SourceMinimum=5
     TargetMinimum=0
     bClamp=false
     Value=&RigVMFunction_MathQuaternionToEuler(RotationOrder=ZYX, Value=@n2.Transform.Rotation).Result.Z
    ).Result
   ) {is=struct RigVMUnitNode, template="AccumulateLerp::Execute(in TargetValue,in InitialValue,in Blend,in bIntegrateDeltaTime,out Result)"}
   n4 = &RigVMFunction_MathFloatRemap(
    SourceMaximum=-100
    SourceMinimum=-40
    TargetMinimum=0
    bClamp=false
    Value=&RigVMFunction_MathQuaternionToEuler(RotationOrder=YZX, Value=&RigUnit_GetRelativeBoneTransform(Bone=upperarm_r, Space=spine_05).Transform.Rotation).Result.Y
   )
   /Script/ControlRig.RigUnit_ModifyBoneTransforms(
    BoneToModify=((Bone="dyn_headLowA_01_Base",Transform=(Rotation=(X=0,Y=0,Z=0,W=1),Translation=(X=-1,Y=0,Z=0),Scale3D=(X=1,Y=1,Z=1))))
    Mode=AdditiveLocal
    WeightMaximum=1
    WeightMinimum=0
    ExecutePin=@n1.ExecutePin
    Weight=&RigVMFunction_MathFloatMul(
     A=@n3.Result
     B=&RigVMFunction_AccumulateFloatLerp(Blend=10, InitialValue=0, bIntegrateDeltaTime=true, TargetValue=@n4.Result) {template="AccumulateLerp::Execute(in TargetValue,in InitialValue,in Blend,in bIntegrateDeltaTime,out Result)"}.Result
    ) {template="Multiply::Execute(in A,in B,out Result)"}.Result
    BoneToModify.0.Transform.Rotation=&RigVMFunction_MathQuaternionFromEuler(Euler=(X=-5,Y=1,Z=-12)).Result
   ) {is=struct RigVMUnitNode}
   > &RigUnit_ModifyBoneTransforms(
    BoneToModify=((Bone="dyn_headLowB_01_Base",Transform=(Rotation=(X=0,Y=0,Z=0,W=1),Translation=(X=0,Y=2,Z=0),Scale3D=(X=1,Y=1,Z=1))))
    Mode=AdditiveLocal
    WeightMaximum=1
    WeightMinimum=0
    Weight=@n3.Result
    BoneToModify.0.Transform.Rotation=&RigVMFunction_MathQuaternionFromEuler(Euler=(X=0,Y=0,Z=-45)).Result
   )
   > &RigUnit_ModifyBoneTransforms(
    BoneToModify=((Bone="dyn_headLowC_01_Base",Transform=(Rotation=(X=0,Y=0,Z=0,W=1),Translation=(X=0,Y=0,Z=0),Scale3D=(X=1,Y=1,Z=1))))
    Mode=AdditiveLocal
    WeightMaximum=1
    WeightMinimum=0
    Weight=@n3.Result
    BoneToModify.0.Transform.Rotation=&RigVMFunction_MathQuaternionFromEuler(Euler=(X=-5,Y=10,Z=-45)).Result
   )
  ]}
]
hierarchy [
 root {loc=<0 0 0>, rot=<0 0 0 1>}
 pelvis {parent=root, loc=<0 -0.7071 93.1191>, rot=<0.0127 0.707 -0.0127 -0.707>}
 spine_01 {parent=pelvis, loc=<6.7494 -0.6388 0>, rot=<0 0 0.034 -0.9994>}
 spine_02 {parent=spine_01, loc=<7.2518 -0.035 0>, rot=<0 0 0.0238 -0.9997>}
 spine_03 {parent=spine_02, loc=<6.0104 0 0>, rot=<0 0 0.0149 -0.9999>}
 spine_04 {parent=spine_03, loc=<8.1024 0 0>, rot=<0 0 -0.0132 -0.9999>}
 spine_05 {parent=spine_04, loc=<10.0146 0 0>, rot=<0 0 -0.0015 -1>}
 neck_01 {parent=spine_05, loc=<11.3138 0.1529 0>, rot=<0 0 -0.2003 -0.9797>}
 neck_02 {parent=neck_01, loc=<4.5627 0 0>, rot=<0 0 0.0504 -0.9987>}
 head {parent=neck_02, loc=<5.3914 0 0>, rot=<0 0 0.115 -0.9934>}
 faceAttach {parent=head, loc=<-0.8266 -0.1832 0>, rot=<0.0044 -0.3827 -0.0044 -0.9239>, scale=<0.89 0.89 0.89>}
 dyn_headLowA_01_Base {parent=head, loc=<2.6942 6.6097 7.3764>, rot=<0.0995 -0.4102 0.9055 0.0431>}
 dyn_headLowA_01 {parent=dyn_headLowA_01_Base, loc=<0 0 0>, rot=<0 0 0 1>}
 dyn_headLowA_02 {parent=dyn_headLowA_01, loc=<4.4747 0 0>, rot=<0.0109 -0.0705 0.2147 -0.9741>}
 dyn_headLowA_03 {parent=dyn_headLowA_02, loc=<6.7951 0 0>, rot=<0 0 0 -1>}
 dyn_headLowB_01_Base {parent=head, loc=<4.8121 0.482 9.1683>, rot=<0.1924 -0.6994 0.6769 0.1251>}
 dyn_headLowB_01 {parent=dyn_headLowB_01_Base, loc=<0 0 0>, rot=<0 0 0 -1>}
 dyn_headLowB_02 {parent=dyn_headLowB_01, loc=<5.6744 0 0>, rot=<-0.0581 -0.0095 0.2106 -0.9758>}
 dyn_headLowB_03 {parent=dyn_headLowB_02, loc=<7.4311 0 0>, rot=<0.0872 0 0 -0.9962>}
 dyn_headLowC_01_Base {parent=head, loc=<4.258 -6.0318 6.7485>, rot=<0.2562 -0.9231 0.2852 0.0305>}
 dyn_headLowC_01 {parent=dyn_headLowC_01_Base, loc=<0 0 0>, rot=<0 0 0 1>}
 dyn_headLowC_02 {parent=dyn_headLowC_01, loc=<4.8473 0 0>, rot=<-0.0179 -0.0323 0.0496 -0.9981>}
 dyn_headLowC_03 {parent=dyn_headLowC_02, loc=<4.6888 0 0>, rot=<0.0436 0 0 -0.999>}
 dyn_headLowD_01_Base {parent=head, loc=<4.1238 -8.9535 0.3239>, rot=<0.3317 -0.9434 0.0026 0.0009>}
 dyn_headLowD_01 {parent=dyn_headLowD_01_Base, loc=<0 0 0>, rot=<0 0 0 -1>}
 dyn_headLowD_02 {parent=dyn_headLowD_01, loc=<4.8841 0 0>, rot=<0.0574 0.0616 0.1515 -0.9849>}
 dyn_headLowD_03 {parent=dyn_headLowD_02, loc=<5.1105 0 0>, rot=<0 0 0 1>}
 dyn_headLowE_01_Base {parent=head, loc=<2.357 -7.3088 -6.8451>, rot=<0.3549 -0.9011 -0.2326 -0.0894>}
 dyn_headLowE_01 {parent=dyn_headLowE_01_Base, loc=<0 0 0>, rot=<0 0 0 -1>}
 dyn_headLowE_02 {parent=dyn_headLowE_01, loc=<3.6158 0 0>, rot=<0.0658 0.0794 0.1357 -0.9854>}
 dyn_headLowE_03 {parent=dyn_headLowE_02, loc=<5.4822 0 0>, rot=<0 0 0 -1>}
 dyn_headLowF_01_Base {parent=head, loc=<2.2068 -1.1348 -8.7143>, rot=<0.2635 -0.7721 -0.5578 -0.1527>}
 dyn_headLowF_01 {parent=dyn_headLowF_01_Base, loc=<0 0 0>, rot=<0 0 0 -1>}
 dyn_headLowF_02 {parent=dyn_headLowF_01, loc=<3.9999 0 0>, rot=<-0.0269 -0.0245 0.0409 -0.9985>}
 dyn_headLowG_01_Base {parent=head, loc=<2.8353 1.1693 -10.2202>, rot=<0.2799 -0.6671 -0.6388 -0.2619>}
 dyn_headLowG_01 {parent=dyn_headLowG_01_Base, loc=<0 0 0>, rot=<0 0 0 -1>}
 dyn_headLowG_02 {parent=dyn_headLowG_01, loc=<7.1196 0 0>, rot=<0 0 0 -1>}
 dyn_headLowH_01_Base {parent=head, loc=<2.6489 6.5516 -9.41>, rot=<0.0959 0.3444 0.9338 0.0137>}
 dyn_headLowH_01 {parent=dyn_headLowH_01_Base, loc=<0 0 0>, rot=<0 0 0 1>}
 dyn_headLowH_02 {parent=dyn_headLowH_01, loc=<5.1153 0 0>, rot=<-0.0002 -0.1006 0.286 -0.9529>}
 dyn_headLowH_03 {parent=dyn_headLowH_02, loc=<4.0061 0 0>, rot=<0 0 0 -1>}
 dyn_headMidA_01_Base {parent=head, loc=<9.5685 11.5343 1.3543>, rot=<0.0184 -0.0674 0.9645 0.2547>}
 dyn_headMidA_01 {parent=dyn_headMidA_01_Base, loc=<0 0 0>, rot=<0 0 0 1>}
 dyn_headMidB_01_Base {parent=head, loc=<9.7549 9.176 6.6111>, rot=<0.0436 -0.2823 0.8709 0.4>}
 dyn_headMidB_01 {parent=dyn_headMidB_01_Base, loc=<0 0 0>, rot=<0 0 0 1>}
 dyn_headMidC_01_Base {parent=head, loc=<9.9242 6.434 8.1593>, rot=<0.2515 -0.5176 0.7385 0.3514>}
 dyn_headMidC_01 {parent=dyn_headMidC_01_Base, loc=<0 0 0>, rot=<0 0 0 1>}
 dyn_headMidD_01_Base {parent=head, loc=<5.0503 5.4069 10.2816>, rot=<0.2931 -0.5224 0.7016 0.386>}
 dyn_headMidD_01 {parent=dyn_headMidD_01_Base, loc=<0 0 0>, rot=<0 0 0 1>}
 dyn_headMidE_01_Base {parent=head, loc=<6.7104 1.8839 10.8461>, rot=<0.2974 -0.6098 0.6629 0.3166>}
 dyn_headMidE_01 {parent=dyn_headMidE_01_Base, loc=<0 0 0>, rot=<0 0 0 1>}
 dyn_headMidF_01_Base {parent=head, loc=<7.9332 2.9303 9.0161>, rot=<0.2147 -0.6366 0.7038 0.2309>}
 dyn_headMidF_01 {parent=dyn_headMidF_01_Base, loc=<0 0 0>, rot=<0 0 0 -1>}
 dyn_headMidG_01_Base {parent=head, loc=<3.8752 -2.4627 11.3042>, rot=<0.3508 -0.6977 0.5333 0.3253>}
 dyn_headMidG_01 {parent=dyn_headMidG_01_Base, loc=<0 0 0>, rot=<0 0 0 -1>}
 dyn_headMidH_01_Base {parent=head, loc=<9.2574 -3.2956 8.0083>, rot=<0.3562 -0.7583 0.496 0.228>}
 dyn_headMidH_01 {parent=dyn_headMidH_01_Base, loc=<0 0 0>, rot=<0 0 0 1>}
 dyn_headMidJ_01_Base {parent=head, loc=<4.3188 -7.5931 7.9323>, rot=<0.4362 -0.8259 0.3173 0.1643>}
 dyn_headMidJ_01 {parent=dyn_headMidJ_01_Base, loc=<0 0 0>, rot=<0 0 0 1>}
 dyn_headMidK_01_Base {parent=head, loc=<10.9147 -5.7424 4.0893>, rot=<0.3717 -0.9095 0.173 0.0691>}
 dyn_headMidK_01 {parent=dyn_headMidK_01_Base, loc=<0 0 0>, rot=<0 0 0 -1>}
 dyn_headMidL_01_Base {parent=head, loc=<5.3949 -8.7586 2.9296>, rot=<0.4092 -0.9122 0.017 0.0075>}
 dyn_headMidL_01 {parent=dyn_headMidL_01_Base, loc=<0 0 0>, rot=<0 0 0 1>}
 dyn_headMidM_01_Base {parent=head, loc=<11.2598 -6.6492 -1.0632>, rot=<0.357 -0.933 -0.0168 0.0432>}
 dyn_headMidM_01 {parent=dyn_headMidM_01_Base, loc=<0 0 0>, rot=<0 0 0 -1>}
 dyn_headMidN_01_Base {parent=head, loc=<11.0556 -5.7837 -4.1266>, rot=<0.3923 -0.9099 -0.1348 -0.0022>}
 dyn_headMidN_01 {parent=dyn_headMidN_01_Base, loc=<0 0 0>, rot=<0 0 0 1>}
 dyn_headMidP_01_Base {parent=head, loc=<5.0316 -7.3549 -5.6118>, rot=<0.3662 -0.9089 -0.1854 -0.0729>}
 dyn_headMidP_01 {parent=dyn_headMidP_01_Base, loc=<0 0 0>, rot=<0 0 0 1>}
 dyn_headMidQ_01_Base {parent=head, loc=<10.3312 -3.9878 -7.1077>, rot=<0.4577 -0.729 -0.4333 -0.267>}
 dyn_headMidQ_01 {parent=dyn_headMidQ_01_Base, loc=<0 0 0>, rot=<0 0 0 1>}
 dyn_headMidR_01_Base {parent=head, loc=<10.1455 -1.7956 -9.0568>, rot=<0.4215 -0.6776 -0.5144 -0.3141>}
 dyn_headMidR_01 {parent=dyn_headMidR_01_Base, loc=<0 0 0>, rot=<0 0 0 1>}
 dyn_headMidS_01_Base {parent=head, loc=<5.3046 -3.1783 -9.8037>, rot=<0.405 -0.6979 -0.5133 -0.2922>}
 dyn_headMidS_01 {parent=dyn_headMidS_01_Base, loc=<0 0 0>, rot=<0 0 0 1>}
 dyn_headMidT_01_Base {parent=head, loc=<9.6982 3.3997 -9.3982>, rot=<-0.2026 0.6226 0.7374 0.1659>}
 dyn_headMidT_01 {parent=dyn_headMidT_01_Base, loc=<0 0 0>, rot=<0 0 0 -1>}
 dyn_headMidT_02 {parent=dyn_headMidT_01, loc=<5.4968 0.1252 0.001>, rot=<-0.0346 0.1686 0.0011 -0.9851>}
 dyn_headMidT_03 {parent=dyn_headMidT_02, loc=<3.2429 0 0>, rot=<0.0997 -0.1099 0.0364 -0.9883>}
 dyn_headMidV_01_Base {parent=head, loc=<8.394 4.7017 -9.2924>, rot=<-0.2233 0.5594 0.7822 0.1593>}
 dyn_headMidV_01 {parent=dyn_headMidV_01_Base, loc=<0 0 0>, rot=<0 0 0 -1>}
 dyn_headMidV_02 {parent=dyn_headMidV_01, loc=<2.0871 0 1.7513>, rot=<-0.0671 0.0502 0.1805 -0.98>}
 dyn_headMidW_01_Base {parent=head, loc=<9.2194 9.2314 -5.45>, rot=<-0.359 0.4333 0.7983 0.2147>}
 dyn_headMidW_01 {parent=dyn_headMidW_01_Base, loc=<0 0 0>, rot=<0 0 0 -1>}
 dyn_headMidW_02 {parent=dyn_headMidW_01, loc=<2.6111 0 0>, rot=<-0.0837 0.2044 0.1269 -0.967>}
 dyn_headTopA_01 {parent=head, loc=<16.2687 9.5131 -1.5851>, rot=<-0.0339 0.0703 0.9017 0.4252>}
 dyn_headTopB_01 {parent=head, loc=<16.0363 8.5935 2.6079>, rot=<0.126 -0.2958 0.8742 0.3639>}
 dyn_headTopC_01 {parent=head, loc=<15.5787 -0.2908 6.453>, rot=<0.3284 -0.65 0.6142 0.304>}
 dyn_headTopD_01 {parent=head, loc=<15.6911 -4.3202 3.2016>, rot=<0.2465 -0.8537 0.3536 0.2923>}
 dyn_headTopE_01 {parent=head, loc=<15.6333 -5.297 0.7424>, rot=<0.4222 -0.9062 -0.0202 -0.0092>}
 dyn_headTopF_01 {parent=head, loc=<15.8701 -2.8156 -4.47>, rot=<0.4214 -0.8597 -0.2602 -0.1249>}
 dyn_headTopG_01 {parent=head, loc=<15.6411 3.3088 -6.3708>, rot=<-0.261 0.5372 0.7243 0.3446>}
 dyn_llamaStar_01 {parent=head, loc=<25.7376 1.8037 -1.3684>, rot=<0.9289 -0.2613 -0.0732 -0.2521>}
 dyn_llamaStar_02 {parent=dyn_llamaStar_01, loc=<5.5468 0 0>, rot=<0.9489 -0.1244 0.02 0.2894>}
 clavicle_l {parent=spine_05, loc=<4.3575 -0.1466 -2.3223>, rot=<0.0465 -0.7289 -0.009 -0.6829>}
 upperarm_l {parent=clavicle_l, loc=<15.5111 0 0>, rot=<0.0226 -0.3234 -0.0088 -0.9459>}
 clavicle_r {parent=spine_05, loc=<4.3575 -0.1466 2.3223>, rot=<0.7289 0.0465 -0.6829 0.009>}
 upperarm_r {parent=clavicle_r, loc=<-15.5111 0 0>, rot=<0.0226 -0.3234 -0.0088 -0.9459>}
]
