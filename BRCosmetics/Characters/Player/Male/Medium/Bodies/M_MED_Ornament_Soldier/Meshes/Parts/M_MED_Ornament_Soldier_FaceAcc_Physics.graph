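# Physics asset for the Ornament Soldier male face accessory, giving its dangling head bones and llama star secondary motion.
# A head capsule anchors box bodies in Low, Mid and Top groups, chained one to three bones deep.
# Low roots swing and twist freely, Mid roots lock twist, and child links carry angular limits.
# Top bodies and several chain tips use tiny, uniform placeholder-sized boxes rather than fitted shapes.
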PhysicsAsset /BRCosmetics/Characters/Player/Male/Medium/Bodies/M_MED_Ornament_Soldier/Meshes/Parts/M_MED_Ornament_Soldier_FaceAcc_Physics
bodies {
 dyn_headLowA_01 [{shapeType=Box, extentX=1.9285, extentY=10.2229, extentZ=11.3886, center=<0.7652 -1.01 -1.2746>, rotation=<rot 0 90 0>}]
 dyn_headLowA_02 [{shapeType=Box, extentX=2.0268, extentY=7.4846, extentZ=12.3605, center=<3.886 0.0665 -0.4497>, rotation=<rot 0 90 0>}]
 dyn_headLowB_01 [{shapeType=Box, extentX=1.9285, extentY=6.8468, extentZ=6.967, center=<3.0565 -0.155 0.1261>, rotation=<rot 0 90 0>}]
 dyn_headLowB_02 [{shapeType=Box, extentX=2.0268, extentY=8.4626, extentZ=12.552, center=<3.9371 1.0322 1.7451>, rotation=<rot 0 90 0>}]
 dyn_headLowC_01 [{shapeType=Box, extentX=1.9285, extentY=7.3988, extentZ=8.992, center=<2.5081 0.1066 -0.1596>, rotation=<rot 0 90 0>}]
 dyn_headLowC_02 [{shapeType=Box, extentX=2.0268, extentY=6.5459, extentZ=10.4983, center=<3.0652 -0.1682 0.2335>, rotation=<rot 0 90 0>}]
 dyn_headLowD_01 [{shapeType=Box, extentX=1.9285, extentY=7.4696, extentZ=9.3815, center=<1.2136 -0.0449 0.4241>, rotation=<rot 0 90 0>}]
 dyn_headLowD_02 [{shapeType=Box, extentX=2.0268, extentY=6.0069, extentZ=10.6114, center=<2.3191 0.4464 -0.2339>, rotation=<rot 0 90 0>}]
 dyn_headLowE_01 [{shapeType=Box, extentX=1.9285, extentY=6.0655, extentZ=8.3414, center=<0.5105 -0.1947 -0.5612>, rotation=<rot 0 90 0>}]
 dyn_headLowE_02 [{shapeType=Box, extentX=2.0268, extentY=6.7155, extentZ=10.1752, center=<2.3135 0.408 -0.0201>, rotation=<rot 0 90 0>}]
 dyn_headLowF_01 [{shapeType=Box, extentX=1.9285, extentY=12.7465, extentZ=11.9113, center=<5.3709 -0.7751 -0.6735>, rotation=<rot 0 90 0>}]
 dyn_headLowF_02 [{shapeType=Box, extentX=2.0268, extentY=1.01, extentZ=1.01, rotation=<rot 0 90 0>}]
 dyn_headLowG_01 [{shapeType=Box, extentX=1.9285, extentY=6.6102, extentZ=14.7247, center=<4.6721 -0.1318 1.6359>, rotation=<rot 0 90 0>}]
 dyn_headLowG_02 [{shapeType=Box, extentX=2.0268, extentY=1.01, extentZ=1.01, rotation=<rot 0 90 0>}]
 dyn_headLowH_01 [{shapeType=Box, extentX=1.9285, extentY=8.1813, extentZ=8.8102, center=<1.5755 -0.3809 0.0536>, rotation=<rot 0 90 0>}]
 dyn_headLowH_02 [{shapeType=Box, extentX=2.0268, extentY=5.514, extentZ=11.3694, center=<1.6627 0.7648 0.7972>, rotation=<rot 0 90 0>}]
 dyn_headMidA_01 [{shapeType=Box, extentX=1.7345, extentY=7.5781, extentZ=11.7327, center=<3.4136 0.2258 -0.5575>, rotation=<rot 0 -85.1405 -87.227>}]
 dyn_headMidB_01 [{shapeType=Box, extentX=1.7345, extentY=6.9919, extentZ=8.4001, center=<2.4798 0.8378 0.2383>, rotation=<rot 0 82.9853 -84.8494>}]
 dyn_headMidC_01 [{shapeType=Box, extentX=1.7345, extentY=3.7238, extentZ=9.969, center=<2.6235 0.1381 0.231>, rotation=<rot 0 -91.3988 -89.6378>}]
 dyn_headMidD_01 [{shapeType=Box, extentX=1.7345, extentY=5.9582, extentZ=9.8027, center=<1.7876 -0.0143 0.6707>, rotation=<rot 0 104.235 -73.9367>}]
 dyn_headMidE_01 [{shapeType=Box, extentX=1.7345, extentY=4.3404, extentZ=7.1579, center=<1.7372 -0.0167 0.3958>, rotation=<rot 0 -82.8056 -81.3479>}]
 dyn_headMidF_01 [{shapeType=Box, extentX=1.7345, extentY=1.7707, extentZ=2.7746, center=<-0.6137 -0.8834 0.2806>, rotation=<rot 0 -83.301 -64.1445>}]
 dyn_headMidG_01 [{shapeType=Box, extentX=1.7345, extentY=6.6017, extentZ=9.2642, center=<1.9653 0.2161 -0.4648>, rotation=<rot 0 -89.1953 -71.2876>}]
 dyn_headMidH_01 [{shapeType=Box, extentX=1.7345, extentY=7.4329, extentZ=9.1706, center=<2.2208 0.1202 -0.1718>, rotation=<rot 0 -88.7688 -79.1514>}]
 dyn_headMidJ_01 [{shapeType=Box, extentX=1.7345, extentY=5.5643, extentZ=9.3326, center=<1.5791 0.1927 0.6588>, rotation=<rot 0 100.8184 -86.8797>}]
 dyn_headMidK_01 [{shapeType=Box, extentX=1.7345, extentY=5.0367, extentZ=6.7071, center=<1.4149 -0.1388 0.881>, rotation=<rot 0 99.6745 -46.0277>}]
 dyn_headMidL_01 [{shapeType=Box, extentX=1.7345, extentY=7.3833, extentZ=9.1487, center=<1.0138 -0.48 -1.943>, rotation=<rot 0 -77.2301 -87.1505>}]
 dyn_headMidM_01 [{shapeType=Box, extentX=1.7345, extentY=5.3291, extentZ=7.2028, center=<2.683 0.2394 -0.0986>, rotation=<rot 0 86.2379 -72.8372>}]
 dyn_headMidN_01 [{shapeType=Box, extentX=1.7345, extentY=3.4295, extentZ=6.4128, center=<2.948 0.3906 -0.0645>, rotation=<rot 0 -101.3792 -72.094>}]
 dyn_headMidP_01 [{shapeType=Box, extentX=1.7345, extentY=3.6985, extentZ=8.7522, center=<1.3987 -0.2222 -1.8201>, rotation=<rot 0 -86.1484 -75.4939>}]
 dyn_headMidQ_01 [{shapeType=Box, extentX=1.7345, extentY=2.3674, extentZ=5.9252, center=<1.9022 0.1917 -0.6395>, rotation=<rot 0 80.7667 -66.9473>}]
 dyn_headMidR_01 [{shapeType=Box, extentX=1.7345, extentY=3.4018, extentZ=6.5672, center=<1.4595 0.2234 0.0471>, rotation=<rot 0 -93.7609 -85.2258>}]
 dyn_headMidS_01 [{shapeType=Box, extentX=1.8403, extentY=7.6547, extentZ=9.5424, center=<1.916 0.3895 0.7272>, rotation=<rot 0 85.7301 -85.4715>}]
 dyn_headMidT_01 [{shapeType=Box, extentX=1.7345, extentY=5.2859, extentZ=7.5273, center=<2.3441 0.4359 -0.4563>, rotation=<rot 0 -78.4021 -89.0557>}]
 dyn_headMidT_02 [{shapeType=Box, extentX=2.0851, extentY=4.2408, extentZ=2.9233, center=<1.6707 0 0>, rotation=<rot 0 90 0>}]
 dyn_headMidT_03 [{shapeType=Box, extentX=4.9337, extentY=7.4543, extentZ=5.9865, center=<4.736 0 -0.0001>, rotation=<rot 0 -90 0>}]
 dyn_headMidV_01 [{shapeType=Box, extentX=2.9293, extentY=3.6762, extentZ=9.9979, center=<3.115 1.0906 1.9241>, rotation=<rot 0 -105.397 -89.3252>}]
 dyn_headMidV_02 [{shapeType=Box, extentX=2.0268, extentY=1.01, extentZ=1.01, rotation=<rot 0 90 0>}]
 dyn_headMidW_01 [{shapeType=Box, extentX=2.3708, extentY=8.3064, extentZ=12.5169, center=<2.4874 0.6259 2.3955>, rotation=<rot 0 89.2058 -60.7131>}]
 dyn_headMidW_02 [{shapeType=Box, extentX=2.0268, extentY=1.01, extentZ=1.01, rotation=<rot 0 90 0>}]
 dyn_headTopA_01 [{shapeType=Box, extentX=1.01, extentY=1.01, extentZ=1.01, rotation=<rot 0 -90 0>}]
 dyn_headTopB_01 [{shapeType=Box, extentX=1.01, extentY=1.01, extentZ=1.01, rotation=<rot 0 -90 0>}]
 dyn_headTopC_01 [{shapeType=Box, extentX=1.01, extentY=1.01, extentZ=1.01, rotation=<rot 0 -90 0>}]
 dyn_headTopD_01 [{shapeType=Box, extentX=1.01, extentY=1.01, extentZ=1.01, rotation=<rot 0 -90 0>}]
 dyn_headTopE_01 [{shapeType=Box, extentX=1.01, extentY=1.01, extentZ=1.01, rotation=<rot 0 -90 0>}]
 dyn_headTopF_01 [{shapeType=Box, extentX=1.01, extentY=1.01, extentZ=1.01, rotation=<rot 0 -90 0>}]
 dyn_headTopG_01 [{shapeType=Box, extentX=1.01, extentY=1.01, extentZ=1.01, rotation=<rot 0 -90 0>}]
 dyn_llamaStar_01 [{shapeType=Box, extentX=3.0274, extentY=3.5301, extentZ=5.9395, center=<2.9579 -0.0821 0.445>, rotation=<rot 0 110.9928 -74.8588>}]
 dyn_llamaStar_02 [{shapeType=Box, extentX=2.0268, extentY=13.1056, extentZ=7.7724, center=<6.0839 -0.0181 1.482>, rotation=<rot 0 90 0>}]
 head [{shapeType=Capsule, length=12.7992, radius=9.4411, rotation=<rot 90 0 0>}]
}
constraints [
 {
  bone1Name=dyn_headLowA_01
  bone2Name=head
  swing1LimitDegrees=45
  swing1Motion=Free
  swing2LimitDegrees=45
  swing2Motion=Free
  twistLimitDegrees=45
  twistMotion=Free
 }
 {
  bone1Name=dyn_headLowB_01
  bone2Name=head
  swing1LimitDegrees=45
  swing1Motion=Free
  swing2LimitDegrees=45
  swing2Motion=Free
  twistLimitDegrees=45
  twistMotion=Free
 }
 {
  bone1Name=dyn_headLowC_01
  bone2Name=head
  swing1LimitDegrees=45
  swing1Motion=Free
  swing2LimitDegrees=45
  swing2Motion=Free
  twistLimitDegrees=45
  twistMotion=Free
 }
 {
  bone1Name=dyn_headLowD_01
  bone2Name=head
  swing1LimitDegrees=45
  swing1Motion=Free
  swing2LimitDegrees=45
  swing2Motion=Free
  twistLimitDegrees=45
  twistMotion=Free
 }
 {
  bone1Name=dyn_headLowE_01
  bone2Name=head
  swing1LimitDegrees=45
  swing1Motion=Free
  swing2LimitDegrees=45
  swing2Motion=Free
  twistLimitDegrees=45
  twistMotion=Free
 }
 {
  bone1Name=dyn_headLowF_01
  bone2Name=head
  swing1LimitDegrees=45
  swing1Motion=Free
  swing2LimitDegrees=45
  swing2Motion=Free
  twistLimitDegrees=45
  twistMotion=Free
 }
 {
  bone1Name=dyn_headLowG_01
  bone2Name=head
  swing1LimitDegrees=45
  swing1Motion=Free
  swing2LimitDegrees=45
  swing2Motion=Free
  twistLimitDegrees=45
  twistMotion=Free
 }
 {
  bone1Name=dyn_headLowH_01
  bone2Name=head
  swing1LimitDegrees=45
  swing1Motion=Free
  swing2LimitDegrees=45
  swing2Motion=Free
  twistLimitDegrees=45
  twistMotion=Free
 }
 {bone1Name=dyn_headMidA_01, bone2Name=head, swing1LimitDegrees=25, swing2LimitDegrees=16, twistLimitDegrees=45, twistMotion=Locked}
 {bone1Name=dyn_headMidB_01, bone2Name=head, swing1LimitDegrees=25, swing2LimitDegrees=16, twistLimitDegrees=45, twistMotion=Locked}
 {bone1Name=dyn_headMidC_01, bone2Name=head, swing1LimitDegrees=25, swing2LimitDegrees=16, twistLimitDegrees=45, twistMotion=Locked}
 {bone1Name=dyn_headMidD_01, bone2Name=head, swing1LimitDegrees=25, swing2LimitDegrees=16, twistLimitDegrees=45, twistMotion=Locked}
 {bone1Name=dyn_headMidE_01, bone2Name=head, swing1LimitDegrees=25, swing2LimitDegrees=16, twistLimitDegrees=45, twistMotion=Locked}
 {bone1Name=dyn_headMidF_01, bone2Name=head, swing1LimitDegrees=25, swing2LimitDegrees=16, twistLimitDegrees=45, twistMotion=Locked}
 {bone1Name=dyn_headMidG_01, bone2Name=head, swing1LimitDegrees=25, swing2LimitDegrees=16, twistLimitDegrees=45, twistMotion=Locked}
 {bone1Name=dyn_headMidH_01, bone2Name=head, swing1LimitDegrees=25, swing2LimitDegrees=16, twistLimitDegrees=45, twistMotion=Locked}
 {bone1Name=dyn_headMidJ_01, bone2Name=head, swing1LimitDegrees=25, swing2LimitDegrees=16, twistLimitDegrees=45, twistMotion=Locked}
 {bone1Name=dyn_headMidK_01, bone2Name=head, swing1LimitDegrees=25, swing2LimitDegrees=16, twistLimitDegrees=45, twistMotion=Locked}
 {bone1Name=dyn_headMidL_01, bone2Name=head, swing1LimitDegrees=25, swing2LimitDegrees=16, twistLimitDegrees=45, twistMotion=Locked}
 {bone1Name=dyn_headMidM_01, bone2Name=head, swing1LimitDegrees=25, swing2LimitDegrees=16, twistLimitDegrees=45, twistMotion=Locked}
 {bone1Name=dyn_headMidN_01, bone2Name=head, swing1LimitDegrees=25, swing2LimitDegrees=16, twistLimitDegrees=45, twistMotion=Locked}
 {bone1Name=dyn_headMidP_01, bone2Name=head, swing1LimitDegrees=25, swing2LimitDegrees=16, twistLimitDegrees=45, twistMotion=Locked}
 {bone1Name=dyn_headMidQ_01, bone2Name=head, swing1LimitDegrees=25, swing2LimitDegrees=16, twistLimitDegrees=45, twistMotion=Locked}
 {bone1Name=dyn_headMidR_01, bone2Name=head, swing1LimitDegrees=25, swing2LimitDegrees=16, twistLimitDegrees=45, twistMotion=Locked}
 {bone1Name=dyn_headMidS_01, bone2Name=head, swing1LimitDegrees=25, swing2LimitDegrees=16, twistLimitDegrees=45, twistMotion=Locked}
 {bone1Name=dyn_headMidT_01, bone2Name=head, swing1LimitDegrees=25, swing2LimitDegrees=16, twistLimitDegrees=45, twistMotion=Locked}
 {bone1Name=dyn_headMidV_01, bone2Name=head, swing1LimitDegrees=25, swing2LimitDegrees=16, twistLimitDegrees=45, twistMotion=Locked}
 {bone1Name=dyn_headMidW_01, bone2Name=head, swing1LimitDegrees=25, swing2LimitDegrees=16, twistLimitDegrees=45, twistMotion=Locked}
 {bone1Name=dyn_headTopA_01, bone2Name=head, swing1LimitDegrees=25, swing2LimitDegrees=16, twistLimitDegrees=8}
 {bone1Name=dyn_headTopB_01, bone2Name=head, swing1LimitDegrees=25, swing2LimitDegrees=16, twistLimitDegrees=8}
 {bone1Name=dyn_headTopC_01, bone2Name=head, swing1LimitDegrees=25, swing2LimitDegrees=16, twistLimitDegrees=8}
 {bone1Name=dyn_headTopD_01, bone2Name=head, swing1LimitDegrees=25, swing2LimitDegrees=16, twistLimitDegrees=8}
 {bone1Name=dyn_headTopE_01, bone2Name=head, swing1LimitDegrees=25, swing2LimitDegrees=16, twistLimitDegrees=8}
 {bone1Name=dyn_headTopF_01, bone2Name=head, swing1LimitDegrees=25, swing2LimitDegrees=16, twistLimitDegrees=8}
 {bone1Name=dyn_headTopG_01, bone2Name=head, swing1LimitDegrees=25, swing2LimitDegrees=16, twistLimitDegrees=8}
 {bone1Name=dyn_llamaStar_01, bone2Name=head, swing1LimitDegrees=45, swing2LimitDegrees=45, twistLimitDegrees=45}
 {bone1Name=dyn_headLowA_02, bone2Name=dyn_headLowA_01, swing1LimitDegrees=25, swing2LimitDegrees=16, twistLimitDegrees=12}
 {bone1Name=dyn_headLowB_02, bone2Name=dyn_headLowB_01, swing1LimitDegrees=25, swing2LimitDegrees=16, twistLimitDegrees=12}
 {bone1Name=dyn_headLowC_02, bone2Name=dyn_headLowC_01, swing1LimitDegrees=25, swing2LimitDegrees=16, twistLimitDegrees=12}
 {bone1Name=dyn_headLowD_02, bone2Name=dyn_headLowD_01, swing1LimitDegrees=25, swing2LimitDegrees=16, twistLimitDegrees=12}
 {bone1Name=dyn_headLowE_02, bone2Name=dyn_headLowE_01, swing1LimitDegrees=25, swing2LimitDegrees=16, twistLimitDegrees=12}
 {bone1Name=dyn_headLowF_02, bone2Name=dyn_headLowF_01, swing1LimitDegrees=25, swing2LimitDegrees=16, twistLimitDegrees=12}
 {bone1Name=dyn_headLowG_02, bone2Name=dyn_headLowG_01, swing1LimitDegrees=25, swing2LimitDegrees=16, twistLimitDegrees=12}
 {bone1Name=dyn_headLowH_02, bone2Name=dyn_headLowH_01, swing1LimitDegrees=25, swing2LimitDegrees=16, twistLimitDegrees=12}
 {bone1Name=dyn_headMidT_02, bone2Name=dyn_headMidT_01, swing1LimitDegrees=20, swing2LimitDegrees=12, twistLimitDegrees=12}
 {bone1Name=dyn_headMidV_02, bone2Name=dyn_headMidV_01, swing1LimitDegrees=25, swing2LimitDegrees=16, twistLimitDegrees=12}
 {bone1Name=dyn_headMidW_02, bone2Name=dyn_headMidW_01, swing1LimitDegrees=25, swing2LimitDegrees=16, twistLimitDegrees=12}
 {bone1Name=dyn_llamaStar_02, bone2Name=dyn_llamaStar_01, swing1LimitDegrees=45, swing2LimitDegrees=45, twistLimitDegrees=45}
 {bone1Name=dyn_headMidT_03, bone2Name=dyn_headMidT_02, swing1LimitDegrees=16, swing2LimitDegrees=16, twistLimitDegrees=16}
]
