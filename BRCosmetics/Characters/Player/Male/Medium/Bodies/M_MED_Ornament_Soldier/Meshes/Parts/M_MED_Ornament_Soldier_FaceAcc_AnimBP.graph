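# Animation Blueprint for the Ornament Soldier's face accessory, layering procedural rigging and physics onto the pose of its attached parent mesh.
# The copied pose and curves pass through a Control Rig, then a RigidBody node simulating in spine_04 space with its own physics asset and gravity override.
# Each update derives an `emoting` value from `bIsPlayingEmote`, driving the physics blend through an interpolated clamp that keeps partial simulation active.
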AnimBlueprint /BRCosmetics/Characters/Player/Male/Medium/Bodies/M_MED_Ornament_Soldier/Meshes/Parts/M_MED_Ornament_Soldier_FaceAcc_AnimBP
target_skeleton=&/M_MED_Ornament_Soldier_FaceAcc_Skeleton
graphs [
 AnimGraph {nodes [
   n1 = Control Rig(Source=Copy Pose From Mesh() {bCopyCurves=true, bUseAttachedParent=true}) {
    alphaBoolBlend {blendOption=Linear}
    controlRigAssetReference {blueprintRigClass=&/M_MED_OrnamentSoldier_FaceAcc_CtrlRig.M_MED_OrnamentSoldier_FaceAcc_CtrlRig_C}
   }
   n2 = RigidBody(ComponentPose=Local To Component(LocalPose=@n1), Alpha=Get emoting() -> (emoting:double)) {
    bClampLinearTranslationLimitToRefPose=true
    bOverrideWorldGravity=true
    overridePhysicsAsset=&/M_MED_Ornament_Soldier_FaceAcc_Physics
    simulationSpace=BaseBoneSpace
    alphaBoolBlend {blendOption=Linear}
    alphaScaleBiasClamp {bClampResult=true, bInterpResult=true, clampMin=0.3, interpSpeedDecreasing=6, interpSpeedIncreasing=6}
    baseBoneRef {boneName=spine_04}
    componentAppliedLinearAccClamp=<3000 3000 3000>
    componentLinearAccScale=<0.4 0.4 0.6>
    componentLinearVelScale=<0.2 2 0.3>
    overrideWorldGravity {z=-980}
   }
   Output Pose(Result=Component To Local(ComponentPose=@n2)) {layerGroup=None}
  ]}
 EventGraph {nodes [
   Event BlueprintUpdateAnimation()
   n1 = Get bIsPlayingEmote()
   > Set emoting(emoting:double=float - float(B="To Float (Boolean)"(InBool=@n1)) {is=title K2Node_CallFunction}) -> (Output_Get:double)
   TryGetPawnOwner()
  ]}
]
variables [emoting {type=number}]
